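; RVT decal static mesh actor in the WinterArt folder that exposes per-instance material settings for a mesh-based decal.
; Its construction script creates a dynamic material instance from Source Material on slot 0 and assigns normal, detail normal, specular mask and diffuse textures.
; It then pushes top and bottom colors, roughness and specular, plus scale, mask, detail and blend controls, onto the mesh's materials.
; BeginPlay, overlap and Tick events are empty.
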(event EventBeginPlay)

(event Collision|EventActorBeginOverlap
  (params (OtherActor :type Actor%20Object%20Reference)))

(event EventTick
  (params (DeltaSeconds :type Float%20%28single-precision%29)))

(fn ConstructionScript
  (bind _staticmeshcomponent (Variables|StaticMeshActor|GetStaticMeshComponent))
  (bind _returnvalue (Rendering|Material|CreateDynamicMaterialInstance _staticmeshcomponent 0 (Variables|Default|GetSource%20Material) "None" :!class /Script/Engine.PrimitiveComponent))
  (Rendering|Material|SetTextureParameterValue _returnvalue "Normal" (Variables|Textures|GetNormal))
  (Rendering|Material|SetTextureParameterValue _returnvalue "DetailsNormal" (Variables|Textures|GetDetailsNormal))
  (Rendering|Material|SetTextureParameterValue _returnvalue "Specular" (Variables|Textures|GetSpecMask))
  (Rendering|Material|SetTextureParameterValue _returnvalue "Diffuse" (Variables|Textures|GetDiffuse))
  (Rendering|Material|SetVectorParameterValueonMaterials _staticmeshcomponent "DecalColor" (Math|Conversions|ToVector(LinearColor) (Variables|Colors|GetColor%20Top)))
  (Rendering|Material|SetVectorParameterValueonMaterials _staticmeshcomponent "DecalColor2" (Math|Conversions|ToVector(LinearColor) (Variables|Colors|GetColor%20Bottom)))
  (Rendering|Material|SetScalarParameterValueonMaterials _staticmeshcomponent "ScaleTextures" (Variables|Colors|GetScaleTextures))
  (Rendering|Material|SetScalarParameterValueonMaterials _staticmeshcomponent "MaskDensity" (Variables|Colors|GetMaskDensity))
  (Rendering|Material|SetScalarParameterValueonMaterials _staticmeshcomponent "RoughnessTweak" (Variables|Colors|GetRoughness%20Top))
  (Rendering|Material|SetScalarParameterValueonMaterials _staticmeshcomponent "SpecularTweak" (Variables|Colors|GetSpecular%20Top))
  (Rendering|Material|SetScalarParameterValueonMaterials _staticmeshcomponent "RoughnessTweak2" (Variables|Colors|GetRoughness%20Bottom))
  (Rendering|Material|SetScalarParameterValueonMaterials _staticmeshcomponent "ColorTransition " (Variables|Colors|GetColorTransition%20))
  (Rendering|Material|SetScalarParameterValueonMaterials _staticmeshcomponent "ColorContrast" (Variables|Colors|GetColorContrast))
  (Rendering|Material|SetScalarParameterValueonMaterials _staticmeshcomponent "DetailsValue" (Variables|Colors|GetDetailsValue))
  (Rendering|Material|SetScalarParameterValueonMaterials _staticmeshcomponent "DetailsScale" (Variables|Colors|GetDetailsScale))
  (Rendering|Material|SetScalarParameterValueonMaterials _staticmeshcomponent "SpecularTweak2" (Variables|Colors|GetSpecular%20Bottom))
  (Rendering|Material|SetScalarParameterValueonMaterials _staticmeshcomponent "Blendin" (Variables|Colors|GetBlend%20In%20))
  (Rendering|Material|SetScalarParameterValueonMaterials _staticmeshcomponent "Blend In Contrast" (Variables|Colors|GetBlend%20In%20Contrast))
  (Rendering|Material|SetScalarParameterValueonMaterials _staticmeshcomponent "Nomal Map Value" (Variables|Colors|GetNomal%20Map%20Value)))
