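; RVT_Decal_Puddle is a static mesh actor Blueprint for a configurable puddle decal in the WinterArt set, presumably for runtime virtual texturing.
; Its construction script creates a dynamic instance of the puddle decal material on slot 0 and sets its Diffuse texture.
; It then pushes exposed color, texture scale, mask density, roughness, specular, height contrast and opacity variables onto the mesh's materials.
; BeginPlay, overlap and Tick events are empty.
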(event EventBeginPlay)

(event Collision|EventActorBeginOverlap
  (params (OtherActor :type Actor%20Object%20Reference)))

(event EventTick
  (params (DeltaSeconds :type Float%20%28single-precision%29)))

(fn ConstructionScript
  (bind _staticmeshcomponent (Variables|StaticMeshActor|GetStaticMeshComponent))
  (bind _returnvalue (Rendering|Material|CreateDynamicMaterialInstance _staticmeshcomponent 0 "/Game/Athena/WinterArt/Materials/M_RVT_Decal_Blueprint_Puddle.M_RVT_Decal_Blueprint_Puddle" "None" :!class /Script/Engine.PrimitiveComponent))
  (Rendering|Material|SetTextureParameterValue _returnvalue "Diffuse" (Variables|Textures|GetDiffuse))
  (Rendering|Material|SetVectorParameterValueonMaterials _staticmeshcomponent "DecalColor" (Math|Conversions|ToVector(LinearColor) (Variables|Colors|GetColorA)))
  (Rendering|Material|SetScalarParameterValueonMaterials _staticmeshcomponent "ScaleTextures" (Variables|Colors|GetScaleTextures))
  (Rendering|Material|SetScalarParameterValueonMaterials _staticmeshcomponent "MaskDensity" (Variables|Colors|GetMaskDensity))
  (Rendering|Material|SetScalarParameterValueonMaterials _staticmeshcomponent "RoughnessTweak" (Variables|Colors|GetRoughnessTweak))
  (Rendering|Material|SetScalarParameterValueonMaterials _staticmeshcomponent "SpecularTweak" (Variables|Colors|GetSpecularTweak))
  (Rendering|Material|SetScalarParameterValueonMaterials _staticmeshcomponent "HeightMapContrast" (Variables|Colors|GetContrast))
  (Rendering|Material|SetScalarParameterValueonMaterials _staticmeshcomponent "Overall Opacity" (Variables|Colors|GetOpacity)))
