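# Material function that computes emissive output for Fortnite Athena WinterArt materials, with day/night and Lumen intensity controls.
# An input mask is scaled by a facing gradient from camera vector dot vertex normal, then multiplied by the emissive color, optionally tinted by base color.
# Brightness dots day/night multipliers against the active time-of-day collection parameter.
# A ray tracing quality switch applies a separate Lumen multiplier, noted to cause swimming at very high values.
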MaterialFunction /Game/Athena/WinterArt/Materials/MF_EmissiveControls
graph {
 properties {description="Multiplier Specific for lumen, super high values will mess with Lumen Scene causing swimming."}
 nodes [
  n1 = FunctionInput() {inputName=EmissiveMask}
  n2 = LinearInterpolate(
   A=ScalarParameter() {defaultValue=-0.1, group=Emissive, parameterName=EmissiveFresnel_Min, sortPriority=5}
   B=ScalarParameter() {defaultValue=1.1, group=Emissive, parameterName=EmissiveFresnel_Max, sortPriority=5}
   Alpha=DotProduct(A=CameraVectorWS(), B=VertexNormalWS())
  )
  n3 = FunctionInput() {inputName=EmissiveColor}
  n4 = StaticSwitchParameter(True=Multiply(A=FunctionInput() {inputName=BaseColor}, B=@n3), False=@n3) {group=Emissive, parameterName=UseBaseColorForEmissive}
  n5 = ScalarParameter() {defaultValue=0.1, group=Emissive, parameterName=DayMult, sortPriority=1}
  n6 = CollectionParameter() {
   collection=&FortniteMaterialParameters
   parameterName=FortniteActiveTimeOfDay
  }
  n7 = DotProduct(
   A=MaterialFunctionCall(X=@n5, Y=@n5, Z=@n5, A=ScalarParameter() {defaultValue=1, group=Emissive, parameterName=NightMult, sortPriority=2}) {materialFunction=&MakeFloat4}.Result
   B=@n6
  )
  n8 = RayTracingQualitySwitch(Normal=@n7, RayTraced=Multiply(A=@n7, B=ScalarParameter() {defaultValue=1, group=Emissive, parameterName=LumenMult})) {desc="Override for Lumen "}
  FunctionOutput(None=Multiply(A=Multiply(A=@n1, B=@n2), B=Multiply(A=@n4, B=@n8)))
 ]
}
