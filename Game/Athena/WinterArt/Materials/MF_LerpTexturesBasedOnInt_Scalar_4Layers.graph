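# Material function that picks one of up to four scalar inputs per pixel from an index stored in the U of UV channel 1.
# Chained If nodes compare U against integer thresholds, giving a hard selection rather than a blend, despite the Lerp name.
# Static switches for 2, 3 or 4 textures set the chain depth at compile time, with the higher count taking priority.
# With all switches at their default off, input 1 passes straight through.
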MaterialFunction /Game/Athena/WinterArt/Materials/MF_LerpTexturesBasedOnInt_Scalar_4Layers
graph {nodes [
  n1 = Constant() {r=3}
  n2 = ComponentMask(None=TextureCoordinate() {coordinateIndex=1}) {r=true}
  n3 = FunctionInput() {inputName=1, inputType=FunctionInput_Scalar}
  n4 = If(A=Constant() {r=1}, B=@n2, "A > B"=@n3, "A < B"=FunctionInput() {inputName=2, inputType=FunctionInput_Scalar, sortPriority=1}) {desc=0}
  n5 = If(A=Constant() {r=2}, B=@n2, "A > B"=@n4, "A < B"=FunctionInput() {inputName=3, inputType=FunctionInput_Scalar, sortPriority=2})
  n6 = FunctionInput() {inputName=4, inputType=FunctionInput_Scalar, sortPriority=3}
  n7 = StaticBool()
  n8 = StaticSwitch(
   True=@n4
   False=@n3
   Value=FunctionInput(Preview=@n7) {bUsePreviewValueAsDefault=true, inputName=Use 2 Textures, inputType=FunctionInput_StaticBool, sortPriority=10}
  )
  n9 = FunctionInput(Preview=@n7) {bUsePreviewValueAsDefault=true, inputName=Use 3 Textures, inputType=FunctionInput_StaticBool, sortPriority=11}
  FunctionOutput(None=StaticSwitch(
    True=If(A=@n1, B=@n2, "A > B"=@n5, "A < B"=@n6)
    False=StaticSwitch(True=@n5, False=@n8, Value=@n9)
    Value=FunctionInput(Preview=@n7) {bUsePreviewValueAsDefault=true, inputName=Use 4 Textures, inputType=FunctionInput_StaticBool, sortPriority=12}
   ))
 ]}
