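# Material function that blends up to four material texture sets into one set of surface outputs for Athena WinterArt materials.
# Shared lerp helper functions, gated by Use 2/3/4 Materials static switches, blend diffuse/height, packed S channels, normals, emissive, clearcoat, subsurface and opacity-cancel inputs.
# Packed alpha masks emissive through MF_EmissiveControls.
# Clearcoat is further masked by upward reflection direction, adjustable with Clearcoat_DownMask, and by a per-layer blended lerp factor named CCMaskByMetal.
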MaterialFunction /Game/Athena/WinterArt/Materials/MF_FN_TextureSetLerp_4Layers
graph {nodes [
  n1 = FunctionInput() {bUsePreviewValueAsDefault=true, inputName=S, inputType=FunctionInput_Vector4, sortPriority=1, previewValue=<1 0 0.2667 1>}
  n2 = FunctionInput() {bUsePreviewValueAsDefault=true, inputName=S2, inputType=FunctionInput_Vector4, sortPriority=4, previewValue=<1 0 0.2667 1>}
  n3 = FunctionInput() {bUsePreviewValueAsDefault=true, inputName=S3, inputType=FunctionInput_Vector4, sortPriority=7, previewValue=<1 0 0.2667 1>}
  n4 = FunctionInput() {bUsePreviewValueAsDefault=true, inputName=S4, inputType=FunctionInput_Vector4, sortPriority=11, previewValue=<1 0 0.2667 1>}
  n5 = MaterialFunctionCall(Float4=MaterialFunctionCall!(
    transitionFactor
    1=@n1
    2=@n2
    3=@n3
    4=@n4
    Use 2 Textures=NamedRerouteUsage()
    Use 3 Textures=NamedRerouteUsage()
    Use 4 Textures=NamedRerouteUsage()
   ) {materialFunction=&/MF_LerpTexturesBasedOnInt_Float4_4Layers}.Result) {materialFunction=&BreakOutFloat4Components}
  n6 = Constant()
  n7 = AppendVector(A=FunctionInput() {bUsePreviewValueAsDefault=true, inputName=E1, sortPriority=17}, B=@n6)
  n8 = AppendVector(A=FunctionInput() {bUsePreviewValueAsDefault=true, inputName=E2, sortPriority=18}, B=@n6)
  n9 = AppendVector(A=FunctionInput() {bUsePreviewValueAsDefault=true, inputName=E3, sortPriority=19}, B=@n6)
  n10 = AppendVector(A=FunctionInput() {bUsePreviewValueAsDefault=true, inputName=E4, sortPriority=21}, B=@n6)
  n11 = NamedRerouteUsage()
  n12 = NamedRerouteUsage()
  n13 = NamedRerouteUsage()
  FunctionOutput(None=MaterialFunctionCall!(
    BaseColor=NamedRerouteUsage()
    EmissiveMask=@n5.A
    EmissiveColor=MaterialFunctionCall!(transitionFactor, 1=@n7, 2=@n8, 3=@n9, 4=@n10, Use 2 Textures=@n11, Use 3 Textures=@n12, Use 4 Textures=@n13) {materialFunction=&/MF_LerpTexturesBasedOnInt_Float3_4Layers}.Result
   ) {materialFunction=&/MF_EmissiveControls}.Result) {outputName=Emissive, sortPriority=4}
  n14 = FunctionInput() {bUsePreviewValueAsDefault=true, inputName=SSS_Depth_1, inputType=FunctionInput_Vector4, sortPriority=38, previewValue=<1 1 1 1>}
  n15 = FunctionInput() {bUsePreviewValueAsDefault=true, inputName=SSS_Depth_2, inputType=FunctionInput_Vector4, sortPriority=39, previewValue=<1 1 1 1>}
  n16 = FunctionInput() {bUsePreviewValueAsDefault=true, inputName=SSS_Depth_3, inputType=FunctionInput_Vector4, sortPriority=40, previewValue=<1 1 1 1>}
  n17 = FunctionInput() {bUsePreviewValueAsDefault=true, inputName=SSS_Depth_4, inputType=FunctionInput_Vector4, sortPriority=41, previewValue=<1 1 1 1>}
  FunctionOutput(None=MaterialFunctionCall(1=@n14, 2=@n15, 3=@n16, 4=@n17, Use 2 Textures=NamedRerouteUsage(), Use 3 Textures=NamedRerouteUsage(), Use 4 Textures=NamedRerouteUsage()) {materialFunction=&MF_LerpTexturesBasedOnInt_Float4_4Layers}.Result) {outputName="SSS Depth ( Opacity)", sortPriority=11}
  n18 = Multiply(
   A=NamedRerouteUsage()
   B=MaterialFunctionCall!(
    1=FunctionInput() {bUsePreviewValueAsDefault=true, inputName=CC1, inputType=FunctionInput_Scalar, sortPriority=22}
    2=FunctionInput() {bUsePreviewValueAsDefault=true, inputName=CC2, inputType=FunctionInput_Scalar, sortPriority=23}
    3=FunctionInput() {bUsePreviewValueAsDefault=true, inputName=CC3, inputType=FunctionInput_Scalar, sortPriority=24}
    4=FunctionInput() {bUsePreviewValueAsDefault=true, inputName=CC4, inputType=FunctionInput_Scalar, sortPriority=25}
    Use 2 Textures=NamedRerouteUsage()
    Use 3 Textures=NamedRerouteUsage()
    Use 4 Textures=NamedRerouteUsage()
   ) {materialFunction=&/MF_LerpTexturesBasedOnInt_Scalar_4Layers}.Result
  )
  n19 = ReflectionVectorWS()
  n20 = Max(A=ScalarParameter() {defaultValue=1, parameterName=Clearcoat_DownMask}, B=Saturate(None=ComponentMask(None=@n19) {b=true}))
  n21 = NamedRerouteUsage()
  n22 = MaterialFunctionCall(
   1=FunctionInput() {bUsePreviewValueAsDefault=true, inputName=CCMaskByMetal_1, inputType=FunctionInput_Scalar, sortPriority=22}
   2=FunctionInput() {bUsePreviewValueAsDefault=true, inputName=CCMaskByMetal_2, inputType=FunctionInput_Scalar, sortPriority=23}
   3=FunctionInput() {bUsePreviewValueAsDefault=true, inputName=CCMaskByMetal_3, inputType=FunctionInput_Scalar, sortPriority=24}
   4=FunctionInput() {bUsePreviewValueAsDefault=true, inputName=CCMaskByMetal_4, inputType=FunctionInput_Scalar, sortPriority=25}
   Use 2 Textures=NamedRerouteUsage()
   Use 3 Textures=NamedRerouteUsage()
   Use 4 Textures=NamedRerouteUsage()
  ) {materialFunction=&MF_LerpTexturesBasedOnInt_Scalar_4Layers}
  FunctionOutput(None=Multiply(A=Multiply(A=@n18, B=@n20), B=LinearInterpolate(B=@n21, Alpha=@n22.Result) {constA=1})) {outputName=Clearcoat, sortPriority=5}
  FunctionOutput(None=@n5.A) {outputName=Spec_A, sortPriority=12}
  FunctionOutput(None=MaterialFunctionCall(
    1=FunctionInput() {bUsePreviewValueAsDefault=true, inputName=N, sortPriority=2, previewValue=<0 0 1 1>}
    2=FunctionInput() {bUsePreviewValueAsDefault=true, inputName=N2, sortPriority=5, previewValue=<0 0 1 1>}
    3=FunctionInput() {bUsePreviewValueAsDefault=true, inputName=N3, sortPriority=8, previewValue=<0 0 1 1>}
    4=FunctionInput() {bUsePreviewValueAsDefault=true, inputName=N4, sortPriority=12, previewValue=<0 0 1 1>}
    Use 2 Textures=NamedRerouteUsage()
    Use 3 Textures=NamedRerouteUsage()
    Use 4 Textures=NamedRerouteUsage()
   ) {materialFunction=&MF_LerpTexturesBasedOnInt_Float3_4Layers}.Result) {outputName=Normals, sortPriority=5}
  FunctionOutput(None=@n5.R) {outputName=S, sortPriority=1}
  FunctionOutput(None=MaterialFunctionCall(
    1=FunctionInput() {bUsePreviewValueAsDefault=true, inputName=CCR_1, inputType=FunctionInput_Scalar, sortPriority=22}
    2=FunctionInput() {bUsePreviewValueAsDefault=true, inputName=CCR_2, inputType=FunctionInput_Scalar, sortPriority=23}
    3=FunctionInput() {bUsePreviewValueAsDefault=true, inputName=CCR_3, inputType=FunctionInput_Scalar, sortPriority=24}
    4=FunctionInput() {bUsePreviewValueAsDefault=true, inputName=CCR_4, inputType=FunctionInput_Scalar, sortPriority=25}
    Use 2 Textures=NamedRerouteUsage()
    Use 3 Textures=NamedRerouteUsage()
    Use 4 Textures=NamedRerouteUsage()
   ) {materialFunction=&MF_LerpTexturesBasedOnInt_Scalar_4Layers}.Result) {outputName=Clearcoat_Roughness, sortPriority=5}
  FunctionOutput(None=@n5.B) {outputName=R, sortPriority=3}
  n23 = FunctionInput() {bUsePreviewValueAsDefault=true, inputName=OpacityMaskCancel_1, sortPriority=30}
  n24 = FunctionInput() {bUsePreviewValueAsDefault=true, inputName=OpacityMaskCancel_2, sortPriority=31}
  n25 = FunctionInput() {bUsePreviewValueAsDefault=true, inputName=OpacityMaskCancel_3, sortPriority=32}
  n26 = FunctionInput() {bUsePreviewValueAsDefault=true, inputName=OpacityMaskCancel_4, sortPriority=33}
  n27 = NamedRerouteUsage()
  n28 = NamedRerouteUsage()
  n29 = NamedRerouteUsage()
  FunctionOutput(None=Max(
    A=NamedRerouteUsage()
    B=MaterialFunctionCall(1=@n23, 2=@n24, 3=@n25, 4=@n26, Use 2 Textures=@n27, Use 3 Textures=@n28, Use 4 Textures=@n29) {materialFunction=&MF_LerpTexturesBasedOnInt_Float4_4Layers}.Result
   )) {outputName=Opacity Mask, sortPriority=5}
  n30 = FunctionInput() {bUsePreviewValueAsDefault=true, inputName=SSS_1, inputType=FunctionInput_Vector4, sortPriority=34, previewValue=<1 1 1 1>}
  n31 = FunctionInput() {bUsePreviewValueAsDefault=true, inputName=SSS_2, inputType=FunctionInput_Vector4, sortPriority=35, previewValue=<1 1 1 1>}
  n32 = FunctionInput() {bUsePreviewValueAsDefault=true, inputName=SSS_3, inputType=FunctionInput_Vector4, sortPriority=36, previewValue=<1 1 1 1>}
  n33 = FunctionInput() {bUsePreviewValueAsDefault=true, inputName=SSS_4, inputType=FunctionInput_Vector4, sortPriority=37, previewValue=<1 1 1 1>}
  n34 = MaterialFunctionCall(1=@n30, 2=@n31, 3=@n32, 4=@n33, Use 2 Textures=NamedRerouteUsage(), Use 3 Textures=NamedRerouteUsage(), Use 4 Textures=NamedRerouteUsage()) {materialFunction=&MF_LerpTexturesBasedOnInt_Float4_4Layers}
  n35 = NamedRerouteUsage()
  FunctionOutput(None=Multiply(A=Multiply(A=@n34.Result, B=@n35), B=NamedRerouteUsage())) {outputName=SSS Color, sortPriority=10}
  n36 = FunctionInput(Preview=StaticBoolParameter() {group=Material_2, parameterName=Use 2 Materials, sortPriority=-1}) {bUsePreviewValueAsDefault=true, inputName=Use2Materials, inputType=FunctionInput_StaticBool, sortPriority=-1}
  NamedRerouteDeclaration(None=@n36) {name=Use2, nodeColor=<rgba 0.2891 0 1 1>}
  n37 = FunctionInput(Preview=StaticBoolParameter() {group=Material_3, parameterName=Use 3 Materials, sortPriority=-1}) {bUsePreviewValueAsDefault=true, inputName=Use3Materials, inputType=FunctionInput_StaticBool, sortPriority=-1}
  NamedRerouteDeclaration(None=@n37) {name=Use3, nodeColor=<rgba 0.2891 0 1 1>}
  n38 = FunctionInput(Preview=StaticBoolParameter() {group=Material_4, parameterName=Use 4 Materials, sortPriority=-1}) {bUsePreviewValueAsDefault=true, inputName=Use4Materials, inputType=FunctionInput_StaticBool, sortPriority=-1}
  NamedRerouteDeclaration(None=@n38) {name=Use4, nodeColor=<rgba 1 0 0.0469 1>}
  n39 = Constant4Vector() {constant=<rgba 0.46 0.4009 0.2599 0>}
  n40 = MaterialFunctionCall(
   1=FunctionInput(Preview=@n39.RGBA) {bUsePreviewValueAsDefault=true, inputName=D, inputType=FunctionInput_Vector4}
   2=FunctionInput(Preview=@n39.RGBA) {bUsePreviewValueAsDefault=true, inputName=D2, inputType=FunctionInput_Vector4, sortPriority=3}
   3=FunctionInput(Preview=@n39.RGBA) {bUsePreviewValueAsDefault=true, inputName=D3, inputType=FunctionInput_Vector4, sortPriority=6}
   4=FunctionInput(Preview=@n39.RGBA) {bUsePreviewValueAsDefault=true, inputName=D4, inputType=FunctionInput_Vector4, sortPriority=9}
   Use 2 Textures=NamedRerouteUsage()
   Use 3 Textures=NamedRerouteUsage()
   Use 4 Textures=NamedRerouteUsage()
  ) {materialFunction=&MF_LerpTexturesBasedOnInt_Float4_4Layers}
  n41 = FunctionOutput(None=ComponentMask(None=@n40.Result) {b=true, g=true, r=true}) {outputName=Diffuse}
  NamedRerouteDeclaration(None=@n41) {name=Diffuse, nodeColor=<rgba 0 1 0.6016 1>}
  n42 = FunctionOutput(None=ComponentMask(None=@n40.Result) {a=true}) {outputName=Height, sortPriority=12}
  NamedRerouteDeclaration(None=@n42) {name=Alpha/Height, nodeColor=<rgba 1 0.375 0 1>}
  n43 = FunctionOutput(None=@n5.G) {outputName=M, sortPriority=2}
  NamedRerouteDeclaration(None=@n43) {name=Metal, nodeColor=<rgba 0 1 0.5313 1>}
  Reroute(None=@n41)
  Reroute(None=@n38)
  Reroute(None=@n37)
  Reroute(None=@n36)
  Reroute(None=@n42)
  Reroute(None=@n43)
 ]}
