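# Material function that adds layered detail normals to incoming material attributes, toggled by static switch and applied only at High and Epic quality.
# Layers 1 and 3 sample normal textures through world-space biplanar mapping, while layers 2 and 4 are flat constants.
# A four-layer lerp function combines them, then angle-corrected blending merges the result with the base normal.
# Only Normal is overridden; spec-mask packing never reaches the output.
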MaterialFunction /Game/Athena/WinterArt/Materials/MF_4Layer_Detail
graph {nodes [
  n1 = GetMaterialAttributes(None=FunctionInput() {inputName=Materials, inputType=FunctionInput_MaterialAttributes}) {attributeGetTypes [
    #1
    #2
    #3
    #4
   ]}
  n2 = StaticBoolParameter() {group=DetailMaps, parameterName=Use2DetailMaterials}
  n3 = StaticBoolParameter() {group=DetailMaps, parameterName=Use4DetailMaterials}
  n4 = StaticBoolParameter() {group=DetailMaps, parameterName=Use3DetailMaterials}
  n5 = NamedRerouteUsage()
  n6 = NamedRerouteUsage()
  n7 = NamedRerouteUsage()
  n8 = NamedRerouteUsage()
  n9 = SetMaterialAttributes!(
   MaterialAttributes=@n1.MaterialAttributes
   Normal=MaterialFunctionCall(
    BaseNormal=NamedRerouteUsage()
    AdditionalNormal=MaterialFunctionCall!(
     Use2Materials=@n2
     Use4Materials=@n3
     Use3Materials=@n4
     D
     S
     N=@n5
     D2
     S2
     N2=@n6
     D3
     S3
     N3=@n7
     D4
     S4
     N4=@n8
     E1
     E2
     E3
     E4
     CC1
     CC2
     CC3
     CC4
    ) {materialFunction=&/MF_FN_TextureSetLerp_4Layers}.Normals
   ) {materialFunction=&BlendAngleCorrectedNormals}.Result
  ) {attributeSetTypes [#4]}
  FunctionOutput(None=StaticSwitchParameter(True=QualitySwitch(Default=@n1.MaterialAttributes, High=@n9, Epic=@n9), False=@n1.MaterialAttributes) {group=DetailMaps, parameterName=Use4layerDetailMaps})
  n10 = TextureObjectParameter() {
   group=DetailLayer_01
   parameterName=DL1_Normal_Texture
   samplerType=SAMPLERTYPE_Normal
   sortPriority=0
   texture=&FlatNormalBlue
  }
  n11 = MaterialFunctionCall!(
   ColorTex
   NormalTex=@n10
   Pos=WorldPosition().XYZ
   Nrm=VertexNormalWS()
   ColorUVScale=ScalarParameter() {defaultValue=100, group=DetailLayer_01, parameterName=DL1_UVScale, sortPriority=1}
   ColorBlendFactor=ScalarParameter() {defaultValue=1, group=DetailLayer_01, parameterName=DL1_UVBlendFactor, sortPriority=2}
   NormalStrength=ScalarParameter() {defaultValue=1, group=DetailLayer_01, parameterName=DL1_Normal_Strength, primitiveDataIndex=1, sortPriority=3}
   NormalUVScale=ScalarParameter() {defaultValue=100, group=DetailLayer_01, parameterName=DL1_Normal_UVScale, sortPriority=1}
   NormalBlendFactor=ScalarParameter() {defaultValue=8, group=DetailLayer_01, parameterName=DL1_Normal_UVBlendFactor, sortPriority=2}
   Local Space Inputs
  ) {materialFunction=/Game/Global/MaterialFunctions/MF_BiPlanar}
  n12 = MaterialFunctionCall(Float3=@n11.Color) {materialFunction=&BreakOutFloat3Components}
  MaterialFunctionCall(X=Min(A=@n12.R, B=NamedRerouteUsage()), Y=@n12.G, Z=Max(A=@n12.B, B=NamedRerouteUsage()), A=Constant() {r=1}) {materialFunction=&MakeFloat4}
  NamedRerouteDeclaration(None=@n11.Normals) {name=L1_Normal, nodeColor=<rgba 0 0.1565 1 1>}
  NamedRerouteDeclaration(None=@n1.Normal) {name=Base_Normal, nodeColor=<rgba 0.0547 0 1 1>}
  NamedRerouteDeclaration(None=MaterialFunctionCall(
    X=NamedRerouteDeclaration(None=@n1.Roughness) {name=Base_Roughness, nodeColor=<rgba 0.0547 0 1 1>}
    Y=NamedRerouteDeclaration(None=@n1.Specular) {name=Base_Specular, nodeColor=<rgba 0.0547 0 1 1>}
    Z=NamedRerouteDeclaration(None=@n1.Metallic) {name=Base_Metallic, nodeColor=<rgba 0.0547 0 1 1>}
    A=Constant()
   ) {materialFunction=&MakeFloat4}.Result) {name=BaseSpecMasks, nodeColor=<rgba 0 1 0.3672 1>}
  NamedRerouteDeclaration(None=Constant()) {name=L1_Spec, nodeColor=<rgba 0.6916 0 1 1>}
  n13 = TextureObjectParameter() {
   group=DetailLayer_03
   parameterName=DL3_Normal_Texture
   samplerType=SAMPLERTYPE_Normal
   sortPriority=0
   texture=&FlatNormalBlue
  }
  n14 = MaterialFunctionCall(
   NormalTex=@n13
   Pos=WorldPosition().XYZ
   Nrm=VertexNormalWS()
   ColorUVScale=ScalarParameter() {defaultValue=100, group=DetailLayer_03, parameterName=DL3_UVScale, sortPriority=1}
   ColorBlendFactor=ScalarParameter() {defaultValue=8, group=DetailLayer_03, parameterName=DL3_UVBlendFactor, sortPriority=2}
   NormalStrength=ScalarParameter() {defaultValue=1, group=DetailLayer_03, parameterName=DL3_Normal_Strength, primitiveDataIndex=1, sortPriority=3}
   NormalUVScale=ScalarParameter() {defaultValue=100, group=DetailLayer_03, parameterName=DL3_Normal_UVScale, sortPriority=1}
   NormalBlendFactor=ScalarParameter() {defaultValue=8, group=DetailLayer_03, parameterName=DL3_Normal_UVBlendFactor, sortPriority=2}
  ) {materialFunction=&MF_BiPlanar}
  n15 = MaterialFunctionCall(Float3=@n14.Color) {materialFunction=&BreakOutFloat3Components}
  n16 = NamedRerouteUsage()
  n17 = NamedRerouteUsage()
  NamedRerouteDeclaration(None=MaterialFunctionCall(X=Min(A=@n15.R, B=@n16), Y=@n15.G, Z=Max(A=@n15.B, B=@n17), A=Constant() {r=1}) {materialFunction=&MakeFloat4}.Result) {name=L3_Spec, nodeColor=<rgba 0.6916 0 1 1>}
  n18 = Constant3Vector() {constant=<rgba 0 0 1 0>}
  NamedRerouteDeclaration(None=@n18) {name=L2_Normal, nodeColor=<rgba 0 0.1565 1 1>}
  NamedRerouteDeclaration(None=@n14.Normals) {name=L3_Normal, nodeColor=<rgba 0 0.1565 1 1>}
  NamedRerouteDeclaration(None=@n18) {name=L4_Normal, nodeColor=<rgba 0 0.1565 1 1>}
  NamedRerouteUsage()
  NamedRerouteUsage()
  NamedRerouteUsage()
  NamedRerouteUsage()
 ]}
