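# Material function for a coaster rail that lights up emissively around a driver's world location, passing through other material attributes.
# A sphere mask around the DriverWorldLocation parameter blends a base color to a boosted lit color, scaled by a switch scalar.
# The SpecularMasks alpha selects where that result replaces the incoming emissive.
# The DaisyCoasterRail static switch bypasses the effect when off.
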MaterialFunction /Game/Athena/WinterArt/Materials/MF_DaisyCoasterRail
graph {nodes [
  n1 = FunctionInput() {inputType=FunctionInput_MaterialAttributes}
  n2 = GetMaterialAttributes(None=@n1) {attributeGetTypes [#1]}
  n3 = VectorParameter() {defaultValue=<rgba 0 0 0 1>}
  n4 = Multiply(
   A=VectorParameter() {group=Coaster, parameterName=Coaster_LitColor, defaultValue=<rgba 1 0 0.4242 1>}.RGB
   B=ScalarParameter() {defaultValue=2, group=Coaster, parameterName=Coaster_LitEmissiveMult}
  )
  n5 = SphereMask(
   A=WorldPosition().XYZ
   B=VectorParameter() {parameterName=DriverWorldLocation, defaultValue=<rgba 0 0 0 1>}.RGB
   Radius=ScalarParameter() {defaultValue=250, group=Coaster, parameterName=Coaster_LitRange}
   Hardness=ScalarParameter() {defaultValue=0.25, group=Coaster, parameterName=Coaster_LitHardness}
  )
  n6 = TextureSampleParameter2D() {
   group=Material_1
   parameterName=SpecularMasks
   samplerSource=SSM_Wrap_WorldGroupSettings
   samplerType=SAMPLERTYPE_Masks
   sortPriority=-5
   texture=&T_Fortnite_Default_S
  }
  n7 = LinearInterpolate(
   A=@n2.EmissiveColor
   B=Multiply(A=LinearInterpolate(A=@n3.RGB, B=@n4, Alpha=@n5), B=ScalarParameter() {defaultValue=1, group=Coaster, parameterName=Coaster_LitSwitch})
   Alpha=@n6.A
  )
  FunctionOutput(None=StaticSwitchParameter(
    True=SetMaterialAttributes!(MaterialAttributes=@n2.MaterialAttributes, EmissiveColor=@n7) {attributeSetTypes [#1]}
    False=@n1
   ) {parameterName=DaisyCoasterRail})
 ]}
