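# Material function in the Athena winter art set that outputs a 0-1 gradient mask along a chosen mesh-local axis.
# The pixel's offset from the actor position is transformed from world to local space, so the mask follows mesh orientation.
# That vector gets a fixed X/Y shift, is dotted with the VP_MaskByMeshOrientation parameter (local Y by default), then divided by a constant and saturated.
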MaterialFunction /Game/Athena/WinterArt/Materials/MF_MeshOrientationMask
graph {nodes [
  n1 = Subtract(A=WorldPosition().XYZ, B=ActorPositionWS())
  n2 = DotProduct(
   A=Add(
    A=Transform(None=@n1) {transformSourceType=TRANSFORMSOURCE_World, transformType=TRANSFORM_Local}
    B=Constant3Vector() {constant=<rgba 5 5 0 1>}
   )
   B=VectorParameter() {group=UpVector, parameterName=VP_MaskByMeshOrientation, defaultValue=<rgba 0 1 0 0>}.RGB
  )
  n3 = Constant() {r=5}
  FunctionOutput(None=Saturate(None=Divide(A=@n2, B=@n3)))
 ]}
