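# Material function that adds an animated "Reality Glitch" look to incoming material attributes behind a static switch, passing them through unchanged when off.
# Emissive builds on time-panned triplanar texture samples blended by vertex normal.
# Named reroutes define biplanar-noise distortion, fresnel, normal-derivative curvature, color tint and object-height masks.
# Glitch pulses come from powered sine waves, a pivot-derived seed offers per-object variation, and time-hashed jitter forms World Position Offset.
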MaterialFunction /Game/Athena/WinterArt/Materials/MF_RealityGlitch
graph {nodes [
  n1 = Divide(A=WorldPosition().XYZ, B=Constant3Vector() {constant=<rgba 150 150 150 0>})
  n2 = NamedRerouteUsage()
  n3 = NamedRerouteUsage()
  n4 = Add(
   A=Multiply(A=Time(), B=VectorParameter() {group=PanningEmissive, parameterName=TriPlanner Speed, defaultValue=<rgba 0.01 0.02 -0.05 1>}.RGB)
   B=LinearInterpolate(A=Multiply(A=@n2) {constB=0.04}, B=Multiply(A=@n3) {constB=0.14}, Alpha=NamedRerouteUsage())
  ) {constB=0}
  n5 = ComponentMask(None=Add(A=@n1, B=@n4)) {b=true, g=true}
  Add(B=@n5)
  n6 = ComponentMask(None=Add(A=@n1, B=@n4)) {g=true, r=true}
  Add(B=@n6)
  n7 = ComponentMask(None=Add(A=@n1, B=@n4)) {b=true, r=true}
  Add(B=@n7)
  Constant2Vector() {g=300, r=100}
  n8 = PixelNormalWS()
  n9 = DDX(Value=@n8)
  n10 = DDY(Value=@n8)
  n11 = Power(Base=Max(A=DotProduct(A=@n9, B=@n9), B=DotProduct(A=@n10, B=@n10))) {constExponent=0.3}
  Divide(A=Floor(None=Multiply(A=@n11) {constB=10})) {constB=20}
  n12 = VectorParameter() {group=PanningEmissive, parameterName=TriPlanner Speed3, defaultValue=<rgba 0 0 -20 1>}
  n13 = Multiply(A=Time(), B=@n12.RGB)
  Frac(None=@n13)
  n14 = VectorParameter() {group=PanningEmissive, parameterName=TriPlanner Speed3, defaultValue=<rgba 0 0 -20 1>}
  n15 = Multiply(A=Time(), B=@n14.RGB)
  Frac(None=@n15)
  n16 = Multiply(A=Time(), B=VectorParameter() {group=PanningEmissive, parameterName=TriPlanner Speed2, defaultValue=<rgba -5 2 -5 1>}.RGB)
  Frac(None=@n16)
  n17 = FunctionInput() {inputType=FunctionInput_MaterialAttributes}
  n18 = NamedRerouteUsage()
  n19 = NamedRerouteUsage()
  FunctionOutput(None=StaticSwitchParameter(
    True=SetMaterialAttributes!(MaterialAttributes=@n17, EmissiveColor=@n18, BaseColor, WorldPositionOffset=@n19) {attributeSetTypes [#1, #2, #3]}
    False=@n17
   ) {parameterName=Reality Glitch})
  n20 = PreSkinnedLocalBounds()
  n21 = MaterialFunctionCall(Float3=Divide(A=WorldPosition().XYZ, B=Subtract(A=@n20.Max, B=@n20.Max))) {materialFunction=&BreakOutFloat3Components}
  n22 = Multiply(A=Time()) {constB=0.1}
  n23 = NamedRerouteUsage()
  n24 = TextureSample(UVs=Add(
    A=MaterialFunctionCall(X=@n21.R, Y=@n21.R) {materialFunction=&MakeFloat2}.Result
    B=Add(A=@n22, B=@n23)
   )) {texture=/Game/Environments/Artemis/Props/Daisy/Daisy_RealityGlitch/Textures/T_4colors}
  n25 = Saturate(None=Multiply(A=@n24.RGB))
  LinearInterpolate(A=@n25, B=Saturate(None=Multiply(A=@n24.RGB) {constB=10}), Alpha=NamedRerouteUsage())
  Multiply(A=NamedRerouteUsage(), B=@n14.RGB)
  Multiply(A=NamedRerouteUsage(), B=@n12.RGB)
  n26 = WorldPosition()
  n27 = Constant() {r=3}
  n28 = TextureObjectParameter() {
   group=PanningEmissive
   parameterName=EmissiveGradient2
   samplerType=SAMPLERTYPE_LinearGrayscale
   texture=/Game/Weapons/FORT_Melee/Pickaxe_SnowFall_Female/FX/Snowfall_pick_Noise
  }
  n29 = MaterialFunctionCall!(
   ColorTex=@n28
   NormalTex
   Pos=Add(A=Divide(A=@n26.XYZ, B=@n27), B=@n15)
   Nrm
   ColorUVScale=Constant() {r=100}
   ColorBlendFactor
   NormalStrength
   NormalUVScale
   NormalBlendFactor
   Local Space Inputs
  ) {materialFunction=/Game/Global/MaterialFunctions/MF_BiPlanar}
  NamedRerouteDeclaration(None=Subtract(A=Multiply(A=@n29.Color) {constB=2})) {name=Distortion_small, nodeColor=<rgba 1 0.7031 0 1>}
  n30 = NamedRerouteUsage()
  n31 = TextureObjectParameter() {
   group=PanningEmissive
   parameterName=EmissiveGradient
   samplerType=SAMPLERTYPE_LinearColor
   texture=/Game/Environments/Artemis/Props/Daisy/Daisy_RealityGlitch/Textures/T_Tendrils
  }
  n32 = TextureSample(UVs=@n5, Tex=@n31) {samplerType=SAMPLERTYPE_Masks, texture=/Game/Weapons/WeaponSkins/Textures/T_Swirly_Noise_highContrast}
  n33 = TextureSample(UVs=@n7, Tex=@n31) {samplerType=SAMPLERTYPE_Masks, texture=&T_Swirly_Noise_highContrast}
  n34 = VertexInterpolator(VS=VertexNormalWS())
  n35 = Abs(None=ComponentMask(None=@n34.PS) {g=true})
  n36 = ScalarParameter() {defaultValue=1, group=PanningEmissive, parameterName=TriPlanner_BlendStrength}
  n37 = Saturate(None=ComponentMask(None=MaterialFunctionCall(In=@n35, Contrast=@n36) {materialFunction=&CheapContrast}.Result) {g=true})
  n38 = ComponentMask(None=@n34.PS) {b=true}
  n39 = ComponentMask(None=MaterialFunctionCall(In=Abs(None=@n38), Contrast=Constant() {r=1}) {materialFunction=&CheapContrast}.Result) {b=true}
  n40 = Subtract(
   A=LinearInterpolate(
    A=LinearInterpolate(A=@n32.RGB, B=@n33.RGB, Alpha=@n37)
    B=TextureSample(UVs=@n6, Tex=@n31) {samplerType=SAMPLERTYPE_Masks, texture=&T_Swirly_Noise_highContrast}.RGB
    Alpha=Saturate(None=@n39)
   )
   B=Constant()
  )
  n41 = ScalarParameter() {defaultValue=5, parameterName=FinalEmissiveMult}
  n42 = Constant() {r=50}
  n43 = Saturate(None=NamedRerouteUsage())
  n44 = Add(A=Multiply(A=Multiply(A=@n30, B=@n40), B=LinearInterpolate(A=@n41, B=@n42, Alpha=@n43)) {constB=5}, B=NamedRerouteUsage()) {constB=0}
  n45 = NamedRerouteUsage()
  n46 = NamedRerouteUsage()
  NamedRerouteDeclaration(None=Add(A=Multiply(A=@n44, B=@n45), B=Multiply(A=@n46) {constB=0.01})) {name=Emissive, nodeColor=<rgba 1 0.4453 0 1>}
  n47 = TextureObjectParameter() {group=PanningEmissive, parameterName=EmissiveGradient, samplerType=SAMPLERTYPE_Grayscale, texture=/FortWater/Textures/LowResBlurredNoise}
  n48 = WorldPosition()
  n49 = Constant() {r=3}
  n50 = MaterialFunctionCall(ColorTex=@n47, Pos=Add(A=Divide(A=@n48.XYZ, B=@n49), B=@n16)) {materialFunction=&MF_BiPlanar}
  NamedRerouteDeclaration(None=Subtract(A=Multiply(A=@n50.Color) {constB=2})) {name=Distortion, nodeColor=<rgba 1 0.7031 0 1>}
  NamedRerouteDeclaration(None=Multiply(A=@n25)) {name=Color Tint, nodeColor=<rgba 1 0.8906 0 1>}
  n51 = WorldPosition()
  n52 = Add(
   A=MaterialFunctionCall() {materialFunction=&ObjectPivotPoint}.Object Pivot Location
   B=Constant3Vector()
  )
  n53 = ComponentMask(None=Divide(
    A=Subtract(A=@n51.XYZ, B=@n52)
    B=MaterialFunctionCall() {materialFunction=&ObjectLocalBounds}.Local Bounds Size
   )) {b=true}
  NamedRerouteDeclaration(None=OneMinus(None=SmoothStep(Value=@n53) {constMax=0.3})) {name=Height Mask, nodeColor=<rgba 0 0.1562 1 1>}
  n54 = Multiply(A=Time(), B=Constant() {r=0.3}) {constB=0.3}
  n55 = NamedRerouteUsage()
  n56 = Sine(None=Frac(None=Add(A=@n54, B=@n55)))
  n57 = Step(Y=NamedRerouteUsage()) {constX=0.5}
  n58 = Sine(None=Multiply(A=@n54, B=Multiply(A=@n57) {constB=0.6}) {constB=0.5})
  n59 = Add(A=Frac(None=Power(Base=@n56) {constExponent=162}), B=Frac(None=Power(Base=@n58) {constExponent=164}))
  NamedRerouteDeclaration(None=Saturate(None=Divide(A=@n59))) {name=Glitch, nodeColor=<rgba 0.2656 0 1 1>}
  NamedRerouteDeclaration(None=Multiply(A=Multiply(A=@n11) {constB=0.06}, B=Constant3Vector() {constant=<rgba 0 0.4532 0.8854 0>})) {name=Curvature, nodeColor=<rgba 0 0.4141 1 1>}
  n60 = Add(A=NamedRerouteUsage(), B=Multiply(A=NamedRerouteUsage()))
  n61 = Constant() {r=0.25}
  n62 = Constant() {r=0.4}
  n63 = NamedRerouteUsage()
  NamedRerouteDeclaration(None=Multiply(A=SmoothStep(Value=@n60) {constMax=4}, B=LinearInterpolate(A=@n61, B=@n62, Alpha=@n63)) {constA=0.5, constB=0.25}) {name=Fresnel, nodeColor=<rgba 0.2656 0 1 1>}
  n64 = Subtract(
   A=WorldPosition().XYZ
   B=Add(
    A=MaterialFunctionCall() {materialFunction=&ObjectPivotPoint}.Object Pivot Location
    B=Constant3Vector()
   )
  )
  n65 = MaterialFunctionCall() {materialFunction=&ObjectLocalBounds}
  n66 = Abs(None=ComponentMask(None=Divide(A=@n64, B=@n65.Local Bounds Size)) {b=true})
  NamedRerouteDeclaration(None=OneMinus(None=SmoothStep(Value=@n66) {constMax=0.65, constMin=0.2})) {name=Height Mask 2, nodeColor=<rgba 0.9688 1 0 1>}
  NamedRerouteDeclaration(None=Multiply(A=Power(Base=@n50.Color) {constExponent=1.25}, B=NamedRerouteUsage())) {name=Opacity Mask, nodeColor=<rgba 1 0.6328 0 1>}
  n67 = MaterialFunctionCall(Float3=MaterialFunctionCall() {materialFunction=&ObjectPivotPoint}.Object Pivot Location) {materialFunction=&BreakOutFloat3Components}
  n68 = Add(A=Add(A=Abs(None=@n67.R), B=Abs(None=@n67.G)), B=Abs(None=@n67.B))
  NamedRerouteDeclaration(None=Frac(None=Divide(A=@n68) {constB=3})) {name=Random Seed 0-1, nodeColor=<rgba 0.8281 1 0 1>}
  n69 = MaterialFunctionCall(Float=Add(A=Time(), B=NamedRerouteUsage())) {materialFunction=&MF_Hash}
  n70 = Multiply(A=@n69.Float) {constB=7}
  n71 = Multiply(A=@n69.Float) {constB=8}
  NamedRerouteDeclaration(None=Multiply(
    A=MaterialFunctionCall(X=@n70, Y=@n71, Z=@n70) {materialFunction=&MakeFloat3}.Result
    B=NamedRerouteUsage()
   )) {name=WPO, nodeColor=<rgba 1 0.375 0 1>}
  NamedRerouteUsage()
  OneMinus(None=Power(Base=@n50.Color) {constExponent=0.1})
  n72 = WorldPosition()
  n73 = Constant() {r=3}
  n74 = MaterialFunctionCall(ColorTex=@n28, Pos=Add(A=Divide(A=@n72.XYZ, B=@n73), B=@n13)) {materialFunction=&MF_BiPlanar}
  Power(Base=@n74.Color) {constExponent=0.1}
  Power(Base=@n29.Color) {constExponent=0.1}
  ScalarParameter() {defaultValue=15, parameterName=FinalEmissiveMult}
  Subtract(A=Multiply(A=@n74.Color) {constB=2})
  VectorParameter() {group=PanningEmissive, parameterName=TriPlanner Speed2, defaultValue=<rgba -5 2 -5 1>}
 ]}
