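# Translucent, two-sided subsurface ice material for Athena winter art, usable on Nanite and instanced static meshes.
# Color combines MF_SharedWinter's ice color, scene-depth tint, a world-height gradient modulated by vertex blue, and view-dependent highlights from vertex and bump-offset scratch normals.
# Fresnel and scene depth drive depth-faded opacity.
# Inverted vertex red height-blends through the scratch mask into a pale layer, altering color and opacity.
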Material /Game/Athena/WinterArt/Materials/M_Athena_IceTransparent
graph {
 properties {
  bUsedWithInstancedStaticMeshes=true
  bUsedWithNanite=true
  blendMode=BLEND_Translucent
  pixelDepthOffsetMode=PDOM_Legacy
  shadingModel=MSM_Subsurface
  translucencyLightingMode=TLM_VolumetricPerVertexDirectional
  twoSided=true
  usedShadingModels=Subsurface
 }
 nodes [
  n1 = LinearInterpolate(A=Constant() {r=0.5}, B=Constant() {r=5}, Alpha=Fresnel(Normal=VertexNormalWS()) {baseReflectFraction=0, exponent=3}) {constB=2}
  n2 = Constant() {r=1}
  n3 = Saturate(None=Divide(A=SceneDepth().R, B=Constant() {r=2000}))
  n4 = VertexColor()
  n5 = OneMinus(None=@n4.R)
  n6 = TextureSample() {samplerType=SAMPLERTYPE_Masks, texture=/Game/Global/Textures/Default/Tileable_Patterns/T_IceScratch1_M}
  n7 = Constant() {r=1.0683}
  n8 = Saturate(None=Multiply(
    A=Constant() {r=1}
    B=MaterialFunctionCall(Transition Phase=@n5, Height Texture=@n6.R, Contrast=@n7) {materialFunction=&HeightLerp}.Alpha
   ))
  LinearInterpolate(A=DepthFade(Opacity=LinearInterpolate(A=@n1, B=@n2, Alpha=@n3), FadeDistance=Constant() {r=179.3529}), Alpha=@n8)
  n9 = Power(Base=DotProduct(A=CameraVectorWS(), B=VertexNormalWS()), Exp=Constant() {r=52.6377})
  n10 = MaterialFunctionCall() {materialFunction=&/MF_SharedWinter}
  n11 = Multiply(A=Multiply(A=Saturate(None=@n9), B=Constant() {r=0.5}), B=@n10.IceColor)
  n12 = TextureCoordinate()
  n13 = Constant() {r=0.5}
  n14 = TextureSample(UVs=BumpOffset(Coordinate=Multiply(A=@n12, B=@n13), Height=@n6.R, HeightRatioInput=Constant() {r=-0.05})) {samplerType=SAMPLERTYPE_Normal, texture=/Game/Global/Textures/Default/Tileable_Patterns/T_IceScratch1_N}
  n15 = DotProduct(A=CameraVectorWS(), B=Transform(None=@n14.RGB))
  n16 = Constant() {r=12}
  n17 = Multiply(A=Saturate(None=Power(Base=@n15, Exp=@n16)), B=Constant() {r=0.075})
  n18 = Multiply(A=Constant() {r=0.35}, B=LinearInterpolate(A=Constant3Vector() {constant=<rgba 0.0234 0.2689 0.3646 1>}, B=@n10.IceColor, Alpha=@n3))
  n19 = Multiply(A=@n10.IceColor, B=Constant3Vector() {constant=<rgba 2.2274 2.5 2.0045 0>})
  n20 = Subtract(A=ComponentMask(None=WorldPosition().XYZ) {b=true}, B=Constant() {r=828})
  n21 = Constant() {r=522.3811}
  n22 = ScalarParameter() {defaultValue=0.5}
  n23 = Saturate(None=Multiply(A=Divide(A=@n20, B=@n21), B=LinearInterpolate(A=@n22, Alpha=@n4.B)))
  LinearInterpolate(
   A=Add(A=Add(A=@n11, B=@n17), B=LinearInterpolate(A=@n18, B=@n19, Alpha=@n23))
   B=Constant3Vector() {constant=<rgba 0.901 0.8483 0.8208 1>}
   Alpha=@n8
  )
  MaterialFunctionCall(Normal=@n14.RGB, Flatness=Constant() {r=0.75}) {materialFunction=&FlattenNormal}
  VectorParameter() {defaultValue=<rgba 0.0126 0.08 0.1615 0>}
 ]
}
