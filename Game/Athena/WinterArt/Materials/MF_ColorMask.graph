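# Material function that tints a background color through a height-blended mask, with metalness suppressing the tint.
# Transition, optionally capped by a remapped world-aligned noise paint mask via static switch, is scaled by inverted Metalness to drive HeightLerp with a height input and contrast.
# It outputs the background multiplied by a white-to-tint blend, the blend alpha as Mask, and a direct background-to-tint blend as LerpMix.
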MaterialFunction /Game/Athena/WinterArt/Materials/MF_ColorMask
graph {nodes [
  n1 = FunctionInput() {inputName=TintColor}
  n2 = FunctionInput() {bUsePreviewValueAsDefault=true, inputName=Transition, inputType=FunctionInput_Scalar, previewValue=<0.5 0 0 1>}
  n3 = TextureObjectParameter() {
   group=WS Paint Mask
   parameterName=WSPaintMask
   samplerType=SAMPLERTYPE_LinearGrayscale
   texture=/Game/Global/Textures/Default/Noise/T_SwirlNoise_H2
  }
  n4 = ScalarParameter() {defaultValue=1566.42, group=WS Paint Mask, parameterName=WS Paint Mask Size, sortPriority=2}
  n5 = LinearInterpolate(
   A=ScalarParameter() {defaultValue=0.1476, group=WS Paint Mask, parameterName=WSPaintMask Min, sortPriority=3}
   B=ScalarParameter() {defaultValue=0.7837, group=WS Paint Mask, parameterName=WSPaintMask Max, sortPriority=4}
   Alpha=StaticComponentMaskParameter(None=MaterialFunctionCall(TextureObject=@n3, TextureSize=@n4) {materialFunction=/Game/Weapons/WeaponSkins/Wraps/MaterialFunctions/WorldAlignedTexture}.XYZ Texture) {defaultR=true, group=WS Paint Mask, parameterName=WSPaintMask Channel, sortPriority=1}
  )
  n6 = Multiply(
   A=StaticSwitchParameter(True=Min(A=@n2, B=@n5), False=@n2) {group=WS Paint Mask, parameterName=WSPaintMask True}
   B=OneMinus(None=FunctionInput() {bUsePreviewValueAsDefault=true, inputName=Metalness, inputType=FunctionInput_Scalar})
  )
  n7 = FunctionInput() {bUsePreviewValueAsDefault=true, inputName=HM, inputType=FunctionInput_Scalar, previewValue=<0.5 0 0 1>}
  n8 = Saturate(None=@n7)
  n9 = FunctionInput() {bUsePreviewValueAsDefault=true, inputName=Contrast, inputType=FunctionInput_Scalar, previewValue=<0.5 0 0 1>}
  n10 = MaterialFunctionCall(A=Constant() {r=1}, B=@n1, Transition Phase=@n6, Height Texture=@n8, Contrast=@n9) {materialFunction=&HeightLerp}
  n11 = FunctionInput() {bUsePreviewValueAsDefault=true, inputName=BG Color}
  FunctionOutput(None=Multiply(A=@n10.Results, B=@n11))
  FunctionOutput(None=@n10.Alpha) {outputName=Mask}
  FunctionOutput(None=MaterialFunctionCall(A=@n11, B=@n1, Transition Phase=@n6, Height Texture=@n8, Contrast=@n9) {materialFunction=&HeightLerp}.Results) {outputName=LerpMix}
 ]}
