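# Material function that multiplies an emissive input by an animated, switch-selected mask, apparently for screens or signs.
# Flipbook mode plays SubUV frames with static, panning, parallax or scale-jump options, plus noise-driven frame glitches and scanlines.
# Mask mode pans quarter-offset UVs, then dots packed mask channels with an animated or input selector, or uses RGB with optional HeightLerp.
# Actor position can randomize timing; mobile paths drop costlier effects.
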MaterialFunction /Game/Athena/WinterArt/Materials/MF_EmissiveMask
graph {nodes [
  n1 = FunctionInput() {inputName=EmissiveTexture}
  n2 = TextureObjectParameter() {group=Adv Emissive, parameterName=EmissiveColor, texture=&T_BlacksRGB}
  n3 = StaticSwitchParameter(True=TextureCoordinate() {coordinateIndex=1}, False=TextureCoordinate()) {group=Adv Emissive, parameterName=FlipbookStaticUV2, sortPriority=15}
  n4 = ScalarParameter() {group=Adv Emissive, parameterName=FlipbookStaticDepth, sortPriority=16}
  n5 = AppendVector(
   A=ScalarParameter() {defaultValue=2, group=Adv Emissive, parameterName=X Rows, sortPriority=2}
   B=ScalarParameter() {defaultValue=2, group=Adv Emissive, parameterName=Y Rows, sortPriority=2}
  )
  n6 = TextureCoordinate()
  n7 = Panner(
   Coordinate=@n6
   Speed=AppendVector(
    A=ScalarParameter() {group=Adv Emissive, parameterName=X Panner, sortPriority=2}
    B=ScalarParameter() {defaultValue=2, group=Adv Emissive, parameterName=Y Panner, sortPriority=2}
   )
  )
  n8 = Multiply(A=Time(), B=ScalarParameter() {defaultValue=0.1, group=Adv Emissive, parameterName=FrameSpeed, sortPriority=1})
  n9 = Divide(
   A=MaterialFunctionCall!(f2, f3=ActorPositionWS(), f4) {materialFunction=/Game/Packages/Fortnite_UniqueMaterials_1/VectorComponentOps}.f3 Add
   B=ScalarParameter() {defaultValue=50, group=Adv Emissive, parameterName=Random Distance}
  )
  n10 = StaticSwitchParameter(True=Add(A=@n8, B=@n9), False=@n8) {group=Adv Emissive, parameterName=Random Timing}
  n11 = Ceil(None=@n10)
  n12 = MaterialFunctionCall!(
   Texture=@n2
   UVs=BumpOffset(Coordinate=@n7, Height=ScalarParameter() {defaultValue=-0.2, group=Adv Emissive, parameterName=Parallax})
   SubImages=@n5
   Frame=@n11
  ) {materialFunction=/Engine/Functions/Engine_MaterialFunctions03/Texturing/SubUV_Function}
  n13 = MaterialFunctionCall(Texture=@n2, UVs=@n7, SubImages=@n5, Frame=@n11) {materialFunction=&SubUV_Function}
  n14 = ScalarParameter() {defaultValue=1.1, group=Adv Emissive, parameterName=BG_JumpMax}
  n15 = Sine(None=Add(A=Time(), B=@n9))
  n16 = LinearInterpolate(A=@n14, Alpha=@n15)
  n17 = StaticSwitchParameter(
   True=MaterialFunctionCall(UVs=@n6, Texture Scale=StaticSwitchParameter(True=@n16, False=@n14) {group=Adv Emissive, parameterName=Random Timing}) {materialFunction=&ScaleUVsByCenter}.UVs
   False=@n6
  ) {group=Adv Emissive, parameterName=AnimJump}
  n18 = MaterialFunctionCall(Texture=@n2, UVs=@n17, SubImages=@n5, Frame=@n11) {materialFunction=&SubUV_Function}
  n19 = TextureCoordinate()
  n20 = ComponentMask(None=Multiply(
    A=Constant3Vector() {constant=<rgba 0.0025 15 0 0>}
    B=ScalarParameter() {defaultValue=1, group=Adv Emissive, parameterName=ScreenEffect Scale, sortPriority=5}
   )) {g=true, r=true}
  n21 = ScalarParameter() {defaultValue=1, group=Adv Emissive, parameterName=ScreenEffect_Speed, sortPriority=5}
  n22 = Constant2Vector() {g=1}
  n23 = TextureSample(UVs=Panner(Coordinate=Multiply(A=@n19, B=@n20), Speed=Multiply(A=@n21, B=@n22))) {samplerType=SAMPLERTYPE_Grayscale, texture=&LowResBlurredNoise}
  n24 = Multiply(A=ScalarParameter() {defaultValue=0.0125, group=Adv Emissive, parameterName=ScreenEffect Distortion, sortPriority=4}, B=@n23.R)
  n25 = Add(
   A=@n10
   B=ScalarParameter() {defaultValue=0.04, desc=Time for next frame to start, group=Adv Emissive, parameterName=TimeOffset, sortPriority=3}
  )
  n26 = Power(Base=@n23.G) {constExponent=12}
  n27 = Constant() {r=7699.6504}
  n28 = LinearInterpolate(
   A=@n18.RGB
   B=MaterialFunctionCall(Texture=@n2, UVs=Add(A=@n17, B=@n24), SubImages=@n5, Frame=Ceil(None=@n25)) {materialFunction=&SubUV_Function}.RGB
   Alpha=Saturate(None=Multiply(A=@n26, B=@n27))
  )
  n29 = TextureSampleParameter2D(UVs=@n9) {
   group=Adv Emissive
   parameterName=AnimMask
   samplerSource=SSM_Wrap_WorldGroupSettings
   samplerType=SAMPLERTYPE_Masks
   texture=/Game/Athena/WinterArt/Textures/T_AnimMask
  }
  n30 = LinearInterpolate(B=ScalarParameter() {defaultValue=1.1, group=Adv Emissive, parameterName=FG_JumpMax}, Alpha=@n15) {constA=1}
  n31 = LinearInterpolate(
   A=@n28
   B=@n29.RGB
   Alpha=MaterialFunctionCall(
    Texture=@n2
    UVs=MaterialFunctionCall(UVs=@n6, Texture Scale=@n30) {materialFunction=&ScaleUVsByCenter}.UVs
    SubImages=@n5
    Frame=@n11
   ) {materialFunction=&SubUV_Function}.Alpha
  )
  n32 = StaticSwitchParameter(
   True=MaterialFunctionCall(
    Texture=@n2
    UVs=BumpOffset(Coordinate=@n3, Height=@n4)
    SubImages=@n5
    Frame=ScalarParameter() {group=Adv Emissive, parameterName=Flipbook Frame, sortPriority=15}
   ) {materialFunction=&SubUV_Function}.RGB
   False=StaticSwitchParameter(
    True=StaticSwitchParameter(True=@n12.RGB, False=@n13.RGB) {group=Adv Emissive, parameterName=Flipbook Panner Parallax}
    False=StaticSwitchParameter(True=@n31, False=@n28) {group=Adv Emissive, parameterName=AnimJump}
   ) {group=Adv Emissive, parameterName=Flipbook Panner true}
  ) {group=Adv Emissive, parameterName=Flipbook Static, sortPriority=15}
  n33 = LinearInterpolate(
   A=Multiply(A=@n32, B=ScalarParameter() {defaultValue=0.5, group=Adv Emissive, parameterName=ScreenLines, sortPriority=7}) {constB=2}
   B=@n32
   Alpha=@n23.G
  ) {constB=0.5}
  n34 = Ceil(None=Add(B=FunctionInput() {inputName=TextureOffset, inputType=FunctionInput_Scalar, previewValue=<1 0 0 1>}) {constA=-1})
  n35 = Add(A=TextureCoordinate(), B=Multiply(A=Constant2Vector() {g=1}, B=Divide(A=@n34) {constB=4}))
  n36 = StaticSwitchParameter(
   True=BumpOffset(Coordinate=@n35, Height=ScalarParameter() {defaultValue=0.25, desc="0.5 is no parallax ", group=Adv Emissive, parameterName=Parallax Height})
   False=@n35
  ) {group=Adv Emissive, parameterName=Paralax true}
  n37 = Panner(
   Coordinate=@n36
   Speed=AppendVector(
    A=FunctionInput() {inputName=X, inputType=FunctionInput_Scalar, previewValue=<0.1 0 0 1>}
    B=FunctionInput() {inputName=Y, inputType=FunctionInput_Scalar}
   )
  )
  n38 = TextureSample(UVs=@n37, Tex=@n2) {texture=/Game/Environments/Sets/Neo/Props/Textures/T_AnimMask2}
  n39 = Constant()
  n40 = Multiply(A=Time(), B=ScalarParameter() {defaultValue=-1, group=Adv Emissive, parameterName=HeightTransition Time})
  n41 = Frac(None=StaticSwitchParameter(True=Add(A=@n40, B=@n9), False=@n40) {group=Adv Emissive, parameterName=Random Timing})
  n42 = Multiply(A=Time(), B=ScalarParameter() {defaultValue=1, group=Adv Emissive, parameterName=AnimMaskTime})
  n43 = Add(A=@n42, B=@n9)
  n44 = TextureSampleParameter2D(UVs=Frac(None=StaticSwitchParameter(True=@n43, False=@n42) {group=Adv Emissive, parameterName=Random Timing})) {
   group=Adv Emissive
   parameterName=AnimMask
   samplerSource=SSM_Wrap_WorldGroupSettings
   samplerType=SAMPLERTYPE_Masks
   texture=&T_AnimMask
  }
  n45 = StaticSwitchParameter(True=@n44.RGB, False=FunctionInput() {inputName=Dot}) {group=Adv Emissive, parameterName=AnimMask true}
  n46 = TextureSampleParameter2D(UVs=@n37) {
   group=Adv Emissive
   parameterName=EmissiveMask
   samplerSource=SSM_Wrap_WorldGroupSettings
   samplerType=SAMPLERTYPE_Masks
   texture=/Game/Athena/WinterArt/Textures/T_Mall_Run2_M
  }
  n47 = StaticSwitchParameter(
   True=StaticSwitchParameter(
    True=MaterialFunctionCall(A=@n38.RGB, B=@n39, Transition Phase=@n41, Height Texture=@n38.A) {materialFunction=&HeightLerp}.Results
    False=@n38.RGB
   ) {group=Adv Emissive, parameterName="HeightMap Animation "}
   False=DotProduct(A=@n45, B=@n46.RGB)
  ) {group=Adv Emissive, parameterName=RGB Texture true}
  n48 = TextureSample(UVs=@n36, Tex=@n2) {texture=&T_AnimMask2}
  n49 = TextureSampleParameter2D(UVs=@n36) {
   group=Adv Emissive
   parameterName=EmissiveMask
   samplerSource=SSM_Wrap_WorldGroupSettings
   samplerType=SAMPLERTYPE_Masks
   texture=&T_Mall_Run2_M
  }
  n50 = DotProduct(A=@n49.RGB, B=@n45)
  n51 = StaticSwitchParameter(
   True=ShadingPathSwitch(Default=@n33, Deferred=@n33, Mobile=@n18.RGB) {desc=Migrated from FeatureLevelSwitch}
   False=ShadingPathSwitch(Default=@n47, Deferred=@n47, Mobile=StaticSwitchParameter(True=@n48.RGB, False=@n50) {group=Adv Emissive, parameterName=RGB Texture true}) {desc=Migrated from FeatureLevelSwitch}
  ) {group=Adv Emissive, parameterName=AdvEmissiveFlipbook}
  FunctionOutput(None=StaticSwitchParameter(True=Multiply(A=@n1, B=@n51), False=@n1) {group=Adv Emissive, parameterName="AdvEmissiveTrue "}) {outputName=EmissiveMasked}
 ]}
