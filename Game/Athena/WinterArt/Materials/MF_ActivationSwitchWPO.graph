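# Material function that adds an optional activation switch rotation to incoming material attributes, turning vertex-color-masked geometry through world position offset.
# The masked region rotates around a parameterized local axis and pivot, at a fixed angle or animated between OFF and ON angles.
# Animation eases over a duration from a Blueprint-supplied start time, with a Blueprint toggle setting direction.
# Rotated normals are corrected and blended in, except on forward and mobile shading paths.
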MaterialFunction /Game/Athena/WinterArt/Materials/MF_ActivationSwitchWPO
graph {nodes [
  n1 = GetMaterialAttributes(None=FunctionInput() {inputType=FunctionInput_MaterialAttributes}) {attributeGetTypes [#1, #2]}
  n2 = VertexColor()
  n3 = StaticComponentMaskParameter(None=AppendVector(A=@n2, B=@n2)) {defaultR=true, group=ActivationSwitch, parameterName=ActivationSwitch_VertexMaskChannel}
  n4 = Normalize(VectorInput=Transform(None=VectorParameter() {group=ActivationSwitch, parameterName=ActivationSwitch_Rotation_Axis, defaultValue=<rgba 1 0 0 1>}.RGB) {transformSourceType=TRANSFORMSOURCE_Local})
  n5 = Subtract(A=Time(), B=ScalarParameter() {group=ActivationSwitch, parameterName=ActivationSwitch_GameTimeFromBP})
  n6 = Add(A=ScalarParameter() {defaultValue=2, group=ActivationSwitch, parameterName=ActivationSwitch_Animation_Duration}) {constB=0}
  n7 = Power(
   Base=Saturate(None=Divide(A=@n5, B=@n6))
   Exp=ScalarParameter() {defaultValue=0.25, group=ActivationSwitch, parameterName=ActivationSwitch_Animation_EaseInOut}
  ) {constExponent=8}
  n8 = OneMinus(None=@n7)
  n9 = ScalarParameter() {defaultValue=1, group=ActivationSwitch, parameterName=ActivationSwitch_Animation_ToggleFromBP}
  n10 = StaticSwitchParameter(
   True=LinearInterpolate(
    A=ScalarParameter() {group=ActivationSwitch, parameterName=ActivationSwitch_Animation_OFF_Rotation}
    B=ScalarParameter() {group=ActivationSwitch, parameterName=ActivationSwitch_Animation_ON_Rotation}
    Alpha=LinearInterpolate(A=@n8, B=@n7, Alpha=@n9)
   )
   False=ScalarParameter() {group=ActivationSwitch, parameterName=ActivationSwitch_Rotation}
  ) {group=ActivationSwitch, parameterName=ActivationSwitch_Animated}
  n11 = TransformPosition(None=VectorParameter() {group=ActivationSwitch, parameterName=ActivationSwitch_Rotation_Pivot, defaultValue=<rgba 0 0 0 1>}.RGB) {transformType=TRANSFORMPOSSOURCE_World}
  n12 = WorldPosition()
  n13 = Add(
   A=@n1.WorldPositionOffset
   B=Multiply(A=@n3, B=RotateAboutAxis(NormalizedRotationAxis=@n4, RotationAngle=@n10, PivotPoint=@n11, Position=@n12.XYZ))
  )
  n14 = MaterialFunctionCall(Rotation Axis=@n4, Rotation Angle=@n10) {materialFunction=&FixRotateAboutAxisNormals}
  n15 = LinearInterpolate(
   A=@n1.Normal
   B=MaterialFunctionCall(BaseNormal=@n1.Normal, AdditionalNormal=VertexInterpolator(VS=@n14.New Tangent Space Vertex Normal).PS) {materialFunction=&BlendAngleCorrectedNormals}.Result
   Alpha=@n3
  )
  n16 = ShadingPathSwitch(Default=@n15, Deferred=@n15, Forward=@n1.Normal, Mobile=@n1.Normal)
  FunctionOutput(None=StaticSwitchParameter(
    True=SetMaterialAttributes!(MaterialAttributes=@n1.MaterialAttributes, WorldPositionOffset=@n13, Normal=@n16) {attributeSetTypes [#1, #2]}
    False=@n1.MaterialAttributes
   ) {parameterName=Use ActivationSwitch})
 ]}
