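# MF_Events is a material function that serves as a shared hook for adding and disabling events across the game.
# It takes a material attributes input and passes it through MF_LockdownPulse, the one event effect currently wired in.
# The pulse function's result is the function's only output, so materials using it pick up that event effect through this single wrapper.
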MaterialFunction /Game/Athena/WinterArt/Materials/MF_Events
graph {
 properties {description=Use This function for adding and disabling events across the game}
 nodes [FunctionOutput(None=MaterialFunctionCall!(MA In=FunctionInput() {inputType=FunctionInput_MaterialAttributes}) {materialFunction=/Game/Global/MaterialFunctions/MF_LockdownPulse}.Result)]
}
