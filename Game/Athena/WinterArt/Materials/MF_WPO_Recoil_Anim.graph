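# Material function that adds a backward recoil animation to guns and turrets through World Position Offset, behind a static switch.
# Time since a start-time parameter, divided by duration and clamped, drives eased sine curves with power falloffs.
# These curves push vertices along a local direction and rotate them around a local axis at the object pivot.
# The offset adds to existing WPO, and normals are corrected for the rotation.
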MaterialFunction /Game/Athena/WinterArt/Materials/MF_WPO_Recoil_Anim
graph {nodes [
  n1 = FunctionInput() {inputType=FunctionInput_MaterialAttributes}
  n2 = GetMaterialAttributes(None=@n1) {attributeGetTypes [#1, #2]}
  n3 = Time()
  n4 = ScalarParameter() {group=WPO_Anim, parameterName=RecoilAnimStartTime}
  n5 = Saturate(None=Divide(A=Subtract(A=@n3, B=@n4), B=ScalarParameter() {defaultValue=0.5, group=WPO_Anim, parameterName=RecoilAnimDuration}))
  n6 = Power(Base=@n5) {constExponent=0.6}
  n7 = OneMinus(None=@n5)
  n8 = Multiply(
   A=Transform(None=VectorParameter() {group=WPO_Anim, parameterName=RecoilDirectionLocal, defaultValue=<rgba 1 0 0 1>}.RGB) {transformSourceType=TRANSFORMSOURCE_Local}
   B=Multiply(A=Sine(None=@n6) {period=1.5}, B=Power(Base=@n7))
  )
  n9 = ScalarParameter() {defaultValue=-30, group=WPO_Anim, parameterName=RecoilDistance}
  n10 = Normalize(VectorInput=Transform(None=VectorParameter() {group=WPO_Anim, parameterName=RecoilRotationAxisLocal, defaultValue=<rgba 0 1 0 1>}.RGB) {transformSourceType=TRANSFORMSOURCE_Local})
  n11 = Multiply(
   A=Multiply(A=Sine(None=@n6) {period=0.8}, B=Power(Base=@n7) {constExponent=4})
   B=ScalarParameter() {defaultValue=-0.006, group=WPO_Anim, parameterName=RecoilRotation}
  )
  n12 = MaterialFunctionCall() {materialFunction=&ObjectPivotPoint}
  n13 = WorldPosition()
  n14 = Add(
   A=@n2.WorldPositionOffset
   B=Add(
    A=Multiply(A=@n8, B=@n9)
    B=RotateAboutAxis(NormalizedRotationAxis=@n10, RotationAngle=@n11, PivotPoint=@n12.Object Pivot Location, Position=@n13.XYZ)
   )
  )
  n15 = MaterialFunctionCall(Rotation Axis=@n10, Rotation Angle=@n11, World Space Vertex Normals=Transform(None=@n2.Normal)) {materialFunction=&FixRotateAboutAxisNormals}
  FunctionOutput(None=StaticSwitchParameter(
    True=SetMaterialAttributes!(MaterialAttributes=@n2.MaterialAttributes, WorldPositionOffset=@n14, Normal=@n15.New Tangent Space Vertex Normal) {attributeSetTypes [#1, #2]}
    False=@n1
   ) {desc=Backward recoil WPO animation for guns and turrets, group=WPO_Anim, parameterName=UseRecoilWPOAnim, sortPriority=0})
 ]}
