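# Material function that layers optional animated world position offset onto input material attributes.
# It combines axis rotation (spinning, oscillating or ramp-driven) with texture-timed normal push and double-sine bounce, using actor position for phase variation.
# Static switches select the motion, masked by vertex color or the incoming Refraction attribute, and rotation also corrects normals.
# Mobile bypasses this animation, and every path then feeds MF_WPO_Recoil_Anim before output.
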MaterialFunction /Game/Athena/WinterArt/Materials/MF_WPOAnim
graph {nodes [
  DotProduct(A=ComponentMask(None=ActorPositionWS()) {b=true, g=true, r=true}, B=Constant() {r=1})
  n1 = Normalize(VectorInput=Transform(None=VectorParameter() {group=WPO_Anim, parameterName=WPO RotationAxis, sortPriority=11, defaultValue=<rgba 0 0 1 0>}.RGB) {transformSourceType=TRANSFORMSOURCE_Local})
  n2 = ActorPositionWS()
  n3 = Frac(None=Divide(
    A=MaterialFunctionCall!(f2, f3=@n2, f4) {materialFunction=/Game/Packages/Fortnite_UniqueMaterials_1/VectorComponentOps}.f3 Add
    B=Constant() {r=32}
   ))
  n4 = Add(A=LinearInterpolate(B=@n3, Alpha=ScalarParameter() {defaultValue=1, group=WPO_Anim, parameterName=UseWorldRandom}), B=Time())
  n5 = ScalarParameter() {defaultValue=1, group=WPO_Anim, parameterName=Rotation Speed, sortPriority=8}
  n6 = Add(A=Multiply(A=Time(), B=ScalarParameter() {defaultValue=0.5, group=WPO_Anim, parameterName=WPO Speed, sortPriority=9}), B=@n3)
  n7 = TextureSampleParameter2D(UVs=@n6) {
   group=WPO_Anim
   parameterName=WPO Push Timing
   samplerType=SAMPLERTYPE_LinearGrayscale
   sortPriority=5
   texture=/Game/Global/Textures/Default/Masks_and_Gradients/AnimRamp
  }
  n8 = StaticSwitchParameter(
   True=LinearInterpolate(A=Multiply(A=@n4, B=@n5), B=NamedRerouteUsage(), Alpha=NamedRerouteUsage())
   False=Sine(None=Multiply(A=@n7.R, B=@n5)) {period=2}
  ) {group=WPO_Anim, parameterName=Simple Rotation, sortPriority=1}
  n9 = RotateAboutAxis(
   NormalizedRotationAxis=@n1
   RotationAngle=@n8
   PivotPoint=Add(
    A=MaterialFunctionCall() {materialFunction=&ObjectPivotPoint}.Object Pivot Location
    B=VectorParameter() {group=WPO_Anim, parameterName=RotationPivotOffset, sortPriority=12}.RGB
   )
   Position=WorldPosition().XYZ
  )
  n10 = GetMaterialAttributes(None=FunctionInput() {inputType=FunctionInput_MaterialAttributes}) {attributeGetTypes [#1, #2, #3]}
  n11 = StaticSwitchParameter(
   True=DotProduct(A=VertexColor(), B=VectorParameter() {group=WPO_Anim, parameterName=Vertex color mask, sortPriority=4, defaultValue=<rgba 0 1 0 1>}.RGB)
   False=StaticSwitchParameter(True=@n10.Refraction, False=@n10.Refraction) {desc=Otherwise Height from Diffuse, group=WPO_Anim, parameterName=Use Spec Alpha, sortPriority=2}
  ) {group=WPO_Anim, parameterName=Use Vertex Color Mask, sortPriority=3}
  n12 = Multiply(A=@n11, B=@n9)
  n13 = Constant3Vector() {constant=<rgba 1 1 0 0>}
  n14 = LinearInterpolate(
   A=@n10.Normal
   B=Add(
    A=Multiply(A=@n10.Normal, B=@n13)
    B=MaterialFunctionCall(Rotation Axis=@n1, Rotation Angle=@n8) {materialFunction=&FixRotateAboutAxisNormals}.New Tangent Space Vertex Normal
   )
   Alpha=@n11
  )
  n15 = ScalarParameter() {defaultValue=15, group=WPO_Anim, parameterName=WPO Push Scale, sortPriority=6}
  n16 = Multiply(A=@n6, B=ScalarParameter() {defaultValue=1, group=WPO_Anim, parameterName=Bounce Speed, sortPriority=7})
  n17 = Add(A=Sine(None=@n16), B=Sine(None=@n16) {period=3})
  n18 = ScalarParameter() {defaultValue=1, group=WPO_Anim, parameterName=WPO Offset Scale, sortPriority=7}
  n19 = Add(
   A=Multiply(A=VertexNormalWS(), B=Multiply(A=@n7.R, B=@n15))
   B=Multiply(A=Multiply(A=@n17, B=@n18), B=VectorParameter() {group=WPO_Anim, parameterName=WPO AnimOffset Vector, defaultValue=<rgba 0 0 1 0>}.RGB)
  )
  n20 = Multiply(
   A=@n11
   B=StaticSwitchParameter(True=Add(A=@n19, B=@n9), False=@n19) {defaultValue=true, desc=Uses Push Texure, group=WPO_Anim, parameterName=Rotation True, sortPriority=1}
  )
  n21 = StaticSwitchParameter(True=@n14, False=@n10.Normal) {defaultValue=true, desc=Uses Push Texure, group=WPO_Anim, parameterName=Rotation True, sortPriority=1}
  n22 = StaticSwitchParameter(
   True=StaticSwitchParameter(
    True=SetMaterialAttributes!(MaterialAttributes=@n10.MaterialAttributes, WorldPositionOffset=@n12, Normal=@n14) {attributeSetTypes [#4, #3]}
    False=SetMaterialAttributes(MaterialAttributes=@n10.MaterialAttributes, WorldPositionOffset=@n20, Normal=@n21) {attributeSetTypes [#4, #3]}
   ) {group=WPO_Anim, parameterName=Rotation only, sortPriority=1}
   False=@n10.MaterialAttributes
  ) {group=WPO_Anim, parameterName=Use WPO Anim, sortPriority=-1}
  n23 = ShadingPathSwitch(Default=@n22, Deferred=@n22, Forward=@n22, Mobile=@n10.MaterialAttributes)
  FunctionOutput(None=MaterialFunctionCall!(In=@n23) {materialFunction=&/MF_WPO_Recoil_Anim}.Result)
  n24 = Multiply(A=ScalarParameter() {defaultValue=1, group=WPO_Anim, parameterName=Rotation Speed, sortPriority=8}, B=Time())
  n25 = ScalarParameter() {desc="If angle > 0 then oscilate back and forth between the specified angle.", group=WPO_Anim, parameterName=OscillateAngle, sortPriority=64}
  n26 = Divide(A=@n25) {constA=15, constB=360}
  NamedRerouteDeclaration(None=Multiply(A=Sine(None=@n24), B=Saturate(None=@n26)) {constB=0.5}) {name=SimpleOscTime_Route, nodeColor=<rgba 0 1 0.5781 1>}
  NamedRerouteDeclaration(None=Saturate(None=Ceil(None=@n25))) {name=SimpleOscEnabled_Route, nodeColor=<rgba 0.0781 0 1 1>}
  n27 = NamedRerouteUsage()
  SetMaterialAttributes!(MaterialAttributes=@n23, EmissiveColor=@n27, BaseColor=@n27) {attributeSetTypes [#5, #6]}
 ]}
