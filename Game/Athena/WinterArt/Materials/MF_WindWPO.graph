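# Material function that adds simulated wind to incoming material attributes through World Position Offset and normal perturbation.
# A panned noise normal map on a switchable UV set drives both: its green channel scales an XYZ offset vector, its red and green channels perturb the normal.
# Both effects are masked by a UV ramp, optionally weighted by a vertex color channel.
# A static switch disables the effect, and the mobile shading path always passes the input through unchanged.
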MaterialFunction /Game/Athena/WinterArt/Materials/MF_WindWPO
graph {
 properties {description="This function gives assets wind simultion through their material attributes. Gives access to wind speed, wind strength, and direction."}
 nodes [
  n1 = FunctionInput() {inputType=FunctionInput_MaterialAttributes}
  n2 = GetMaterialAttributes(None=@n1) {attributeGetTypes [#1, #2]}
  n3 = StaticSwitchParameter(
   True=StaticSwitchParameter(True=TextureCoordinate() {coordinateIndex=4}, False=TextureCoordinate() {coordinateIndex=3}) {parameterName=WindUse UV4}
   False=TextureCoordinate() {coordinateIndex=2}
  ) {parameterName=WindUse UV3}
  n4 = ComponentMask(None=AppendVector(
    A=ScalarParameter() {defaultValue=0.4, group=Wind, parameterName=UV_Wind_X, sortPriority=2}
    B=ScalarParameter() {defaultValue=0.2, group=Wind, parameterName=UV_Wind_Y, sortPriority=2}
   )) {g=true, r=true}
  n5 = ScalarParameter() {defaultValue=-0.5, group=Wind, parameterName=WindSpeed_X, sortPriority=1}
  n6 = ScalarParameter() {group=Wind, parameterName=WindSpeed_Y, sortPriority=1}
  n7 = TextureSampleParameter2D(UVs=Panner(Coordinate=Multiply(A=@n3, B=@n4), Speed=AppendVector(A=@n5, B=@n6))) {group=Wind, parameterName=WindTexture, samplerType=SAMPLERTYPE_Normal, texture=/Game/Global/Textures/Default/Noise/T_SwirlNoise_N}
  n8 = Add(A=@n2.Normal, B=Multiply(A=@n7.RGB, B=Constant3Vector() {constant=<rgba 1 1 0 0>}))
  n9 = ComponentMask(None=TextureCoordinate() {coordinateIndex=2}) {r=true}
  n10 = VertexColor()
  n11 = Multiply(
   A=LinearInterpolate(A=ScalarParameter() {group=Wind, parameterName=UV Ramp Min Weight}, Alpha=@n9)
   B=StaticComponentMaskParameter(None=AppendVector(A=@n10, B=@n10)) {defaultA=true, group=Wind, parameterName=ChannelSelect}
  )
  n12 = StaticSwitchParameter(True=@n11, False=@n9) {group=Wind, parameterName=UseVertexColorsMask}
  n13 = AppendVector(
   A=ScalarParameter() {defaultValue=5, group=Wind, parameterName=WPO_Wind_X, sortPriority=3}
   B=ScalarParameter() {group=Wind, parameterName=WPO_Wind_Y, sortPriority=3}
  )
  n14 = ScalarParameter() {defaultValue=-4, group=Wind, parameterName=WPO_Wind_Z, sortPriority=3}
  n15 = Multiply(
   A=StaticSwitchParameter(True=@n11, False=@n9) {group=Wind, parameterName=UseVertexColorsMask}
   B=Multiply(A=@n7.G, B=AppendVector(A=@n13, B=@n14))
  )
  n16 = StaticSwitchParameter(
   True=SetMaterialAttributes!(
    MaterialAttributes=@n2.MaterialAttributes
    Normal=LinearInterpolate(A=@n2.Normal, B=@n8, Alpha=@n12)
    WorldPositionOffset=Add(A=@n2.WorldPositionOffset, B=@n15)
   ) {attributeSetTypes [#1, #2]}
   False=@n1
  ) {group=Wind, parameterName=WindWPO}
  FunctionOutput(None=ShadingPathSwitch(Default=@n16, Deferred=@n16, Forward=@n16, Mobile=@n1))
  Multiply()
 ]
}
