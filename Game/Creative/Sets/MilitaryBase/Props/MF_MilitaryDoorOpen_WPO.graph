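# Material function for sliding doors that moves geometry via world position offset and masks out parts that overlap.
# A slide vector, mirrored by vertex color green and scaled by object bounds, is driven by the DoorAnimation parameter or a debug sine.
# With sliding enabled, opacity mask derives from local position over scaled bounds; otherwise an optional analytical box around the biased actor position clips pixels.
# Static switches select these overrides on incoming material attributes.
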MaterialFunction /Game/Creative/Sets/MilitaryBase/Props/MF_MilitaryDoorOpen_WPO
graph {
 properties {description=This function is used for any sliding door that needs overlapping geometry to be masked out.}
 nodes [
  n1 = GetMaterialAttributes(None=FunctionInput() {inputType=FunctionInput_MaterialAttributes}) {attributeGetTypes [#1, #2]}
  n2 = Sine(None=Time()) {period=2}
  n3 = VectorParameter() {group=DoorSlide, parameterName=DoorSlideVector, defaultValue=<rgba 2 0 0 1>}
  n4 = LinearInterpolate(A=@n3.RGB, B=Multiply(A=@n3.RGB) {constB=-1}, Alpha=VertexColor().G)
  n5 = Multiply(
   A=StaticSwitchParameter(True=ConstantBiasScale(None=@n2), False=ScalarParameter() {parameterName=DoorAnimation}) {group=DoorSlide, parameterName=DebugDoorSlide}
   B=Multiply(A=Transform(None=@n4) {transformSourceType=TRANSFORMSOURCE_Local}, B=ObjectBounds())
  )
  n6 = TransformPosition(None=WorldPosition().XYZ) {transformSourceType=TRANSFORMPOSSOURCE_World}
  n7 = Multiply(
   A=MaterialFunctionCall() {materialFunction=&ObjectLocalBounds}.Local Bounds Size
   B=VectorParameter() {group=DoorSlide, parameterName=LocalBoundsScale, defaultValue=<rgba 1 1 1 0>}.RGB
  )
  n8 = OneMinus(None=Abs(None=Divide(A=@n6, B=@n7)))
  n9 = WorldPosition()
  n10 = Add(
   A=ActorPositionWS()
   B=Transform(None=VectorParameter() {group=DoorSlide, parameterName=AnalyticalBoxActorPositionBias}.RGB) {transformSourceType=TRANSFORMSOURCE_Local}
  )
  n11 = Abs(None=Transform(None=Subtract(A=@n9.XYZ, B=@n10)) {transformSourceType=TRANSFORMSOURCE_World, transformType=TRANSFORM_Local})
  n12 = VectorParameter() {group=DoorSlide, parameterName=AnalyticalBoxDimensions, defaultValue=<rgba 200 800 700 0>}
  n13 = Floor(None=Saturate(None=Divide(A=@n11, B=@n12.RGB)))
  n14 = Constant() {r=1}
  n15 = StaticSwitchParameter(True=OneMinus(None=DotProduct(A=@n13, B=@n14)), False=@n1.OpacityMask) {group=DoorSlide, parameterName=UseAnalyticalBoxMask}
  FunctionOutput(None=SetMaterialAttributes!(
    MaterialAttributes=@n1.MaterialAttributes
    WorldPositionOffset=StaticSwitchParameter(True=@n5, False=@n1.WorldPositionOffset) {group=DoorSlide, parameterName=UseDoorSlideWPO}
    OpacityMask=StaticSwitchParameter(True=@n8, False=@n15) {group=DoorSlide, parameterName=UseDoorSlideWPO}
   ) {attributeSetTypes [#1, #2]})
 ]
}
