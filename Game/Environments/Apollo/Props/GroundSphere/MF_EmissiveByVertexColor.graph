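# Material function in the GroundSphere prop folder that adds tinted emissive driven by mesh vertex color to incoming material attributes.
# A dot product of vertex color and a channel-mask parameter scales the emissive color parameter.
# A scalar parameter blends that glow toward a base-color-modulated version before it is added to existing emissive.
# A static switch parameter passes the input through untouched when the effect is off.
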MaterialFunction /Game/Environments/Apollo/Props/GroundSphere/MF_EmissiveByVertexColor
graph {nodes [
  n1 = FunctionInput() {inputName=MA_In, inputType=FunctionInput_MaterialAttributes}
  n2 = GetMaterialAttributes(None=@n1) {attributeGetTypes [#1, #2]}
  n3 = VertexColor()
  n4 = VectorParameter() {group=EmissiveByVertexColor, parameterName=VertexColorChannelMask}
  n5 = Multiply(
   A=DotProduct(A=AppendVector(A=@n3, B=@n3), B=AppendVector(A=@n4.RGB, B=@n4.RGB))
   B=VectorParameter() {group=EmissiveByVertexColor, parameterName=EmissiveColor}.RGB
  )
  n6 = Add(
   A=@n2.EmissiveColor
   B=LinearInterpolate(
    A=@n5
    B=Multiply(A=@n2.BaseColor, B=@n5)
    Alpha=ScalarParameter() {group=EmissiveByVertexColor, parameterName=ModulateEmissiveByBaseColorScale}
   )
  )
  FunctionOutput(None=StaticSwitchParameter(
    True=SetMaterialAttributes!(MaterialAttributes=@n2.MaterialAttributes, EmissiveColor=@n6) {attributeSetTypes [#1]}
    False=@n1
   ) {group=EmissiveByVertexColor, parameterName=UseEmissiveByVertexColor})
 ]}
