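# Material function that scales the emissive color of a material by a global intensity value, located with the Apollo GroundSphere prop assets.
# It takes material attributes as input and reads GlobalEmissiveIntensity from the FortniteMaterialParameters collection.
# The red component of that parameter is expanded to a float3 and multiplied into EmissiveColor, and the updated attributes are output.
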MaterialFunction /Game/Environments/Apollo/Props/GroundSphere/MF_GlobalEmissiveIntentisy
graph {nodes [
  n1 = GetMaterialAttributes(None=FunctionInput() {inputType=FunctionInput_MaterialAttributes}) {attributeGetTypes [#1]}
  n2 = CollectionParameter() {
   collection=&FortniteMaterialParameters
   parameterName=GlobalEmissiveIntensity
  }
  n3 = MaterialFunctionCall(Float4=@n2) {materialFunction=&BreakOutFloat4Components}
  n4 = MaterialFunctionCall(X=@n3.R, Y=@n3.R, Z=@n3.R) {materialFunction=&MakeFloat3}
  FunctionOutput(None=SetMaterialAttributes!(MaterialAttributes=@n1.MaterialAttributes, EmissiveColor=Multiply(A=@n1.EmissiveColor, B=@n4.Result)) {attributeSetTypes [#1]})
 ]}
