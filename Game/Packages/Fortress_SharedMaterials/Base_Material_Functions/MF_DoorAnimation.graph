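# Material function that adds parameter-driven door animation, grouped as Chest Animation, to incoming material attributes through World Position Offset, behind a static switch.
# Animation_Progress, or looping time when DebugDoorAnim is on (its default), drives delayed, power-eased rotation and two translations.
# Vertex color G masks the pivot rotation, while R and B mask the translations.
# Rotated normals are corrected on default and deferred paths; forward and mobile keep the original.
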MaterialFunction /Game/Packages/Fortress_SharedMaterials/Base_Material_Functions/MF_DoorAnimation
graph {nodes [
  n1 = GetMaterialAttributes(None=FunctionInput() {inputType=FunctionInput_MaterialAttributes}) {attributeGetTypes [#1, #2]}
  n2 = VertexColor()
  n3 = Normalize(VectorInput=Transform(None=VectorParameter() {group=Chest Animation, parameterName=Door_Rotation_Axis, defaultValue=<rgba 0 1 0 1>}.RGB) {transformSourceType=TRANSFORMSOURCE_Local})
  n4 = Frac(None=Multiply(A=Time(), B=ScalarParameter() {defaultValue=3, group=Chest Animation, parameterName=DoorSpeed}) {constB=2})
  n5 = ScalarParameter() {parameterName=Animation_Progress}
  n6 = Multiply(
   A=Saturate(None=StaticSwitchParameter(True=@n4, False=@n5) {defaultValue=true, parameterName=DebugDoorAnim})
   B=ScalarParameter() {defaultValue=3, parameterName=MaxDelayLength}
  )
  n7 = Subtract(A=@n6, B=ScalarParameter() {parameterName=Delay_Rotation})
  n8 = ScalarParameter() {defaultValue=1, parameterName=Door_Anim_Rotate_Power}
  n9 = Multiply(A=ScalarParameter() {group=Chest Animation, parameterName=Door_Animation_Max_Rotation}, B=Saturate(None=Power(Base=@n7, Exp=@n8)))
  n10 = RotateAboutAxis(
   NormalizedRotationAxis=@n3
   RotationAngle=@n9
   PivotPoint=TransformPosition(None=VectorParameter() {group=Chest Animation, parameterName=Door_Rotation_Pivot, defaultValue=<rgba 0 0 0 1>}.RGB) {transformType=TRANSFORMPOSSOURCE_World}
   Position=WorldPosition().XYZ
  )
  n11 = Subtract(A=@n6, B=ScalarParameter() {parameterName=Delay_Door_Movement_1})
  n12 = ScalarParameter() {defaultValue=1, parameterName=Door_Movement_1_Power}
  n13 = Multiply(
   A=Saturate(None=Power(Base=@n11, Exp=@n12))
   B=Transform(None=VectorParameter() {group=Chest Animation, parameterName=Door_Move_to_Location_1, defaultValue=<rgba 0 0 0 1>}.RGB) {transformSourceType=TRANSFORMSOURCE_Local}
  )
  n14 = VertexColor()
  n15 = Subtract(A=@n6, B=ScalarParameter() {parameterName=Delay_Door_Movement_2})
  n16 = ScalarParameter() {defaultValue=1, parameterName=Door_Movement_2_Power}
  n17 = Multiply(
   A=Saturate(None=Power(Base=@n15, Exp=@n16))
   B=Transform(None=VectorParameter() {group=Chest Animation, parameterName=Door_Move_to_Location_2, defaultValue=<rgba 0 0 0 1>}.RGB) {transformSourceType=TRANSFORMSOURCE_Local}
  )
  n18 = Add(A=Add(A=@n1.WorldPositionOffset, B=Multiply(A=@n2.G, B=@n10)), B=Add(A=Multiply(A=@n13, B=@n14.R), B=Multiply(A=@n14.B, B=@n17)))
  n19 = MaterialFunctionCall(Rotation Axis=@n3, Rotation Angle=@n9, World Space Vertex Normals=Transform(None=@n1.Normal)) {materialFunction=&FixRotateAboutAxisNormals}
  n20 = LinearInterpolate(A=@n1.Normal, B=@n19.New Tangent Space Vertex Normal, Alpha=@n2.G)
  n21 = ShadingPathSwitch(Default=@n20, Deferred=@n20, Forward=@n1.Normal, Mobile=@n1.Normal)
  FunctionOutput(None=StaticSwitchParameter(
    True=SetMaterialAttributes!(MaterialAttributes=@n1.MaterialAttributes, WorldPositionOffset=@n18, Normal=@n21) {attributeSetTypes [#1, #2]}
    False=@n1.MaterialAttributes
   ) {parameterName=Use Door Animation})
  MaterialFunctionCall(BaseNormal=VertexInterpolator(VS=@n19.New Tangent Space Vertex Normal).PS) {materialFunction=&BlendAngleCorrectedNormals}
  PixelNormalWS()
 ]}
