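# Material function for living billboards that swaps between two message textures by camera distance, showing the second up close and the first far away.
# Distance drives a HeightLerp over scrolling digital noise, which also indexes a glow gradient shown during the transition.
# A blurred difference of both textures confines that effect to changed areas.
# Static switches gate the base color override and time-of-day tinted emissive with overlay.
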MaterialFunction /Game/Packages/Fortress_SharedMaterials/Base_Material_Functions/MF_LivingBillboardMessage
graph {nodes [
  n1 = FunctionInput() {inputType=FunctionInput_MaterialAttributes}
  n2 = ScalarParameter() {group=LivingBillboard, parameterName=BillboardOffset_U}
  n3 = ScalarParameter() {group=LivingBillboard, parameterName=BillboardOffset_V}
  n4 = Multiply(
   A=Add(A=TextureCoordinate(), B=AppendVector(A=@n2, B=@n3))
   B=AppendVector(
    A=ScalarParameter() {defaultValue=1, group=LivingBillboard, parameterName=BillboardScale_U}
    B=ScalarParameter() {defaultValue=1, group=LivingBillboard, parameterName=BillboardScale_V}
   )
  )
  n5 = Distance(A=ObjectPositionWS(), B=CameraPositionWS())
  n6 = ScalarParameter() {defaultValue=500, group=LivingBillboard, parameterName=BillboardDistanceOffset}
  n7 = Saturate(None=Divide(A=Subtract(A=@n5, B=@n6), B=ScalarParameter() {defaultValue=1000, group=LivingBillboard, parameterName=FadeLength}))
  n8 = TextureSample(UVs=Multiply(A=TextureCoordinate(), B=ScalarParameter() {defaultValue=2.5915, group=LivingBillboard, parameterName=BillBoardFXScale})) {samplerType=SAMPLERTYPE_LinearColor, texture=/Game/Global/Textures/Default/Noise/T_Atlas_DigitalNoise_01}
  n9 = ScalarParameter() {defaultValue=72.2404, group=LivingBillboard, parameterName=BillboardDistance2}
  n10 = Frac(None=Add(A=Desaturation(None=@n8.RGB), B=Divide(A=@n5, B=@n9)))
  n11 = MaterialFunctionCall(Transition Phase=OneMinus(None=@n7), Height Texture=@n10) {materialFunction=&HeightLerp}
  n12 = Saturate(None=Power(Base=@n7, Exp=ScalarParameter() {defaultValue=1, group=LivingBillboard, parameterName=BillBoardFXFadeout}))
  n13 = OneMinus(None=@n12)
  n14 = TextureCoordinate()
  n15 = ScalarParameter() {defaultValue=2, group=LivingBillboard, parameterName=BillboardMaskBias}
  n16 = Subtract(
   A=TextureSampleParameter2D!(UVs=@n4, Bias=@n15, Apply View MipBias) {group=LivingBillboard, mipValueMode=TMVM_MipBias, parameterName=BillboardTexture_1}.RGB
   B=TextureSampleParameter2D(UVs=@n4, Bias=@n15) {group=LivingBillboard, mipValueMode=TMVM_MipBias, parameterName=BillboardTexture_2}.RGB
  )
  n17 = Saturate(None=Divide(
    A=MaterialFunctionCall(Vector 3=@n16) {materialFunction=/Engine/Functions/Engine_MaterialFunctions02/Utility/VectorLength}.V3 Length
    B=ScalarParameter() {defaultValue=0.2889, group=LivingBillboard, parameterName=MaskThreshold}
   ))
  n18 = LinearInterpolate(
   A=LinearInterpolate(
    A=TextureSampleParameter2D!(UVs=@n4, Apply View MipBias) {group=LivingBillboard, parameterName=BillboardTexture_1}.RGB
    B=TextureSampleParameter2D(UVs=@n4) {group=LivingBillboard, parameterName=BillboardTexture_2}.RGB
    Alpha=Multiply(A=@n11.Alpha, B=@n13)
   )
   B=TextureSample!(UVs=@n10, Tex, "DDX(UVs)"=DDX(Value=@n14), "DDY(UVs)"=DDY(Value=@n14), Apply View MipBias) {mipValueMode=TMVM_Derivative, texture=/Game/Characters/Player/Male/Large/Bodies/M_LRG_Invader/Textures/T_InvaderGlow_Glass_Gradient}.RGB
   Alpha=Multiply(A=Multiply(A=@n17, B=@n11.Alpha), B=@n12)
  )
  n19 = Multiply(A=@n18, B=TextureSampleParameter2D(UVs=@n4) {group=LivingBillboard, parameterName=BillboardEmissiveOverlay}.RGB)
  n20 = VectorParameter() {group=LivingBillboard, parameterName=LivingBillboardEmissive_Color_Morning, defaultValue=<rgba 1 1 1 1>}
  n21 = VectorParameter() {group=LivingBillboard, parameterName=LivingBillboardEmissive_Color_Evening, defaultValue=<rgba 1 1 1 1>}
  n22 = MaterialFunctionCall(
   Color_Morning=@n20.RGB
   Color_Day=VectorParameter() {group=LivingBillboard, parameterName=LivingBillboardEmissive_Color_Day, defaultValue=<rgba 1 1 1 1>}.RGB
   Color_Evening=@n21.RGB
   Color_Night=VectorParameter() {group=LivingBillboard, parameterName=LivingBillboardEmissive_Color_Night, defaultValue=<rgba 1 1 1 1>}.RGB
  ) {materialFunction=&Color_Time_of_Day}
  n23 = StaticSwitchParameter(
   True=Multiply(A=Multiply(A=@n19, B=@n22.Result), B=ScalarParameter() {defaultValue=1, group=LivingBillboard, parameterName=BillBoardEmissiveBrightness})
   False=Constant()
  ) {group=LivingBillboard, parameterName=Use_LivingBillboardMessage_Emissive}
  FunctionOutput(None=StaticSwitchParameter(
    True=SetMaterialAttributes!(MaterialAttributes=@n1, BaseColor=@n18, EmissiveColor=@n23) {attributeSetTypes [#1, #2]}
    False=@n1
   ) {group=LivingBillboard, parameterName=Use_LivingBillboardMessage})
 ]}
