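# Material function that animates a looping squash and stretch on a mesh by adding to World Position Offset in material attributes, behind a static switch.
# Speed-scaled time, randomized by actor position and optionally offset by vertex color, drives a sine horizontal sway and a custom fast-attack, slow-decay pulse.
# The pulse stretches Z and inversely squashes XY, weighted by local-bounds height gradients.
# Offsets scale with object radius, with optional vertex-color masking.
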MaterialFunction /Game/Packages/Fortress_SharedMaterials/Base_Material_Functions/MF_SquashStretch
graph {
 properties {description=This function activates a squash and stretch effect on a mesh.}
 nodes [
  n1 = GetMaterialAttributes(None=FunctionInput() {inputType=FunctionInput_MaterialAttributes}) {attributeGetTypes [#1]}
  n2 = VertexColor()
  n3 = MaterialFunctionCall(Float3=ActorPositionWS()) {materialFunction=&BreakOutFloat3Components}
  n4 = Frac(None=Add(A=Add(A=@n3.R, B=@n3.R), B=@n3.B))
  n5 = ScalarParameter() {defaultValue=1, group=SquashStretch, parameterName=SquashStretchActorPositionRandomness}
  n6 = Time()
  n7 = VertexColor()
  n8 = Add(
   A=@n6
   B=StaticComponentMaskParameter(None=AppendVector(A=@n7, B=@n7)) {defaultA=true, group=SquashStretch, parameterName=SquashStretch_VertexTimeOffsetChannel}
  )
  n9 = Multiply(
   A=Add(A=Multiply(A=@n4, B=@n5), B=StaticSwitchParameter(True=@n8, False=@n6) {group=SquashStretch, parameterName=SquashStretch_VertexTimeOffset})
   B=ScalarParameter() {defaultValue=1, group=SquashStretch, parameterName=SquashStretchSpeed}
  )
  n10 = Sine(None=Multiply(A=@n9) {constB=0.5})
  n11 = Multiply(
   A=AppendVector(
    A=ScalarParameter() {defaultValue=6, group=SquashStretch, parameterName=HorizontalStretchAmplitudeX}
    B=ScalarParameter() {group=SquashStretch, parameterName=HorizontalStretchAmplitudeY}
   )
   B=AppendVector(A=@n10, B=@n10)
  )
  n12 = MaterialFunctionCall() {materialFunction=&ObjectLocalBounds}
  n13 = Divide(
   A=Subtract(
    A=MaterialFunctionCall() {materialFunction=&LocalPosition}."Local Position (Excluding Offsets)"
    B=@n12.Local Bounds Minimum
   )
   B=@n12.Local Bounds Size
  )
  MaterialExpressionComponentMask_0 = ComponentMask(None=@n13) {b=true}
  MaterialExpressionAppendVector_4 = AppendVector(
   A=ScalarParameter() {defaultValue=0.3, group=SquashStretch, parameterName=HorizontalStretchGradientStart}
   B=ScalarParameter() {group=SquashStretch, parameterName=VerticalStretchGradientStart}
  )
  MaterialExpressionAppendVector_5 = AppendVector(
   A=ScalarParameter() {defaultValue=1, group=SquashStretch, parameterName=HorizontalStretchGradientEnd}
   B=ScalarParameter() {defaultValue=0.7, group=SquashStretch, parameterName=VerticalStretchGradientEnd}
  )
  n14 = Custom!(varA=@MaterialExpressionComponentMask_0, varMin=@MaterialExpressionAppendVector_4, varMax=@MaterialExpressionAppendVector_5) {
   code="return smoothstep(varMin,varMax,varA);"
   description=SmoothStep float2
   outputType=CMOT_Float2
   inputs [
    {
     inputName=varA
     input {expression=&self:MaterialExpressionComponentMask_0}
    }
    {
     inputName=varMin
     input {expression=&self:MaterialExpressionAppendVector_4}
    }
    {
     inputName=varMax
     input {expression=&self:MaterialExpressionAppendVector_5}
    }
   ]
  }
  n15 = ComponentMask(None=@n14) {r=true}
  n16 = ComponentMask(None=@n13) {b=true}
  n17 = ScalarParameter() {group=SquashStretch, parameterName=SquashHeight, sliderMax=1}
  n18 = Power(
   Base=OneMinus(None=Distance(A=@n16, B=@n17))
   Exp=ScalarParameter() {defaultValue=2, group=SquashStretch, parameterName=SquashVerticalAttenuation}
  )
  MaterialExpressionPower_2 = Power(Base=Frac(None=@n9), Exp=ScalarParameter() {defaultValue=1.25, group=SquashStretch, parameterName=SquashStretchDamping})
  n19 = Custom!(InputFunction=@MaterialExpressionPower_2) {
   code="float result = (2 * InputFunction) -2;\r\nresult  = result  * result  * result ;\r\nreturn -1.181 * InputFunction * result;"
   description=FastAttackSlowDecay
   outputType=CMOT_Float1
   inputs [{
     inputName=InputFunction
     input {expression=&self:MaterialExpressionPower_2}
    }]
  }
  n20 = Multiply(A=ComponentMask(None=@n13) {g=true, r=true}) {constB=2}
  n21 = Multiply(
   A=Multiply(A=OneMinus(None=@n19), B=Subtract(A=@n20))
   B=ScalarParameter() {defaultValue=2, group=SquashStretch, parameterName=SquashAmplitude}
  )
  n22 = ScalarParameter() {defaultValue=3, group=SquashStretch, parameterName=VerticalStretchAmplitude}
  n23 = AppendVector(
   A=Add(A=Multiply(A=@n11, B=@n15), B=Multiply(A=@n18, B=@n21))
   B=Multiply(A=ComponentMask(None=@n14) {g=true}, B=Multiply(A=@n19, B=@n22)) {constB=1.0095}
  )
  n24 = Multiply(
   A=StaticComponentMaskParameter(None=AppendVector(A=@n2, B=@n2)) {defaultA=true, group=SquashStretch, parameterName=SquashStretch_MaskChannel}
   B=@n23
  )
  n25 = ObjectRadius()
  n26 = ScalarParameter() {defaultValue=1, group=SquashStretch, parameterName=SquashStretchGlobalScale}
  n27 = Add(
   A=@n1.WorldPositionOffset
   B=Multiply(A=StaticSwitchParameter(True=@n24, False=@n23) {group=SquashStretch, parameterName=SquashStretch_VertexMask}, B=Multiply(A=@n25, B=@n26))
  )
  FunctionOutput(None=StaticSwitchParameter(
    True=SetMaterialAttributes!(MaterialAttributes=@n1.MaterialAttributes, WorldPositionOffset=@n27) {attributeSetTypes [#1]}
    False=@n1.MaterialAttributes
   ) {parameterName=Use SquashSretch Animation})
 ]
}
