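# MF_GameWideShared is a Fortnite material function that bundles game-wide shared effects into one material-attributes pass-through.
# The input runs through a cover function, then a static switch optionally keeps season 19 tree snow.
# Destructible damage follows, with a switch for additive WPO.
# A long serial chain then layers wind, door, fan, chest and other WPO animations, plus glow, emissive, fallback, Creative tool and event effects, ending in deploy wobble.
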MaterialFunction /Game/Athena/WinterArt/Materials/MF_GameWideShared
graph {
 properties {description=This function harbors various other functions that are utilized throughout all of Fortnite.}
 nodes [
  n1 = MaterialFunctionCall!(SFX V3, In=FunctionInput() {inputType=FunctionInput_MaterialAttributes}) {materialFunction=/Game/Global/Base4Layers_2022/Material/MF_Cover}
  n2 = MaterialFunctionCall!(MA In=@n1.Result, SnowMaskOverride, Fresnel_SnowBlendExponent_Override, Fresnel_SnowBlendBrightness_Override) {materialFunction=/Game/Packages/Fortnite_UniqueMaterials_1/MaterialFunctions/Trees/MF_Tree_Snow_s19}
  n3 = StaticBool() {value=true}
  n4 = StaticBool()
  n5 = MaterialFunctionCall!(In=MaterialFunctionCall(
    Material_In=StaticSwitchParameter(True=@n2.Result, False=@n1.Result) {group=s19_Snow, parameterName=UseS19TreeSnow}
    UseAdditiveWPO=StaticSwitchParameter(True=@n3, False=@n4) {parameterName=Use Additive WPO}
   ) {materialFunction=&MF_FN_DestructibleDamage}.Material_Out) {materialFunction=&/MF_WPOAnim}
  n6 = MaterialFunctionCall!(In=MaterialFunctionCall!(In=MaterialFunctionCall!(In=@n5.Result) {materialFunction=&/MF_WPO_WindNew}.Result) {materialFunction=&/MF_WindWPO}.Result) {materialFunction=/Game/Creative/Sets/MilitaryBase/Props/MF_MilitaryDoorOpen_WPO}
  n7 = MaterialFunctionCall!(
   In=MaterialFunctionCall!(MA=MaterialFunctionCall!(Material Attributes=@n6.Result) {materialFunction=/Game/Creative/Sets/MilitaryBase/Props/MF_FanSpin}.Result) {materialFunction=/Game/Environments/Sets/PrincessCastle/Props/Materials/MF_GlowFlicker}.Result
   UseSphericalMappingOverride
  ) {materialFunction=/Game/Packages/Fortress_SharedMaterials/Base_Material_Functions/MF_CurieFallbackEffect}
  n8 = MaterialFunctionCall!(In=MaterialFunctionCall!(MA In=MaterialFunctionCall(MA In=@n7.Result) {materialFunction=&MF_CreativeMoveTool}.MA Out) {materialFunction=/Game/Weapons/Prototype/ObjectMover/Materials/MF_Creative_EmissiveRamp}.MA Out) {materialFunction=&/MF_Events}
  n9 = MaterialFunctionCall!(In=MaterialFunctionCall!(In=MaterialFunctionCall!(In=@n8.Result) {materialFunction=/Game/Packages/Fortress_SharedMaterials/Base_Material_Functions/MF_ChestAnimation}.Result) {materialFunction=/Game/Packages/Fortress_SharedMaterials/Base_Material_Functions/MF_LivingBillboardMessage}.Result) {materialFunction=/Game/Packages/Fortress_SharedMaterials/Base_Material_Functions/MF_DoorAnimation}
  n10 = MaterialFunctionCall!(MA_In=MaterialFunctionCall!(In=MaterialFunctionCall!(In=@n9.Result) {materialFunction=/Game/Packages/Fortress_SharedMaterials/Base_Material_Functions/MF_SquashStretch}.Result) {materialFunction=&/MF_ActivationSwitchWPO}.Result) {materialFunction=/Game/Environments/Apollo/Props/GroundSphere/MF_EmissiveByVertexColor}
  FunctionOutput(None=MaterialFunctionCall!(In=MaterialFunctionCall!(In=@n10.Result) {materialFunction=/Game/Environments/Apollo/Props/GroundSphere/MF_GlobalEmissiveIntentisy}.Result) {materialFunction=/Game/Global/MaterialFunctions/MF_DeployWobble_Simple}.Result)
 ]
}
