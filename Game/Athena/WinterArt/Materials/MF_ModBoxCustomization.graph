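# Material function that customizes the vehicle ModBox crate's base color with a paint color and logo, behind a static switch.
# Masks split the painted crate into top and sides, which are recolored separately from the original diffuse, summed, then blended into the input material attributes.
# The side logo samples a center-scaled UV1, is clipped to the UV range, takes an optional tint and the diffuse's painted shadows, and leaves wood cracks darkened.
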MaterialFunction /Game/Athena/WinterArt/Materials/MF_ModBoxCustomization
graph {nodes [
  n1 = NamedRerouteUsage()
  n2 = SetMaterialAttributes!(MaterialAttributes=@n1, BaseColor=Add(A=NamedRerouteUsage(), B=NamedRerouteUsage())) {attributeSetTypes [#1]}
  n3 = TextureSample() {
   desc="Paint Mask (alpha)"
   samplerType=SAMPLERTYPE_Masks
   texture=/VehicleMods/Gameplay/Actors/ModBox/Textures/T_ModBox_PaintMask_TopAndSides_Monochrome
  }
  FunctionOutput(None=StaticSwitchParameter(True=BlendMaterialAttributes(A=@n1, B=@n2, Alpha=@n3.R), False=@n1) {parameterName=UseModBoxCustomization})
  n4 = GetMaterialAttributes(None=FunctionInput() {inputType=FunctionInput_MaterialAttributes}) {attributeGetTypes [#1]}
  NamedRerouteDeclaration(None=@n4.MaterialAttributes) {name=Input_Material, nodeColor=<rgba 0.6172 1 0 1>}
  NamedRerouteDeclaration(None=@n4.BaseColor) {name=Input_BaseColor, nodeColor=<rgba 0.1016 0 1 1>}
  n5 = TextureSample() {
   desc="Paint Mask (sides only)"
   samplerType=SAMPLERTYPE_Masks
   texture=/VehicleMods/Gameplay/Actors/ModBox/Textures/T_ModBox_PaintMask_SidesOnly
  }
  n6 = Multiply(A=Divide(A=OneMinus(None=@n5.R) {desc=Get the cracks}) {constB=5, desc=Darken it}, B=NamedRerouteUsage())
  n7 = NamedRerouteUsage()
  n8 = NamedRerouteUsage()
  n9 = Saturate(None=Add(A=Multiply(A=@n6, B=@n7), B=Multiply(A=@n5.R, B=@n8)))
  n10 = MaterialFunctionCall(
   UVs=TextureCoordinate() {coordinateIndex=1}
   Texture Scale=ScalarParameter() {defaultValue=1, group=ModBox_Customization, parameterName=ModBox_LogoScale}
  ) {materialFunction=&ScaleUVsByCenter}
  n11 = TextureSampleParameter2D(UVs=Multiply(A=@n10.UVs, B=@n10.UVs)) {
   group=ModBox_Customization
   parameterName=LogoTexture
   sortPriority=30
   texture=/Game/Athena/UI/InGame/Creative/PropertyEditor/Icon_Picker/MilitaryTextures/T_128x128/T_UI_IconLibrary_assaultrifle_128
  }
  n12 = StaticSwitchParameter(True=MaterialFunctionCall!(Value=@n11.A, Threshold, Smoothing=Constant() {r=0.1}).EitherSide, False=@n11.A) {
   desc="Always False for now, may add an option for this later."
   group=ModBox_Customization
   parameterName=ModBox_Switch_ApplyAntiAliasingToLogo
   sortPriority=31
  }
  n13 = Multiply(A=@n12, B=@n10.0-1 mask) {desc="Apply UV-scale-based mask so colors on\r\nthe texture edges don't extend infinitely."}
  n14 = VectorParameter() {group=ModBox_Customization, parameterName=ModBox_Logo_Color, sortPriority=33, defaultValue=<rgba 0.6353 0.0392 0.1294 0>}
  n15 = Desaturation(None=NamedRerouteUsage(), Fraction=Constant() {r=1})
  n16 = Constant() {desc=0.04, r=0.04}
  n17 = Constant()
  n18 = Constant() {r=1}
  n19 = Constant() {desc=0.16, r=0.16}
  n20 = Multiply(
   A=Multiply(A=@n11.RGB, B=@n13) {desc="Applying anti-aliasing to the edges of\r\nthe icon if necessary."}
   B=Saturate(None=MaterialFunctionCall(Input=@n15, Input Low=@n16, Input High=@n19, Target Low=@n17, Target High=@n18) {materialFunction=&RemapValueRange}.Result)
  ) {desc=Apply the diffuse's painted shadows to the logo}
  n21 = Desaturation(None=@n20, Fraction=Constant() {r=1})
  n22 = LinearInterpolate(
   A=@n9
   B=If(A=@n14.A, "A > B"=Multiply(A=@n14.RGB, B=@n21), "A == B"=@n20, "A < B"=@n20)
   Alpha=Multiply(A=@n5.R, B=@n13) {desc=Logo texture's alpha minus cracks/holes in the wood}
  )
  NamedRerouteDeclaration(None=Saturate(None=StaticSwitchParameter(True=@n22, False=@n9) {defaultValue=true, group=ModBox_Customization, parameterName=ModBox_Switch_ShowLogo, sortPriority=29})) {name=CrateSides, nodeColor=<rgba 0 0.5294 1 1>}
  n23 = MaterialFunctionCall(
   Input=NamedRerouteUsage()
   Input Low=Constant() {desc=0}
   Input High=Constant() {desc=1.3, r=1.3}
   Target Low=Constant()
   Target High=Constant() {desc=0.7, r=0.7}
  ) {materialFunction=&RemapValueRange}
  n24 = VectorParameter() {group=ModBox_Customization, parameterName=ModBox_Primary_Color, sortPriority=35, defaultValue=<rgba 0.0933 0.1465 0.2083 1>}
  n25 = Saturate(None=Multiply(A=Saturate(None=@n23.Result), B=@n24.RGB))
  n26 = NamedRerouteUsage()
  NamedRerouteDeclaration(None=Multiply(
    A=StaticSwitchParameter(True=@n25, False=@n26) {defaultValue=true, group=ModBox_Customization, parameterName=ModBox_Switch_ShowPrimaryColor, sortPriority=34}
    B=NamedRerouteUsage()
   )) {name=CrateTop, nodeColor=<rgba 1 0.2621 0.8578 1>}
  n27 = NamedRerouteUsage()
  n28 = Divide(A=@n27, B=Constant3Vector() {constant=<rgba 0.2623 0.1413 0.0307 1>}) {desc="Prep for custom color,\r\naccounting for the diffuse"}
  n29 = Saturate(None=NamedRerouteDeclaration(None=ComponentMask(None=@n28) {r=true}) {name=CustomColorPrep, desc=Intentionally Saturating after this, nodeColor=<rgba 0 1 0.1094 1>})
  n30 = VectorParameter() {group=ModBox_Customization, parameterName=ModBox_Primary_Color, sortPriority=35, defaultValue=<rgba 0.0933 0.1465 0.2083 1>}
  NamedRerouteDeclaration(None=StaticSwitchParameter(True=Multiply(A=@n29, B=@n30.RGB), False=@n27) {defaultValue=true, group=ModBox_Customization, parameterName=ModBox_Switch_ShowPrimaryColor, sortPriority=34}) {name=CratePaint_Base, nodeColor=<rgba 1 0.1406 0 1>}
  n31 = TextureSample() {desc="Paint mask (sides only)", texture=/VehicleMods/Gameplay/Actors/ModBox/Textures/T_ModBox_PaintMask_TopOnly}
  NamedRerouteDeclaration(None=Saturate(None=Multiply(A=@n31.R) {constB=800}) {desc=Pure alpha for crate top}) {name=CrateTop_Alpha, nodeColor=<rgba 1 0 0.2109 1>}
  n32 = TextureSample() {
   desc="Paint Mask (alpha)"
   samplerType=SAMPLERTYPE_Masks
   texture=&T_ModBox_PaintMask_TopAndSides_Monochrome
  }
  NamedRerouteDeclaration(None=Subtract(A=@n32.R, B=NamedRerouteUsage())) {name=CrateSides_Alpha, nodeColor=<rgba 0 0.102 0.6196 1>}
 ]}
