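# Material function producing masked, tinted neon sign emissive, with optional per-object random color.
# Default path multiplies a parallax, optionally panning world-space pattern, projected on two normal-blended planes, by sign texture, with per-object contrast flicker from a scrolling lightning texture.
# The isFakeInterior switch instead samples a bump-offset interior texture with optional near-camera fade, zeroed at Low and Medium quality.
# Blur and fake-reflection branches are not connected to the output.
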MaterialFunction /Game/Athena/WinterArt/Materials/MF_NeonSignEmissive
graph {nodes [
  n1 = VectorParameter() {group=NeonEmissive, parameterName=NeonSignEmissiveColor, defaultValue=<rgba 5 1.9691 0 1>}
  n2 = ObjectPositionWS()
  n3 = Add(A=Add(A=ComponentMask(None=@n2) {g=true}, B=ComponentMask(None=@n2) {r=true}), B=ComponentMask(None=@n2) {b=true})
  n4 = Frac(None=Divide(A=@n3, B=ScalarParameter() {defaultValue=104.45, parameterName=RandomValue1}))
  n5 = LinearInterpolate(A=@n1.RGB, B=VectorParameter() {group=NeonEmissive, parameterName=NeonSignEmissiveColor2, defaultValue=<rgba 5 1.9691 0 1>}.RGB, Alpha=@n4)
  n6 = ScalarParameter() {defaultValue=59.5534, parameterName=Offset}
  n7 = Add(A=TextureCoordinate(), B=AppendVector(A=Multiply(A=@n4, B=@n6), B=Constant()))
  n8 = Constant() {r=1}
  n9 = CameraVectorWS()
  n10 = Divide(A=ScalarParameter() {defaultValue=-49.7268, group=NeonEmissive, parameterName=BumpOffsetNeonSign2}, B=DotProduct(A=VertexNormalWS(), B=@n9))
  n11 = ScalarParameter() {defaultValue=292.059, group=NeonEmissive, parameterName=TilingNeonOffsetX}
  n12 = ScalarParameter() {defaultValue=148.4555, group=NeonEmissive, parameterName=TilingNeonOffsetY}
  n13 = Divide(
   A=Add(A=BumpOffset(Coordinate=@n7, Height=@n8, HeightRatioInput=@n10), B=AppendVector(A=@n11, B=@n12))
   B=AppendVector(
    A=ScalarParameter() {defaultValue=292.059, group=NeonEmissive, parameterName=TilingNeonBackGroundX}
    B=ScalarParameter() {defaultValue=148.4555, group=NeonEmissive, parameterName=TilingNeonBackGroundY}
   )
  )
  n14 = TextureSampleParameter2D(UVs=@n13) {
   group=NeonEmissive
   parameterName=Neon_Emissive_Inside
   samplerSource=SSM_Wrap_WorldGroupSettings
   samplerType=SAMPLERTYPE_Masks
   texture=/Game/Environments/Sets/Neo/Architecture/Textures/T_NeonSign_M
  }
  n15 = ChannelMaskParameter(None=@n14.RGB) {group=NeonEmissive, parameterName=Neon_EmissiveInside_Channel}
  n16 = ScalarParameter() {defaultValue=0.1, group=NeonEmissive, parameterName=EmissiveNeonMinBirghtness}
  n17 = MaterialFunctionCall(Distance=ScalarParameter() {defaultValue=8000, group=NeonEmissive, parameterName=DistanceBlur}) {materialFunction=&Near_Camera_Fade}
  n18 = ScalarParameter() {defaultValue=3, group=NeonEmissive, parameterName=EmissiveNeonSignBrightness}
  n19 = Divide(A=@n3, B=ScalarParameter() {defaultValue=504.3145, parameterName=RandomValue2})
  n20 = StaticSwitchParameter(
   True=LinearInterpolate(A=ScalarParameter() {defaultValue=3, group=NeonEmissive, parameterName=EmissiveNeonSignBrightness2}, B=@n18, Alpha=Frac(None=@n19))
   False=@n18
  ) {parameterName=UseRandomEmissiveColor}
  n21 = MaterialProxyReplace(Realtime=Multiply(A=Add(A=@n17.Result, B=@n16), B=@n20), MaterialProxy=@n20)
  n22 = ScalarParameter() {defaultValue=3.2869, group=NeonEmissive, parameterName=EmissiveNeonSignContrast}
  n23 = ScalarParameter() {defaultValue=2.567, group=NeonEmissive, parameterName=EmissiveNeonSignContrast2}
  n24 = LinearInterpolate(A=@n22, B=@n23, Alpha=@n17.Result)
  n25 = Power(
   Base=Multiply(
    A=StaticSwitchParameter(True=@n14.RGB, False=@n15.R) {group=NeonEmissive, parameterName=Neon_Emissive_RGBorChannel}
    B=StaticSwitchParameter(True=@n21, False=@n20) {parameterName=UseNearFadeEmissiveColor}
   )
   Exp=StaticSwitchParameter(True=MaterialProxyReplace(Realtime=@n24, MaterialProxy=@n23), False=@n22) {parameterName=UseNearFadeEmissiveColor}
  )
  n26 = Constant()
  n27 = QualitySwitch(Default=@n25, Low=@n26, High=@n25, Medium=@n26)
  n28 = Constant() {r=0.5}
  n29 = Add(A=Multiply(A=@n9, B=@n10), B=WorldPosition().XYZ)
  n30 = AppendVector(
   A=ScalarParameter() {defaultValue=292.059, group=NeonEmissive, parameterName=TilingNeonBackGroundX}
   B=ScalarParameter() {defaultValue=148.4555, group=NeonEmissive, parameterName=TilingNeonBackGroundY}
  )
  n31 = Divide(A=ComponentMask(None=@n29) {b=true, r=true}, B=@n30)
  n32 = ScalarParameter() {group=NeonEmissive, parameterName=NeonPanX}
  n33 = ScalarParameter() {defaultValue=1, group=NeonEmissive, parameterName=NeonPanY}
  n34 = Frac(None=Multiply(A=Time(), B=AppendVector(A=@n32, B=@n33)))
  n35 = TextureSampleParameter2D(UVs=StaticSwitchParameter(True=Add(A=@n31, B=@n34), False=@n31) {group=NeonEmissive, parameterName=NeonPanning}) {
   group=NeonEmissive
   parameterName=Neon_Emissive_Inside
   samplerSource=SSM_Wrap_WorldGroupSettings
   samplerType=SAMPLERTYPE_Masks
   texture=&T_NeonSign_M
  }
  n36 = ChannelMaskParameter(None=@n35.RGBA) {group=NeonEmissive, parameterName=Neon_EmissiveInside_Channel}
  n37 = Divide(A=ComponentMask(None=@n29) {b=true, g=true}, B=@n30)
  n38 = TextureSampleParameter2D(UVs=StaticSwitchParameter(True=Add(A=@n37, B=@n34), False=@n37) {group=NeonEmissive, parameterName=NeonPanning}) {
   group=NeonEmissive
   parameterName=Neon_Emissive_Inside
   samplerSource=SSM_Wrap_WorldGroupSettings
   samplerType=SAMPLERTYPE_Masks
   texture=&T_NeonSign_M
  }
  n39 = ChannelMaskParameter(None=@n38.RGBA) {group=NeonEmissive, parameterName=Neon_EmissiveInside_Channel}
  n40 = Abs(None=ComponentMask(None=VertexNormalWS()) {r=true})
  n41 = ScalarParameter() {group=NeonEmissive, parameterName=BlendContrastNeon}
  n42 = MaterialProxyReplace(
   Realtime=LinearInterpolate(
    A=StaticSwitchParameter(True=@n35.RGB, False=@n36.R) {group=NeonEmissive, parameterName=Neon_Emissive_RGBorChannel}
    B=StaticSwitchParameter(True=@n38.RGB, False=@n39.R) {group=NeonEmissive, parameterName=Neon_Emissive_RGBorChannel}
    Alpha=MaterialFunctionCall(In=@n40, Contrast=@n41) {materialFunction=&CheapContrast}.Result
   )
   MaterialProxy=Constant() {r=1}
  )
  n43 = ShadingPathSwitch(Default=@n42, Deferred=@n42, Forward=@n28, Mobile=@n28)
  n44 = FunctionInput() {inputName=Base Diffuse, inputType=FunctionInput_Texture2D}
  n45 = TextureSampleParameter2D() {
   group=NeonEmissive
   parameterName=EmissiveNeonTesture
   samplerSource=SSM_Wrap_WorldGroupSettings
   texture=/Game/Environments/Sets/Neo/Architecture/Textures/T_Neo_StoreSign_D
  }
  n46 = StaticSwitchParameter(True=TextureSample(Tex=@n44).RGB, False=@n45.RGB) {group=Emissive, parameterName=UseBaseColorForEmissive}
  n47 = Constant() {r=1}
  n48 = TextureCoordinate()
  n49 = MaterialProxyReplace(
   Realtime=BumpOffset(
    Coordinate=@n48
    Height=Constant() {r=1}
    HeightRatioInput=ScalarParameter() {defaultValue=-0.09, group=NeonEmissive, parameterName=BumpOffsetNeonSign}
   )
   MaterialProxy=@n48
  )
  n50 = TextureSample(UVs=@n49, Tex=@n44)
  n51 = TextureSampleParameter2D(UVs=@n49) {
   group=NeonEmissive
   parameterName=EmissiveNeonTesture
   samplerSource=SSM_Wrap_WorldGroupSettings
   texture=&T_Neo_StoreSign_D
  }
  n52 = Add(
   A=@n46
   B=Multiply(
    A=Desaturation(None=@n46, Fraction=@n47)
    B=StaticSwitchParameter(True=@n50.RGB, False=@n51.RGB) {group=Emissive, parameterName=UseBaseColorForEmissive}
   )
  )
  n53 = Desaturation(
   None=ShadingPathSwitch(Default=@n52, Deferred=@n52, Forward=@n46, Mobile=@n46)
   Fraction=ScalarParameter() {defaultValue=1, group=NeonEmissive, parameterName=EmissiveNeonDesat}
  )
  n54 = ScalarParameter() {defaultValue=3.2869, group=NeonEmissive, parameterName=EmissiveNeonSignContrast}
  n55 = ScalarParameter() {defaultValue=2.567, group=NeonEmissive, parameterName=EmissiveNeonSignContrast2}
  n56 = Time()
  n57 = ObjectPositionWS()
  n58 = Divide(A=ComponentMask(None=@n57) {r=true}, B=ComponentMask(None=@n57) {g=true})
  n59 = ComponentMask(None=@n57) {b=true}
  n60 = Multiply(A=Frac(None=Divide(A=@n58, B=@n59)), B=ScalarParameter() {defaultValue=23425.4395, group=NeonEmissive, parameterName=RandomNeon})
  n61 = TextureSample(UVs=Multiply(A=Add(A=@n56, B=@n60), B=ScalarParameter() {defaultValue=0.06, group=NeonEmissive, parameterName=SpeedNeonFlicker})) {samplerSource=SSM_Wrap_WorldGroupSettings, texture=/Game/VisualThreatManager/StormVisuals/Textures/StormLightning}
  n62 = Power(
   Base=Multiply(A=Multiply(A=@n43, B=@n53), B=ScalarParameter() {defaultValue=3, group=NeonEmissive, parameterName=EmissiveNeonSignBrightness})
   Exp=MaterialProxyReplace(Realtime=LinearInterpolate(A=@n54, B=@n55, Alpha=@n61.R), MaterialProxy=@n55)
  )
  n63 = TextureSampleParameter2D() {
   group=Material_1
   parameterName=SpecularMasks
   samplerSource=SSM_Wrap_WorldGroupSettings
   samplerType=SAMPLERTYPE_Masks
   sortPriority=2
   texture=&T_Fortnite_Default_S
  }
  n64 = TextureSampleParameter2D() {
   group=NeonEmissive
   parameterName=EmissiveNeonMasks
   samplerSource=SSM_Wrap_WorldGroupSettings
   samplerType=SAMPLERTYPE_Masks
   texture=/Game/Environments/Sets/Neo/Architecture/Textures/T_Neo_StoreSign_S
  }
  n65 = StaticSwitchParameter(True=@n63.RGBA, False=@n64.RGBA) {group=Emissive, parameterName=UseBaseColorForEmissive}
  n66 = Multiply(
   A=Multiply(
    A=StaticSwitchParameter(True=@n5, False=@n1.RGB) {parameterName=UseRandomEmissiveColor}
    B=StaticSwitchParameter(True=@n27, False=@n62) {parameterName=isFakeInterior}
   )
   B=Ceil(None=StaticComponentMaskParameter(None=@n65) {defaultA=true, group=NeonEmissive, parameterName=EmissiveNeonMask})
  )
  FunctionOutput(None=@n66)
  n67 = TextureSampleParameter2D(UVs=@n13) {
   group=NeonEmissive
   parameterName=Neon_Emissive_Inside_Blur
   samplerSource=SSM_Wrap_WorldGroupSettings
   samplerType=SAMPLERTYPE_Masks
   texture=&T_NeonSign_M
  }
  LinearInterpolate(A=@n67.RGB, B=@n14.RGB, Alpha=@n17.Result)
  MaterialFunctionCall!(HeightTexture, TextureCoordinates, ReferencePlane, HeightRatioInput, CameraVector) {materialFunction=/Engine/Functions/Engine_MaterialFunctions02/Utility/BumpOffset_Advanced.BumpOffset_advanced}
  n68 = VectorParameter() {group=NeonEmissive, parameterName=ReflectionsColor, defaultValue=<rgba 1 1 1 1>}
  n69 = TextureSampleParameterCube!(UVs=ReflectionVectorWS(), Apply View MipBias) {
   group=NeonEmissive
   parameterName=FakeReflection
   samplerSource=SSM_Wrap_WorldGroupSettings
   samplerType=SAMPLERTYPE_LinearColor
   texture=/Game/Environments/Sets/Neo/Architecture/Textures/T_FakeWindowReflections
  }
  StaticSwitchParameter(True=Add(A=Multiply(A=@n68.RGB, B=@n69.RGB), B=@n66), False=@n66) {parameterName=UseFakeWindowReflections}
 ]}
