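# Material function that builds UVs for four material layers and outputs each separately, plus a combined set for layered and POM use.
# Each layer scales UV0, optionally pans it, adds an offset and optionally applies a FlipBook, using per-layer parameters and static switches.
# Nested If nodes compare UV1's X value against 1, 2 and 3 to pick which layer's UVs feed both combined outputs.
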MaterialFunction /Game/Athena/WinterArt/Materials/MF_UVsSet
graph {
 properties {description=This function checks through UV Coordinate's 0 through 3}
 nodes [
  n1 = Constant() {r=3}
  n2 = ComponentMask(None=TextureCoordinate() {coordinateIndex=1}) {r=true}
  n3 = Multiply(A=TextureCoordinate(), B=ScalarParameter() {defaultValue=1, group=Material_1, parameterName=Mat1 UV Scale, sortPriority=3})
  n4 = AppendVector(
   A=ScalarParameter() {defaultValue=1, group=Material_1, parameterName="Mat 1 Speed  X", sortPriority=-5}
   B=ScalarParameter() {group=Material_1, parameterName="Mat 1 Speed  Y", sortPriority=-5}
  )
  n5 = Add(
   A=StaticSwitchParameter(True=Panner(Coordinate=@n3, Speed=@n4) {bFractionalPart=true}, False=@n3) {group=Material_1, parameterName=L1_Panner True, sortPriority=-5}
   B=AppendVector(
    A=ScalarParameter() {group=Material_1, parameterName="Mat 1 Offset  X", sortPriority=5}
    B=ScalarParameter() {group=Material_1, parameterName="Mat 1 Offset  Y", sortPriority=5}
   )
  )
  n6 = QualitySwitch(Default=@n5, Low=@n5, High=@n5, Medium=@n5, Epic=@n5)
  n7 = ShadingPathSwitch(Default=@n6, Deferred=@n6, Forward=@n5, Mobile=@n5)
  n8 = StaticSwitchParameter(
   True=MaterialFunctionCall(
    "Animation  Phase (0-1)"=ScalarParameter() {group=Material_1, parameterName=L1_FB_Frame, sortPriority=24}
    Number of Rows=ScalarParameter() {defaultValue=2, group=Material_1, parameterName=L1_FB_Rows, sortPriority=22}
    Number of Columns=ScalarParameter() {defaultValue=2, group=Material_1, parameterName=L1_FB_Columns, sortPriority=23}
    UVs=@n7
   ) {materialFunction=&FlipBook}.UVs
   False=@n7
  ) {group=Material_1, parameterName=L1_UseFlipBook, sortPriority=21}
  n9 = Multiply(A=TextureCoordinate(), B=ScalarParameter() {defaultValue=1, group=Material_2, parameterName="Mat 2 UV Scale ", sortPriority=4})
  n10 = AppendVector(
   A=ScalarParameter() {defaultValue=1, group=Material_2, parameterName="Mat 2 Speed  X", sortPriority=-5}
   B=ScalarParameter() {group=Material_2, parameterName="Mat 2 Speed  Y", sortPriority=-5}
  )
  n11 = Add(
   A=StaticSwitchParameter(True=Panner(Coordinate=@n9, Speed=@n10) {bFractionalPart=true}, False=@n9) {group=Material_2, parameterName=L2_Panner True, sortPriority=-5}
   B=AppendVector(
    A=ScalarParameter() {group=Material_2, parameterName=Mat 2 Offset X, sortPriority=5}
    B=ScalarParameter() {group=Material_2, parameterName=Mat 2 Offset Y, sortPriority=5}
   )
  )
  n12 = StaticSwitchParameter(
   True=MaterialFunctionCall(
    "Animation  Phase (0-1)"=ScalarParameter() {group=Material_2, parameterName=L2_FB_Frame, sortPriority=24}
    Number of Rows=ScalarParameter() {defaultValue=2, group=Material_2, parameterName=L2_FB_Rows, sortPriority=22}
    Number of Columns=ScalarParameter() {defaultValue=2, group=Material_2, parameterName=L2_FB_Columns, sortPriority=23}
    UVs=@n11
   ) {materialFunction=&FlipBook}.UVs
   False=@n11
  ) {group=Material_2, parameterName=L2_UseFlipBook, sortPriority=21}
  n13 = Multiply(A=TextureCoordinate(), B=ScalarParameter() {defaultValue=1, group=Material_3, parameterName="Mat 3 UV Scale ", sortPriority=4})
  n14 = AppendVector(
   A=ScalarParameter() {defaultValue=1, group=Material_3, parameterName="Mat 3 Speed  X", sortPriority=-5}
   B=ScalarParameter() {group=Material_3, parameterName="Mat 3 Speed  Y", sortPriority=-5}
  )
  n15 = Add(
   A=StaticSwitchParameter(True=Panner(Coordinate=@n13, Speed=@n14) {bFractionalPart=true}, False=@n13) {group=Material_3, parameterName=L3_Panner True, sortPriority=-5}
   B=AppendVector(
    A=ScalarParameter() {group=Material_3, parameterName=Mat 3 Offset X, sortPriority=5}
    B=ScalarParameter() {group=Material_3, parameterName=Mat 3 Offset Y, sortPriority=5}
   )
  )
  n16 = StaticSwitchParameter(
   True=MaterialFunctionCall(
    "Animation  Phase (0-1)"=ScalarParameter() {group=Material_3, parameterName=L3_FB_Frame, sortPriority=24}
    Number of Rows=ScalarParameter() {defaultValue=2, group=Material_3, parameterName=L3_FB_Rows, sortPriority=22}
    Number of Columns=ScalarParameter() {defaultValue=2, group=Material_3, parameterName=L3_FB_Columns, sortPriority=23}
    UVs=@n15
   ) {materialFunction=&FlipBook}.UVs
   False=@n15
  ) {group=Material_3, parameterName=L3_UseFlipBook, sortPriority=21}
  n17 = If(A=Constant() {r=2}, B=@n2, "A > B"=If(A=Constant() {r=1}, B=@n2, "A > B"=@n8, "A < B"=@n12) {desc=0}, "A < B"=@n16)
  n18 = Multiply(A=TextureCoordinate(), B=ScalarParameter() {defaultValue=1, group=Material_4, parameterName="Mat 4 UV Scale ", sortPriority=4})
  n19 = AppendVector(
   A=ScalarParameter() {defaultValue=1, group=Material_4, parameterName="Mat 4 Speed  X", sortPriority=-5}
   B=ScalarParameter() {group=Material_4, parameterName="Mat 4 Speed  Y", sortPriority=-5}
  )
  n20 = Add(
   A=StaticSwitchParameter(True=Panner(Coordinate=@n18, Speed=@n19) {bFractionalPart=true}, False=@n18) {group=Material_4, parameterName=L4_Panner True, sortPriority=-5}
   B=AppendVector(
    A=ScalarParameter() {group=Material_4, parameterName=Mat 4 Offset X, sortPriority=5}
    B=ScalarParameter() {group=Material_4, parameterName=Mat 4 Offset Y, sortPriority=5}
   )
  )
  n21 = StaticSwitchParameter(
   True=MaterialFunctionCall(
    "Animation  Phase (0-1)"=ScalarParameter() {group=Material_4, parameterName=L4_FB_Frame, sortPriority=24}
    Number of Rows=ScalarParameter() {defaultValue=2, group=Material_4, parameterName=L4_FB_Rows, sortPriority=22}
    Number of Columns=ScalarParameter() {defaultValue=2, group=Material_4, parameterName=L4_FB_Columns, sortPriority=23}
    UVs=@n20
   ) {materialFunction=&FlipBook}.UVs
   False=@n20
  ) {group=Material_4, parameterName=L4_UseFlipBook, sortPriority=21}
  FunctionOutput(None=FunctionOutput(None=If(A=@n1, B=@n2, "A > B"=@n17, "A < B"=@n21)) {outputName=UVs 4 Layers}) {outputName=POM UVs 4 Layers, sortPriority=1}
  FunctionOutput(None=@n8) {outputName=UVs Layer 1}
  FunctionOutput(None=@n12) {outputName=UVs Layer 2}
  FunctionOutput(None=@n16) {outputName=UVs Layer 3}
  FunctionOutput(None=@n21) {outputName=UVs Layer 4}
 ]
}
