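# Four-layer Nanite-enabled master material for environment meshes, with ModBox customization; texture notes suggest building use (interior walls, trims, doors).
# Each layer's tintable diffuse, spec masks, normals and animated emissive feed MF_FN_TextureSetLerp_4Layers, with optional two-sided and neon on layer 1.
# A shading-path switch applies mobile spec, roughness and metal multipliers before MF_ModBoxCustomization.
# That result feeds game-wide and cinematic functions, plus a separate Apollo snow-progression call.
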Material /Game/Athena/WinterArt/Materials/M_BaseSimple_4Layers_ModBox
graph {
 properties {
  bAutomaticallySetUsageInEditor=false
  bUseMaterialAttributes=true
  bUsedWithNanite=true
  ditheredLODTransition=true
  pixelDepthOffsetMode=PDOM_Legacy
  refractionMethod=RM_IndexOfRefraction
  usedShadingModels=Default Lit
 }
 nodes [
  n1 = VectorParameter() {
   group=Material_3
   parameterName=Emissive3
   sortPriority=7
   channelNames {a=Emissive Multiplier, b="", g="", r=""}
   defaultValue=<rgba 0 0 0 1>
  }
  n2 = Multiply(A=@n1, B=@n1)
  n3 = Saturate(None=@n2)
  n4 = VectorParameter() {
   group=Material_4
   parameterName=Emissive4
   sortPriority=7
   channelNames {a=Emissive Multiplier, b="", g="", r=""}
   defaultValue=<rgba 0 0 0 1>
  }
  n5 = Multiply(A=@n4, B=@n4)
  n6 = Saturate(None=@n5)
  n7 = ScalarParameter() {defaultValue=0.5, group=SSS, parameterName=SSSDepth}
  n8 = MaterialFunctionCall() {materialFunction=&/MF_UVsSet}
  n9 = VectorParameter() {
   group=Material_1
   parameterName=L1_UVScaleByComponent
   sortPriority=3
   channelNames {a="", b=Normals, g=Specular, r=Diffuse}
   defaultValue=<rgba 1 1 1 0>
  }
  n10 = TextureSampleParameter2D(UVs=Multiply(A=@n8.UVs Layer 1, B=@n9.Diffuse)) {
   group=Material_1
   parameterName=Diffuse Side 2
   samplerSource=SSM_Wrap_WorldGroupSettings
   sortPriority=1
   texture=/Game/Global/Textures/Default/Blanks/T_MidgreysRGB
  }
  n11 = TwoSidedSign()
  n12 = VectorParameter() {
   group=Material_1
   parameterName=L1_UVScaleByComponent
   sortPriority=3
   channelNames {a="", b=Normals, g=Specular, r=Diffuse}
   defaultValue=<rgba 1 1 1 0>
  }
  n13 = Multiply(A=@n8.UVs Layer 1, B=@n12.Diffuse)
  n14 = StaticSwitchParameter(True=Multiply(A=ScalarParameter() {defaultValue=1, group=Material_1, parameterName=L1_UV_ColorSwatchScaleAdd}, B=@n13), False=@n13) {group=Material_1, parameterName=Use Color Swatch}
  n15 = TextureSampleParameter2D(UVs=Multiply(A=ScalarParameter() {defaultValue=1, group=Material_1, parameterName=L1_CustomHeightScale, sortPriority=1}, B=@n14)) {
   group=Material_1
   parameterName=L1_Height
   samplerSource=SSM_Wrap_WorldGroupSettings
   samplerType=SAMPLERTYPE_LinearGrayscale
   sortPriority=-5
   texture=/Game/Global/Textures/Default/Blanks/T_BlackLinearGrayscale
  }
  n16 = TextureSampleParameter2D(UVs=@n13) {
   group=Material_1
   parameterName=Background Diffuse
   samplerSource=SSM_Wrap_WorldGroupSettings
   sortPriority=-5
   texture=&T_WhitesRGB
  }
  n17 = TextureSampleParameter2D(UVs=@n14) {
   group=Material_1
   parameterName=Diffuse
   samplerSource=SSM_Wrap_WorldGroupSettings
   sortPriority=-5
   texture=&T_MidgreysRGB
  }
  n18 = StaticSwitchParameter(True=@n15.R, False=StaticSwitchParameter(True=@n16.A, False=@n17.A) {group=Material_1, parameterName=Use Color Swatch}) {desc=Use Smale scale grayscale, group=Material_1, parameterName=L1 Custom Height, sortPriority=1}
  n19 = Multiply(A=@n12.Specular, B=@n8.UVs Layer 1)
  n20 = TextureSampleParameter2D(UVs=@n19) {
   group=Material_1
   parameterName=SpecularMasks
   samplerSource=SSM_Wrap_WorldGroupSettings
   samplerType=SAMPLERTYPE_Masks
   sortPriority=-5
   texture=&T_Fortnite_Default_S
  }
  n21 = MaterialFunctionCall!(
   Transition=ScalarParameter() {defaultValue=0.5, group=Material_1, parameterName=L1_TintTransition, sortPriority=5}
   HM=@n18
   Contrast=ScalarParameter() {defaultValue=5, group=Material_1, parameterName=L1_TintContrast, sortPriority=5}
   Metalness=@n20.G
   BG Color=StaticSwitchParameter(True=@n16.RGB, False=@n17.RGB) {group=Material_1, parameterName=Use Color Swatch}
   TintColor=StaticSwitchParameter(True=@n17.RGB, False=VectorParameter() {group=Material_1, parameterName=TintColor, sortPriority=7, defaultValue=<rgba 1 1 1 0>}.RGB) {group=Material_1, parameterName=Use Color Swatch}
  ) {materialFunction=&/MF_ColorMask}
  n22 = AppendVector(
   A=StaticSwitchParameter(True=@n21.LerpMix, False=@n21.LerpMix) {group=Material_1, parameterName=L1_UseTintColorAsDiffuse}
   B=Multiply(A=@n18, B=ScalarParameter() {defaultValue=1, group=Material_1, parameterName=L1 Alpha, sortPriority=1})
  )
  n23 = MaterialFunctionCall() {materialFunction=&MF_UVsSet}
  n24 = VectorParameter() {
   group=Material_3
   parameterName=L3_UVScaleByComponent
   sortPriority=3
   channelNames {a="", b=Normals, g=Specular, r=Diffuse}
   defaultValue=<rgba 1 1 1 0>
  }
  n25 = TextureSampleParameter2D(UVs=Multiply(A=@n23.UVs Layer 3, B=@n24.Normals)) {
   desc=Use for trims and accents
   group=Material_3
   parameterName=Normals_Texture_3
   samplerSource=SSM_Wrap_WorldGroupSettings
   samplerType=SAMPLERTYPE_Normal
   sortPriority=-5
   texture=/Engine/EngineMaterials/FlatNormal
  }
  n26 = ScalarParameter() {defaultValue=1, group=Material_3, parameterName=L3_BoostNormals, sortPriority=3}
  n27 = AppendVector(A=@n26, B=@n26)
  n28 = Constant() {r=1}
  n29 = VectorParameter() {
   group=Material_4
   parameterName=L4_UVScaleByComponent
   sortPriority=3
   channelNames {a="", b=Normals, g=Specular, r=Diffuse}
   defaultValue=<rgba 1 1 1 0>
  }
  n30 = MaterialFunctionCall() {materialFunction=&MF_UVsSet}
  n31 = Multiply(A=@n29.Specular, B=@n30.UVs Layer 4)
  n32 = TextureSampleParameter2D(UVs=@n31) {
   desc=Use for doors
   group=Material_4
   parameterName=SpecularMasks_4
   samplerSource=SSM_Wrap_WorldGroupSettings
   samplerType=SAMPLERTYPE_Masks
   sortPriority=-5
   texture=&T_Fortnite_Default_S
  }
  n33 = TextureCoordinate() {coordinateIndex=1}
  n34 = TextureSampleParameter2D(UVs=StaticSwitchParameter(
    True=TextureCoordinate() {coordinateIndex=2}
    False=StaticSwitchParameter(True=@n33, False=@n31) {group=Material_4, parameterName=L4_Ch2Emissive, sortPriority=-5}
   ) {group=Material_4, parameterName=L4_Ch3Emissive, sortPriority=-5}) {
   group=Material_4
   parameterName=L4_Emissive
   samplerSource=SSM_Wrap_WorldGroupSettings
   samplerType=SAMPLERTYPE_LinearGrayscale
   sortPriority=-5
   texture=&T_BlackLinearGrayscale
  }
  n35 = Multiply(A=@n24.Diffuse, B=@n23.UVs Layer 3)
  n36 = StaticSwitchParameter(True=Multiply(A=ScalarParameter() {defaultValue=1, group=Material_3, parameterName=L3_UV_ColorSwatchScaleAdd}, B=@n35), False=@n35) {group=Material_3, parameterName=L3 Use Color Swatch, sortPriority=1}
  n37 = TextureSampleParameter2D(UVs=Multiply(A=ScalarParameter() {defaultValue=1, group=Material_3, parameterName=L3_CustomHeightScale, sortPriority=1}, B=@n36)) {
   group=Material_3
   parameterName=L3_Height
   samplerSource=SSM_Wrap_WorldGroupSettings
   samplerType=SAMPLERTYPE_LinearGrayscale
   sortPriority=-5
   texture=&T_BlackLinearGrayscale
  }
  n38 = TextureSampleParameter2D(UVs=@n35) {
   group=Material_3
   parameterName=Background Diffuse 3
   samplerSource=SSM_Wrap_WorldGroupSettings
   sortPriority=-5
   texture=&T_WhitesRGB
  }
  n39 = TextureSampleParameter2D(UVs=@n36) {
   desc=Use for trims and accents
   group=Material_3
   parameterName=Diffuse_Texture_3
   samplerSource=SSM_Wrap_WorldGroupSettings
   sortPriority=-5
   texture=&T_MidgreysRGB
  }
  n40 = StaticSwitchParameter(True=@n37.R, False=StaticSwitchParameter(True=@n38.A, False=@n39.A) {group=Material_3, parameterName=L3 Use Color Swatch, sortPriority=1}) {desc=Use Smale scale grayscale, group=Material_3, parameterName=L3 Custom Height, sortPriority=1}
  n41 = Multiply(A=@n24.Specular, B=@n23.UVs Layer 3)
  n42 = TextureSampleParameter2D(UVs=@n41) {
   desc=Use for trims and accents
   group=Material_3
   parameterName=SpecularMasks_3
   samplerSource=SSM_Wrap_WorldGroupSettings
   samplerType=SAMPLERTYPE_Masks
   sortPriority=-5
   texture=&T_Fortnite_Default_S
  }
  n43 = MaterialFunctionCall(
   Transition=ScalarParameter() {defaultValue=0.5, group=Material_3, parameterName=L3_TintTransition, sortPriority=5}
   HM=@n40
   Contrast=ScalarParameter() {defaultValue=1, group=Material_3, parameterName=L3_TintContrast, sortPriority=5}
   Metalness=@n42.G
   BG Color=StaticSwitchParameter(True=@n38.RGB, False=@n39.RGB) {group=Material_3, parameterName=L3 Use Color Swatch, sortPriority=1}
   TintColor=StaticSwitchParameter(True=@n39.RGB, False=VectorParameter() {group=Material_3, parameterName=M3_TintColor, sortPriority=5, defaultValue=<rgba 1 1 1 0>}.RGB) {group=Material_3, parameterName=L3 Use Color Swatch, sortPriority=1}
  ) {materialFunction=&MF_ColorMask}
  n44 = ScalarParameter() {defaultValue=1, group=Material_3, parameterName=L3 Alpha, sortPriority=1}
  n45 = Multiply(A=@n29.Diffuse, B=@n30.UVs Layer 4)
  n46 = StaticSwitchParameter(True=Multiply(A=ScalarParameter() {defaultValue=1, group=Material_4, parameterName=L4_UV_ColorSwatchScaleAdd}, B=@n45), False=@n45) {group=Material_4, parameterName=L4 Use Color Swatch, sortPriority=1}
  n47 = TextureSampleParameter2D(UVs=Multiply(A=ScalarParameter() {defaultValue=1, group=Material_4, parameterName=L4_CustomHeightScale, sortPriority=1}, B=@n46)) {
   group=Material_4
   parameterName=L4_Height
   samplerSource=SSM_Wrap_WorldGroupSettings
   samplerType=SAMPLERTYPE_LinearGrayscale
   sortPriority=-5
   texture=&T_BlackLinearGrayscale
  }
  n48 = TextureSampleParameter2D(UVs=@n45) {
   group=Material_4
   parameterName=Background Diffuse 4
   samplerSource=SSM_Wrap_WorldGroupSettings
   sortPriority=-5
   texture=&T_WhitesRGB
  }
  n49 = TextureSampleParameter2D(UVs=@n46) {
   desc=Use for doors
   group=Material_4
   parameterName=Diffuse_Texture_4
   samplerSource=SSM_Wrap_WorldGroupSettings
   sortPriority=-5
   texture=&T_MidgreysRGB
  }
  n50 = StaticSwitchParameter(True=@n47.R, False=StaticSwitchParameter(True=@n48.A, False=@n49.A) {group=Material_4, parameterName=L4 Use Color Swatch, sortPriority=1}) {desc=Use Smale scale grayscale, group=Material_4, parameterName=L4 Custom Height, sortPriority=1}
  n51 = MaterialFunctionCall(
   Transition=ScalarParameter() {defaultValue=0.5, group=Material_4, parameterName=L4_TintTransition, sortPriority=5}
   HM=@n50
   Contrast=ScalarParameter() {defaultValue=1, group=Material_4, parameterName=L4_TintContrast, sortPriority=5}
   Metalness=@n32.G
   BG Color=StaticSwitchParameter(True=@n48.RGB, False=@n49.RGB) {group=Material_4, parameterName=L4 Use Color Swatch, sortPriority=1}
   TintColor=StaticSwitchParameter(True=@n49.RGB, False=VectorParameter() {group=Material_4, parameterName=M4_TintColor, sortPriority=5, defaultValue=<rgba 1 1 1 0>}.RGB) {group=Material_4, parameterName=L4 Use Color Swatch, sortPriority=1}
  ) {materialFunction=&MF_ColorMask}
  n52 = ScalarParameter() {defaultValue=1, group=Material_4, parameterName=L4 Alpha, sortPriority=1}
  n53 = TextureSampleParameter2D(UVs=Multiply(A=@n29.Normals, B=@n30.UVs Layer 4)) {
   desc=Use for doors
   group=Material_4
   parameterName=Normals_Texture_4
   samplerSource=SSM_Wrap_WorldGroupSettings
   samplerType=SAMPLERTYPE_Normal
   sortPriority=-5
   texture=&FlatNormal
  }
  n54 = ScalarParameter() {defaultValue=1, group=Material_4, parameterName=L4_BoostNormals, sortPriority=3}
  n55 = AppendVector(A=@n54, B=@n54)
  n56 = Constant() {r=1}
  n57 = TextureObjectParameter() {group=Material_1, parameterName=Diffuse, sortPriority=1, texture=&T_MidgreysRGB}
  n58 = ScalarParameter() {group=Adv Emissive, parameterName=L1 X Speed, sortPriority=12}
  n59 = ScalarParameter() {group=Adv Emissive, parameterName=L1 Y Speed, sortPriority=13}
  n60 = VectorParameter() {group=Adv Emissive, parameterName=L1 Channel, sortPriority=9}
  n61 = ScalarParameter() {defaultValue=1, group=Adv Emissive, parameterName=L1 Offset, sliderMax=4, sliderMin=1, sortPriority=11}
  n62 = VectorParameter() {
   bUseCustomPrimitiveData=true
   desc="For when emissive absolutely needs to be done with primitive data on the mesh (likely Creative)"
   parameterName=EmissiveWithPrimitiveData
   channelNames {a=Emissive Multiplier, b="", g="", r=""}
   defaultValue=<rgba 0 0 0 1>
  }
  n63 = VectorParameter() {
   desc=WARNING* Run through Saturate node when Lumen
   group=Material_1
   parameterName=Emissive
   sortPriority=7
   channelNames {a=Emissive Multiplier, b="", g="", r=""}
   defaultValue=<rgba 0 0 0 1>
  }
  n64 = StaticSwitchParameter(True=Multiply(A=@n62, B=@n62), False=Multiply(A=@n63, B=@n63)) {
   desc="For when emissive absolutely needs to be done with primitive data on the mesh (likely Creative)"
   parameterName=UseEmissiveWithPrimitiveData
  }
  n65 = RayTracingQualitySwitch(Normal=@n64, RayTraced=Saturate(None=@n64)) {desc=For Lumen}
  n66 = Multiply(A=@n8.UVs Layer 1, B=@n9.Specular)
  n67 = TextureSampleParameter2D(UVs=@n66) {
   group=Material_1
   parameterName=SpecularMasks Side 2
   samplerSource=SSM_Wrap_WorldGroupSettings
   samplerType=SAMPLERTYPE_Masks
   sortPriority=2
   texture=&T_Fortnite_Default_S
  }
  n68 = TextureCoordinate() {coordinateIndex=1}
  n69 = TextureSampleParameter2D(UVs=StaticSwitchParameter(
    True=TextureCoordinate() {coordinateIndex=2}
    False=StaticSwitchParameter(True=@n68, False=@n66) {group=Material_1, parameterName=L1_Ch2Emissive Side2, sortPriority=-5}
   ) {group=Material_1, parameterName=L1_Ch3Emissive Side2, sortPriority=-5}) {
   group=Material_1
   parameterName=L1_Emissive Side 2
   samplerSource=SSM_Wrap_WorldGroupSettings
   samplerType=SAMPLERTYPE_LinearGrayscale
   sortPriority=-4
   texture=&T_BlackLinearGrayscale
  }
  n70 = AppendVector(A=@n67.RGB, B=StaticSwitchParameter(True=@n69.R, False=@n67.A) {group=Material_1, parameterName=L1_CustomEmissive Side 2, sortPriority=-5})
  n71 = TextureCoordinate() {coordinateIndex=1}
  n72 = TextureSampleParameter2D(UVs=StaticSwitchParameter(
    True=TextureCoordinate() {coordinateIndex=2}
    False=StaticSwitchParameter(True=@n71, False=@n19) {group=Material_1, parameterName=L1_Ch2Emissive, sortPriority=-5}
   ) {group=Material_1, parameterName=L1_Ch3Emissive, sortPriority=1}) {
   group=Material_1
   parameterName=L1_Emissive
   samplerSource=SSM_Wrap_WorldGroupSettings
   samplerType=SAMPLERTYPE_LinearGrayscale
   sortPriority=-5
   texture=&T_BlackLinearGrayscale
  }
  n73 = AppendVector(A=@n20.RGB, B=StaticSwitchParameter(True=@n72.R, False=@n20.A) {group=Material_1, parameterName=L1_CustomEmissive, sortPriority=-5})
  n74 = VectorParameter() {
   group=Material_2
   parameterName=Emissive2
   sortPriority=7
   channelNames {a=Emissive Multiplier, b="", g="", r=""}
   defaultValue=<rgba 0 0 0 1>
  }
  n75 = Multiply(A=@n74, B=@n74)
  n76 = Saturate(None=@n75)
  n77 = TextureCoordinate() {coordinateIndex=1}
  n78 = TextureSampleParameter2D(UVs=StaticSwitchParameter(
    True=TextureCoordinate() {coordinateIndex=2}
    False=StaticSwitchParameter(True=@n77, False=@n41) {group=Material_3, parameterName=L3_Ch2Emissive, sortPriority=-5}
   ) {group=Material_3, parameterName=L3_Ch3Emissive, sortPriority=-5}) {
   group=Material_3
   parameterName=L3_Emissive
   samplerSource=SSM_Wrap_WorldGroupSettings
   samplerType=SAMPLERTYPE_LinearGrayscale
   sortPriority=-5
   texture=&T_BlackLinearGrayscale
  }
  n79 = TextureSampleParameter2D(UVs=Multiply(A=@n9.Normals, B=@n8.UVs Layer 1)) {
   group=Material_1
   parameterName=Normals Side 2
   samplerSource=SSM_Wrap_WorldGroupSettings
   samplerType=SAMPLERTYPE_Normal
   sortPriority=3
   texture=&FlatNormal
  }
  n80 = TextureSampleParameter2D(UVs=Multiply(A=@n12.Normals, B=@n8.UVs Layer 1)) {
   group=Material_1
   parameterName=Normals
   samplerSource=SSM_Wrap_WorldGroupSettings
   samplerType=SAMPLERTYPE_Normal
   sortPriority=-5
   texture=&FlatNormal
  }
  n81 = ScalarParameter() {defaultValue=1, group=Material_1, parameterName=L1_BoostNormals, sortPriority=3}
  n82 = Multiply(A=@n80.RGB, B=AppendVector(A=AppendVector(A=@n81, B=@n81), B=Constant() {r=1}))
  n83 = VectorParameter() {
   group=Material_2
   parameterName=L2_UVScaleByComponent
   sortPriority=3
   channelNames {a="", b=Normals, g=Specular, r=Diffuse}
   defaultValue=<rgba 1 1 1 0>
  }
  n84 = MaterialFunctionCall() {materialFunction=&MF_UVsSet}
  n85 = Multiply(A=@n83.Diffuse, B=@n84.UVs Layer 2)
  n86 = StaticSwitchParameter(True=Multiply(A=ScalarParameter() {defaultValue=1, group=Material_2, parameterName=L2_UV_ColorSwatchScaleAdd}, B=@n85), False=@n85) {group=Material_2, parameterName=L2 Use Color Swatch, sortPriority=1}
  n87 = TextureSampleParameter2D(UVs=Multiply(A=ScalarParameter() {defaultValue=1, group=Material_2, parameterName=L2_CustomHeightScale, sortPriority=1}, B=@n86)) {
   group=Material_2
   parameterName=L2_Height
   samplerSource=SSM_Wrap_WorldGroupSettings
   samplerType=SAMPLERTYPE_LinearGrayscale
   sortPriority=-5
   texture=&T_BlackLinearGrayscale
  }
  n88 = TextureSampleParameter2D(UVs=@n85) {
   group=Material_2
   parameterName=Background Diffuse 2
   samplerSource=SSM_Wrap_WorldGroupSettings
   sortPriority=-5
   texture=&T_WhitesRGB
  }
  n89 = TextureSampleParameter2D(UVs=@n86) {
   desc=Use for interior walls and ceilings
   group=Material_2
   parameterName=Diffuse_Texture_2
   samplerSource=SSM_Wrap_WorldGroupSettings
   sortPriority=-5
   texture=&T_MidgreysRGB
  }
  n90 = StaticSwitchParameter(True=@n87.R, False=StaticSwitchParameter(True=@n88.A, False=@n89.A) {group=Material_2, parameterName=L2 Use Color Swatch, sortPriority=1}) {desc=Use Smale scale grayscale, group=Material_2, parameterName=L2 Custom Height, sortPriority=1}
  n91 = Multiply(A=@n83.Specular, B=@n84.UVs Layer 2)
  n92 = TextureSampleParameter2D(UVs=@n91) {
   desc=Use for interior walls and ceilings
   group=Material_2
   parameterName=SpecularMasks_2
   samplerSource=SSM_Wrap_WorldGroupSettings
   samplerType=SAMPLERTYPE_Masks
   sortPriority=-5
   texture=&T_Fortnite_Default_S
  }
  n93 = MaterialFunctionCall(
   Transition=ScalarParameter() {defaultValue=0.5, group=Material_2, parameterName=L2_TintTransition, sortPriority=5}
   HM=@n90
   Contrast=ScalarParameter() {defaultValue=1, group=Material_2, parameterName=L2_TintContrast, sortPriority=5}
   Metalness=@n92.G
   BG Color=StaticSwitchParameter(True=@n88.RGB, False=@n89.RGB) {group=Material_2, parameterName=L2 Use Color Swatch, sortPriority=1}
   TintColor=StaticSwitchParameter(True=@n89.RGB, False=VectorParameter() {group=Material_2, parameterName=M2_TintColor, sortPriority=5, defaultValue=<rgba 1 1 1 0>}.RGB) {group=Material_2, parameterName=L2 Use Color Swatch, sortPriority=1}
  ) {materialFunction=&MF_ColorMask}
  n94 = ScalarParameter() {defaultValue=1, group=Material_2, parameterName=L2 Alpha, sortPriority=1}
  n95 = TextureCoordinate() {coordinateIndex=1}
  n96 = TextureSampleParameter2D(UVs=StaticSwitchParameter(
    True=TextureCoordinate() {coordinateIndex=2}
    False=StaticSwitchParameter(True=@n95, False=@n91) {group=Material_2, parameterName=L2_Ch2Emissive, sortPriority=-5}
   ) {group=Material_2, parameterName=L2_Ch3Emissive, sortPriority=-5}) {
   group=Material_2
   parameterName=L2_Emissive
   samplerSource=SSM_Wrap_WorldGroupSettings
   samplerType=SAMPLERTYPE_LinearGrayscale
   sortPriority=-5
   texture=&T_BlackLinearGrayscale
  }
  n97 = TextureSampleParameter2D(UVs=Multiply(A=@n83.Normals, B=@n84.UVs Layer 2)) {
   desc=Use for interior walls and ceilings
   group=Material_2
   parameterName=Normals_Texture_2
   samplerSource=SSM_Wrap_WorldGroupSettings
   samplerType=SAMPLERTYPE_Normal
   sortPriority=-5
   texture=&FlatNormal
  }
  n98 = ScalarParameter() {defaultValue=1, group=Material_2, parameterName=L2_BoostNormals, sortPriority=3}
  n99 = AppendVector(A=@n98, B=@n98)
  n100 = Constant() {r=1}
  n101 = MaterialFunctionCall!(
   Use3Materials
   Use4Materials
   Use2Materials
   D=StaticSwitchParameter(True=LinearInterpolate(A=@n10.RGBA, B=@n22, Alpha=@n11), False=@n22) {group=Material_1, parameterName=TwoSided, sortPriority=18}
   S=StaticSwitchParameter(True=LinearInterpolate(A=@n70, B=@n73, Alpha=@n11), False=@n73) {group=Material_1, parameterName=TwoSided, sortPriority=18}
   N=StaticSwitchParameter(True=LinearInterpolate(A=@n79.RGB, B=@n82, Alpha=@n11), False=@n82) {group=Material_1, parameterName=TwoSided, sortPriority=18}
   D2=AppendVector(
    A=StaticSwitchParameter(True=@n93.LerpMix, False=@n93.LerpMix) {group=Material_2, parameterName=L2_UseTintColorAsDiffuse}
    B=Multiply(A=@n90, B=@n94)
   )
   S2=AppendVector(A=@n92.RGB, B=StaticSwitchParameter(True=@n96.R, False=@n92.A) {group=Material_2, parameterName=L2_CustomEmissive, sortPriority=-5})
   N2=Multiply(A=@n97.RGB, B=AppendVector(A=@n99, B=@n100))
   D3=AppendVector(
    A=StaticSwitchParameter(True=@n43.LerpMix, False=@n43.LerpMix) {group=Material_3, parameterName=L3_UseTintColorAsDiffuse}
    B=Multiply(A=@n40, B=@n44)
   )
   S3=AppendVector(A=@n42.RGB, B=StaticSwitchParameter(True=@n78.R, False=@n42.A) {group=Material_3, parameterName=L3_CustomEmissive, sortPriority=-5})
   N3=Multiply(A=@n25.RGB, B=AppendVector(A=@n27, B=@n28))
   D4=AppendVector(
    A=StaticSwitchParameter(True=@n51.LerpMix, False=@n51.LerpMix) {group=Material_4, parameterName=L4_UseTintColorAsDiffuse}
    B=Multiply(A=@n50, B=@n52)
   )
   S4=AppendVector(A=@n32.RGB, B=StaticSwitchParameter(True=@n34.R, False=@n32.A) {group=Material_4, parameterName=L4_CustomEmissive, sortPriority=-5})
   N4=Multiply(A=@n53.RGB, B=AppendVector(A=@n55, B=@n56))
   E1=StaticSwitchParameter(
    True=MaterialFunctionCall!(Base Diffuse=@n57) {materialFunction=&/MF_NeonSignEmissive}.Result
    False=MaterialFunctionCall!(X=@n58, Y=@n59, Dot=@n60.RGB, TextureOffset=@n61, EmissiveTexture=@n65) {materialFunction=&/MF_EmissiveMask}.EmissiveMasked
   ) {group=NeonEmissive, parameterName=UseNeonSignEmissive_Layer1}
   E2=MaterialFunctionCall(
    X=ScalarParameter() {group=Adv Emissive, parameterName=L2 X Speed, sortPriority=16}
    Y=ScalarParameter() {group=Adv Emissive, parameterName=L2 Y Speed, sortPriority=17}
    Dot=VectorParameter() {group=Adv Emissive, parameterName=L2 Channel, sortPriority=14}.RGB
    TextureOffset=ScalarParameter() {defaultValue=1, group=Adv Emissive, parameterName=L2 Offset, sliderMax=4, sliderMin=1, sortPriority=15}
    EmissiveTexture=RayTracingQualitySwitch(Normal=@n75, RayTraced=@n76) {desc=For Lumen}
   ) {materialFunction=&MF_EmissiveMask}.EmissiveMasked
   E3=MaterialFunctionCall(
    X=ScalarParameter() {group=Adv Emissive, parameterName=L3 X Speed, sortPriority=21}
    Y=ScalarParameter() {group=Adv Emissive, parameterName=L3 Y Speed, sortPriority=22}
    Dot=VectorParameter() {group=Adv Emissive, parameterName=L3 Channel, sortPriority=18}.RGB
    TextureOffset=ScalarParameter() {defaultValue=1, group=Adv Emissive, parameterName=L3 Offset, sliderMax=4, sliderMin=1, sortPriority=19}
    EmissiveTexture=RayTracingQualitySwitch(Normal=@n2, RayTraced=@n3) {desc=For Lumen}
   ) {materialFunction=&MF_EmissiveMask}.EmissiveMasked
   E4=MaterialFunctionCall(
    X=ScalarParameter() {group=Adv Emissive, parameterName=L4 X Speed, sortPriority=25}
    Y=ScalarParameter() {group=Adv Emissive, parameterName=L4 Y Speed, sortPriority=26}
    Dot=VectorParameter() {group=Adv Emissive, parameterName=L4 Channel, sortPriority=23}.RGB
    TextureOffset=ScalarParameter() {defaultValue=1, group=Adv Emissive, parameterName=L4 Offset, sliderMax=4, sliderMin=1, sortPriority=24}
    EmissiveTexture=RayTracingQualitySwitch(Normal=@n5, RayTraced=@n6) {desc=For Lumen}
   ) {materialFunction=&MF_EmissiveMask}.EmissiveMasked
   CCMaskByMetal_1=ScalarParameter() {group=Material_1, parameterName=L1 Metal Mask To Clearcoat, sortPriority=24}
   CC1=ScalarParameter() {group=Material_1, parameterName=L1 Clearcoat, sortPriority=22}
   CCR_1=ScalarParameter() {defaultValue=0.025, group=Material_1, parameterName=L1 Clearcoat Roughness, sortPriority=23}
   CC2=ScalarParameter() {group=Material_2, parameterName=L2 Clearcoat, sortPriority=22}
   CCR_2=ScalarParameter() {defaultValue=0.025, group=Material_2, parameterName=L2 Clearcoat Roughness, sortPriority=23}
   CCMaskByMetal_2=ScalarParameter() {group=Material_2, parameterName=L2 Metal Mask To Clearcoat, sortPriority=24}
   CC3=ScalarParameter() {group=Material_3, parameterName=L3 Clearcoat, sortPriority=22}
   CCR_3=ScalarParameter() {defaultValue=0.025, group=Material_4, parameterName=L3 Clearcoat Roughness, sortPriority=23}
   CCMaskByMetal_3=ScalarParameter() {group=Material_4, parameterName=L3 Metal Mask To Clearcoat, sortPriority=24}
   CC4=ScalarParameter() {group=Material_4, parameterName=L4 Clearcoat, sortPriority=22}
   CCMaskByMetal_4=ScalarParameter() {group=Material_4, parameterName=L4 Metal Mask To Clearcoat, sortPriority=24}
   CCR_4=ScalarParameter() {defaultValue=0.025, group=Material_4, parameterName=L4 Clearcoat Roughness, sortPriority=23}
   OpacityMaskCancel_1=ScalarParameter() {group=Material_1, parameterName=L1_OpacityMaskCancel}
   OpacityMaskCancel_2=ScalarParameter() {group=Material_2, parameterName=L2_OpacityMaskCancel}
   OpacityMaskCancel_3=ScalarParameter() {group=Material_3, parameterName=L3_OpacityMaskCancel}
   OpacityMaskCancel_4=ScalarParameter() {group=Material_4, parameterName=L4_OpacityMaskCancel}
   SSS_1=ScalarParameter() {defaultValue=1, group=SSS, parameterName=L1_SSS}
   SSS_2=ScalarParameter() {defaultValue=1, group=SSS, parameterName=L2_SSS}
   SSS_3=ScalarParameter() {defaultValue=1, group=SSS, parameterName=L3_SSS}
   SSS_4=ScalarParameter() {defaultValue=1, group=SSS, parameterName=L4_SSS}
   SSS_Depth_1=@n7
   SSS_Depth_2=@n7
   SSS_Depth_3=@n7
   SSS_Depth_4=@n7
  ) {materialFunction=&/MF_FN_TextureSetLerp_4Layers}
  n102 = SetMaterialAttributes!(
   MaterialAttributes
   Base Color=@n101.Diffuse
   Specular=@n101.Diffuse
   Metallic=@n101.Diffuse
   Roughness=@n101.Diffuse
   Emissive Color=@n101.Diffuse
   Normal=@n101.Diffuse
   Custom Data 0=@n101.Diffuse
   Custom Data 1=@n101.Diffuse
   Opacity Mask=@n101.Diffuse
   Subsurface Color=@n101.Diffuse
   Opacity=@n101.Diffuse
   Refraction=NamedRerouteDeclaration(None=@n101.Spec_A) {name=Spec A, nodeColor=<rgba 0 0.2734 1 1>}
  ) {attributeSetTypes [
    #1
    #2
    #3
    #4
    #5
    #6
    #7
    #8
    #9
    #10
    #11
    #12
   ]}
  n103 = GetMaterialAttributes(None=@n102) {attributeGetTypes [
    #1
    #5
    #9
    #2
    #4
    #3
   ]}
  n104 = SetMaterialAttributes!(
   MaterialAttributes
   Base Color=@n103.Base Color
   Emissive Color=@n103.Base Color
   Opacity Mask=@n103.Base Color
   Specular=Multiply(A=@n103.Specular, B=ScalarParameter() {defaultValue=1, group=Mobile, parameterName=Mobile_SpecularMult})
   Roughness=Multiply(A=@n103.Roughness, B=ScalarParameter() {defaultValue=1, group=Mobile, parameterName=Mobile_RoughnessMult})
   Metallic=Multiply(A=@n103.Metallic, B=ScalarParameter() {defaultValue=1, group=Mobile, parameterName=Mobile_MetalMult})
   Subsurface Color=@n103.Base Color
  ) {attributeSetTypes [
    #1
    #5
    #9
    #2
    #4
    #3
    #10
   ]}
  n105 = GetMaterialAttributes(None=@n102) {attributeGetTypes [
    #1
    #5
    #9
    #2
    #4
    #3
    #6
   ]}
  n106 = SetMaterialAttributes!(
   MaterialAttributes
   Base Color=@n105.Base Color
   Emissive Color=@n105.Base Color
   Opacity Mask=@n105.Base Color
   Specular=Multiply(A=@n105.Specular, B=ScalarParameter() {defaultValue=1, group=Mobile, parameterName=Mobile_SpecularMult})
   Roughness=Multiply(A=@n105.Roughness, B=ScalarParameter() {defaultValue=1, group=Mobile, parameterName=Mobile_RoughnessMult})
   Metallic=Multiply(A=@n105.Metallic, B=ScalarParameter() {defaultValue=1, group=Mobile, parameterName=Mobile_MetalMult})
   Normal=@n105.Base Color
   Subsurface Color=@n105.Base Color
  ) {attributeSetTypes [
    #1
    #5
    #9
    #2
    #4
    #3
    #6
    #10
   ]}
  n107 = MaterialFunctionCall!(In=ShadingPathSwitch(Default=@n102, Forward=@n102, Mobile=DistanceFieldsRenderingSwitch!(No=@n104, Yes=@n106))) {materialFunction=&/MF_ModBoxCustomization}
  MaterialFunctionCall!(In=MaterialFunctionCall!(In=@n107.Result) {materialFunction=&/MF_GameWideShared}.Result) {materialFunction=/Game/Packages/Fortress_SharedMaterials/Cinematics/CINE_MF_EnvPostFX}
  n108 = StaticBool() {value=true}
  MaterialFunctionCall!(
   IsStaticSnow
   UseAlternateStaticSnowMask
   AlternateStaticSnowMask
   "UseRoad&&OcclusionMask"
   IsForGrassMeshToPushDown
   AdditionalMaskToRemoveSnow
   IsFallingSnowFX
   UseMinimapCoords
   IncreaseSnowUsingSlope
   UseSmoothBorder
   LerpHighM
   TemporalHeightLerpContrast
   IsLobbyIsland
   UseBR-IslandMask=StaticBool()
   IsTree
   IsForRenderTarget
   UseSparkle
   UseSnowNormals=@n108
   UseAdditionalMaskToRemoveSnow
   IsTangentSpaceInput=@n108
   UseVertexNormalsForBasicSnow
   UseTangentSnowNormals=@n108
   Material=@n107.Result
  ) {
   desc=Realtime
   materialFunction=/Game/Athena/Apollo/Environments/Landscape/Materials/Biome/UniqueBiomeMaterialFunctions/MF_Apollo_Snow_Progression
  }
  NamedRerouteUsage()
  Reroute()
  Reroute()
 ]
}
