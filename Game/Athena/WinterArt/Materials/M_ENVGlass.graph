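# Unlit translucent, refractive winter-environment glass material with building hit-effect and shrink vertex offsets.
# Fresnel on an optional tiled normal drives opacity and a gradient-lookup iridescent film, alongside edge thickness tinting and depth-behind-surface opacity.
# Powered dot products of sun and reflection vectors fake specular glints over scene-color tint.
# Radial masks and a dirt texture height-blend in opaque, sun-lit grime; mobile skips this and scene-color sampling.
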Material /Game/Athena/WinterArt/Materials/M_ENVGlass
graph {
 properties {
  bUseMaterialAttributes=true
  blendMode=BLEND_Translucent
  pixelDepthOffsetMode=PDOM_Legacy
  refractionMethod=RM_IndexOfRefraction
  shadingModel=MSM_Unlit
  usedShadingModels=Unlit
 }
 nodes [
  n1 = TextureSample(UVs=Multiply(A=TextureCoordinate(), B=ScalarParameter() {defaultValue=3, parameterName=DirtTile})) {samplerType=SAMPLERTYPE_Masks, texture=/Game/Athena/WinterArt/Textures/T_DirtLeaks_M}
  n2 = DotProduct(A=AtmosphericLightVector(), B=ReflectionVectorWS())
  n3 = Power(Base=@n2, Exp=Constant() {r=512})
  n4 = Constant() {r=8}
  n5 = Multiply(A=Saturate(None=Power(Base=@n2, Exp=@n4)), B=Constant() {r=0.4})
  n6 = Add(A=Multiply(A=Constant() {r=155}, B=Saturate(None=@n3)), B=@n5)
  n7 = Multiply(A=Max(A=@n1.B, B=Constant() {r=0.8}), B=@n6)
  n8 = TextureCoordinate()
  n9 = ScalarParameter() {defaultValue=0.7102, parameterName=ScaleTiledNormal}
  n10 = MaterialFunctionCall(
   Normal=TextureSample(UVs=Multiply(A=@n8, B=@n9)) {samplerType=SAMPLERTYPE_Normal, texture=/Game/Gadgets/Assets/Telephone_Booth/Textures/T_BlindsNormalNoise_N}.RGB
   Flatness=ScalarParameter() {defaultValue=0.5, parameterName=FlattenNormal}
  ) {materialFunction=&FlattenNormal}
  n11 = StaticBoolParameter() {defaultValue=true, parameterName=UseNormalTexture}
  n12 = Fresnel(
   ExponentIn=ScalarParameter() {defaultValue=2, parameterName=Opacity-Fresnel Exponent}
   BaseReflectFractionIn=ScalarParameter() {defaultValue=0.5, parameterName=BaseTransparency}
   Normal=StaticSwitch(True=Transform(None=@n10.Result), False=VertexNormalWS(), Value=@n11)
  )
  n13 = TextureSample(UVs=@n12) {samplerSource=SSM_Clamp_WorldGroupSettings, texture=/Game/Global/Textures/Default/Masks_and_Gradients/3pxRGBImage}
  n14 = Multiply(A=Constant() {r=4}, B=@n13.RGB)
  n15 = DotProduct(A=AtmosphericLightVector(), B=ReflectionVectorWS())
  n16 = Constant() {r=55}
  n17 = Saturate(None=Multiply(A=Power(Base=@n15, Exp=@n16), B=Constant() {r=512}))
  n18 = Constant() {r=0.15}
  n19 = VectorParameter() {parameterName=Base Color, defaultValue=<rgba 0.849 0.8087 0.7053 0>}
  n20 = LinearInterpolate(A=SceneColor!(UVs).RGB, B=Multiply(A=@n18, B=@n19.RGB))
  n21 = ShadingPathSwitch(Default=@n20, Deferred=@n20, Forward=@n18, Mobile=@n18)
  n22 = Multiply(A=Fresnel() {baseReflectFraction=0, exponent=12}, B=ScalarParameter() {defaultValue=12, parameterName=GlassThickness})
  n23 = LinearInterpolate(
   A=Multiply(A=@n21, B=@n19.RGB)
   B=Multiply(A=@n21, B=Multiply(A=@n19.RGB, B=@n19.RGB)) {constB=0.5}
   Alpha=MaterialFunctionCall(In=Saturate(None=@n22), Contrast=Constant() {r=3}) {materialFunction=&CheapContrast}.Result
  ) {constB=0}
  n24 = Multiply(
   A=ScalarParameter() {defaultValue=0.35, group=Anodizing, parameterName=Film, sortPriority=1}
   B=Multiply(A=LinearInterpolate(A=@n13.RGB, B=@n14, Alpha=@n17), B=@n23)
  )
  n25 = MaterialFunctionCall!(Distance Bias, Distance Divisor=ScalarParameter() {defaultValue=155, parameterName=DepthTint}, Scene Depth, Pixel Depth) {materialFunction=/Game/Effects/Fort_Effects/Materials/_Functions/MF_DistanceBehindTranslucency_01}
  n26 = Add(A=@n12, B=@n22)
  n27 = Saturate(None=DotProduct(A=AtmosphericLightVector(), B=PixelNormalWS()))
  n28 = SetMaterialAttributes!(
   MaterialAttributes
   Base Color
   Metallic=Constant() {r=1}
   Specular
   Roughness
   Emissive Color=Add(A=ShadingPathSwitch(Default=@n7, Deferred=@n7, Forward=@n6, Mobile=@n6), B=Add(A=@n24, B=@n23))
   Opacity=Add(A=Add(A=@n25.Inverted, B=@n26), B=@n27)
   Opacity Mask
   Normal=StaticSwitch(True=@n10.Result, False=Constant3Vector() {constant=<rgba 0 0 1 0>}, Value=@n11)
   World Position Offset
   Refraction=ScalarParameter() {defaultValue=0.99, parameterName=Refraction, sliderMax=2}
  ) {attributeSetTypes [
    #1
    #2
    #3
    #4
    #5
    #6
    #7
    #8
    #9
    #10
   ]}
  n29 = Multiply(A=@n5) {constB=0.25}
  n30 = Multiply(A=VectorParameter() {parameterName=DirtColor, defaultValue=<rgba 0.1771 0.1663 0.1388 0>}.RGB, B=Max(A=@n27, B=Constant() {r=0.35}))
  n31 = MaterialFunctionCall(Radius=Constant() {r=0.5167}, Density=Constant() {r=4}) {materialFunction=&RadialGradientExponential}
  n32 = OneMinus(None=MaterialFunctionCall(Radius=Constant() {r=0.1237}, Density=Constant() {r=4}) {materialFunction=&RadialGradientExponential}.RadialGradientExponential)
  n33 = BlendMaterialAttributes(
   A=SetMaterialAttributes!(MaterialAttributes=@n28, Emissive Color=Add(A=@n29, B=@n30), Opacity=Constant() {r=1}) {attributeSetTypes [#5, #6]}
   B=@n28
   Alpha=MaterialFunctionCall(Transition Phase=Multiply(A=@n31.RadialGradientExponential, B=@n32), Height Texture=@n1.B) {materialFunction=&HeightLerp}.Alpha
  )
  SetMaterialAttributes!(
   MaterialAttributes=ShadingPathSwitch(Default=@n33, Deferred=@n33, Mobile=@n28)
   World Position Offset=MaterialFunctionCall!(
    WSDeformationCenter=VectorParameter() {bUseCustomPrimitiveData=true, group=z_code, parameterName=WSDeformationCenter, primitiveDataIndex=20}.RGB
    Code_DestructionAmount=ScalarParameter() {group=z_code, parameterName=Code_DestructionAmount}
    Code_UseUvForDamage
    Code_UseHitEffects=StaticBoolParameter() {defaultValue=true, group=z_code, parameterName=Code_UseHitEffects}
    Code_Emissive
    WSDamageRadius=ScalarParameter() {bUseCustomPrimitiveData=true, defaultValue=1, group=z_code, parameterName=WSDamageRadius, primitiveDataIndex=26}
    CrackDamageShader
    Diffuse
    ScalarModDamageCracks
    ScalarOffsetforDamageCracks
    ExposeLerpFunctionForInvertingTheHitDistanceCalc
    InvertDamageFalloff
    CrackWorldPosition
    WorldPosition
    DistanceCullFadeAnimationOffset=ScalarParameter() {bUseCustomPrimitiveData=true, group=z_code, parameterName=DistanceCullFadeAnimationOffset, primitiveDataIndex=12}
   ) {materialFunction=/Game/Packages/jl_tests/Fortnite_HitEffects}.Hit And Shrink WPO
  ) {attributeSetTypes [#9]}
 ]
}
