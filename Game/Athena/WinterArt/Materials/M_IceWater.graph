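# M_IceWater is an ice water material for the Athena WinterArt set, likely a frozen lake surface.
# It delegates its logic to a single MF_FrozenLake material function call, using material attributes with Default Lit shading.
# Tangent-space normals are off, and the material is flagged for instanced static meshes and uses legacy pixel depth offset.
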Material /Game/Athena/WinterArt/Materials/M_IceWater
graph {
 properties {
  bTangentSpaceNormal=false
  bUseMaterialAttributes=true
  bUsedWithInstancedStaticMeshes=true
  pixelDepthOffsetMode=PDOM_Legacy
  usedShadingModels=Default Lit
 }
 nodes [MaterialFunctionCall() {materialFunction=&/MF_FrozenLake}]
}
