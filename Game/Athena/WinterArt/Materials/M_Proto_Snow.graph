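# Prototype snow material for Athena winter art, using the Subsurface shading model.
# A tiled smooth texture's alpha blends two scalar constants, while two other lerps, one mixing white and blue snow color parameters, have no alpha input.
# An MF_SharedWinter function call, a near-white constant and a dark blue vector parameter sit alongside.
# The graph shows no links from these pieces to each other or to material outputs.
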Material /Game/Athena/WinterArt/Materials/M_Proto_Snow
graph {
 properties {pixelDepthOffsetMode=PDOM_Legacy, shadingModel=MSM_Subsurface, usedShadingModels=Subsurface}
 nodes [
  Constant3Vector() {constant=<rgba 0.9844 0.9844 0.9844 0>}
  LinearInterpolate(
   A=Constant() {r=1}
   B=Constant() {r=0.6}
   Alpha=TextureSample(UVs=TextureCoordinate() {uTiling=0.4, vTiling=0.4}) {texture=&T_Smooth_tile}.A
  )
  LinearInterpolate(
   A=VectorParameter() {parameterName=Snow, defaultValue=<rgba 0.7969 0.7969 0.7969 0>}.RGB
   B=VectorParameter() {parameterName=Snow 2, defaultValue=<rgba 0.2335 0.4858 0.7865 1>}.RGB
  )
  LinearInterpolate(A=Constant() {r=0.25}, B=Constant() {r=0.5})
  MaterialFunctionCall() {materialFunction=&/MF_SharedWinter}
  VectorParameter() {defaultValue=<rgba 0.067 0.1082 0.2083 0>}
 ]
}
