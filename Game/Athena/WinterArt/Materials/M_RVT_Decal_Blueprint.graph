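# Translucent decal-style surface material for instanced meshes that writes into the landscape runtime virtual texture, including heightmap-driven displacement.
# World-aligned or UV-scaled textures feed color, roughness, specular and detailed normals, with diffuse-alpha height driving two-tone variation and optional macro-detail biome tinting.
# A radial gradient height-blended with texture height shapes opacity, unless a plain opacity switch overrides it.
# A static switch substitutes a flat-color puddle with a noise-broken mask.
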Material /Game/Athena/WinterArt/Materials/M_RVT_Decal_Blueprint
graph {
 properties {
  bAutomaticallySetUsageInEditor=false
  bCastRayTracedShadows=false
  bTangentSpaceNormal=false
  bUseMaterialAttributes=true
  bUsedWithInstancedStaticMeshes=true
  blendMode=BLEND_Translucent
  pixelDepthOffsetMode=PDOM_Legacy
  usedShadingModels=Default Lit
 }
 nodes [
  Constant() {r=1}
  n1 = MaterialFunctionCall(Radius=Constant() {r=0.5}, Density=ScalarParameter() {defaultValue=4, parameterName=MaskDensity}) {materialFunction=&RadialGradientExponential}
  n2 = ComponentMask(None=WorldPosition().XYZ) {b=true}
  n3 = Sine(None=Frac(None=Divide(A=@n2) {constB=1024}))
  n4 = WorldPosition()
  n5 = TextureSample(
   UVs=Divide(A=ComponentMask(None=@n4.XYZ) {g=true, r=true}, B=ScalarParameter() {defaultValue=1000, parameterName=ScaleTextures})
   Tex=TextureObjectParameter() {parameterName=Diffuse, texture=/Game/Environments/Apollo/Sets/Terrain/Textures/T_Apollo_Cliff_D_02}
  ) {samplerType=SAMPLERTYPE_Grayscale, texture=&LowResBlurredNoise}
  NamedRerouteDeclaration(None=MaterialFunctionCall(
    Transition Phase=Multiply(A=@n1.RadialGradientExponential, B=@n3)
    Height Texture=OneMinus(None=@n5.A)
    Contrast=ScalarParameter() {defaultValue=1, parameterName=HeightMapContrast}
   ) {materialFunction=&HeightLerp}.Alpha) {name=Puddle Mask, nodeColor=<rgba 0.9219 1 0 1>}
  n6 = VectorParameter() {parameterName=DecalColor, defaultValue=<rgba 0.5 0.5 0.5 0>}
  n7 = ScalarParameter() {defaultValue=1, parameterName=RoughnessTweak}
  n8 = ScalarParameter() {defaultValue=1, parameterName=SpecularTweak}
  n9 = ScalarParameter() {defaultValue=1, parameterName=Overall Opacity}
  n10 = NamedRerouteUsage()
  NamedRerouteDeclaration(None=SetMaterialAttributes!(
    MaterialAttributes=SetMaterialAttributes!(MaterialAttributes, Base Color=@n6.RGB, Roughness=@n7, Specular=@n8) {attributeSetTypes [#1, #2, #3]}
    Opacity=Multiply(A=@n9, B=@n10)
   ) {attributeSetTypes [#4]}) {name=Puddle, nodeColor=<rgba 0 0.6953 1 1>}
  n11 = MaterialFunctionCall(Float4=MaterialFunctionCall!(BiomeMask=Constant() {r=1}, IsForStaticMesh) {materialFunction=/Game/Athena/Helios/Environments/Landscape/Materials/MaterialFunctions_Utilities/MF_OpenDesert_MacroDetail}.SedimentLayers) {materialFunction=&BreakOutFloat4Components}
  NamedRerouteDeclaration(None=@n11.R) {name=OD_MD_Layer1, nodeColor=<rgba 0.5 0 1 1>}
  NamedRerouteDeclaration(None=@n11.G) {name=OD_MD_Layer2, nodeColor=<rgba 0.5 0 1 1>}
  n12 = MaterialFunctionCall(Texture Scale=ScalarParameter() {defaultValue=1000, parameterName=ScaleTextures}) {materialFunction=&ScaleUVsByCenter}
  n13 = TextureObjectParameter() {parameterName=Specular, samplerType=SAMPLERTYPE_Masks, texture=/Game/Environments/Apollo/Sets/Terrain/Textures/T_Apollo_Cliff_S_02}
  n14 = ScalarParameter() {defaultValue=1000, parameterName=ScaleTextures}
  n15 = MaterialFunctionCall(Float3=StaticSwitchParameter(
    True=TextureSample(UVs=@n12.UVs, Tex=@n13) {samplerType=SAMPLERTYPE_Masks}.RGB
    False=MaterialFunctionCall(TextureObject=@n13, TextureSize=@n14) {materialFunction=&WorldAlignedTexture}.Z Texture
   ) {parameterName=UseUVs}) {materialFunction=&BreakOutFloat3Components}
  NamedRerouteDeclaration(None=@n15.R) {name=Spec, nodeColor=<rgba 0 0.6953 1 1>}
  NamedRerouteDeclaration(None=@n11.B) {name=OD_MD_Layer3, nodeColor=<rgba 0.5 0 1 1>}
  NamedRerouteDeclaration(None=@n11.A) {name=OD_MD_Layer4, nodeColor=<rgba 0.5 0 1 1>}
  NamedRerouteDeclaration(None=@n15.B) {name=Rough, nodeColor=<rgba 0 0.6953 1 1>}
  n16 = TextureObjectParameter() {
   group=Textures
   parameterName=Normal
   samplerType=SAMPLERTYPE_Normal
   texture=/Game/Environments/Apollo/Sets/Terrain/Textures/T_Apollo_Cliff_N_02
  }
  NamedRerouteDeclaration(None=StaticSwitchParameter(
    True=TextureSample(UVs=@n12.UVs, Tex=@n16) {samplerType=SAMPLERTYPE_Normal}.RGB
    False=MaterialFunctionCall(TextureObject=@n16, TextureSize=@n14) {materialFunction=&WorldAlignedNormal}.Z Texture
   ) {parameterName=UseUVs}) {name=Normals, nodeColor=<rgba 0 0.6953 1 1>}
  n17 = TextureObjectParameter() {parameterName=Diffuse, texture=&T_Apollo_Cliff_D_02}
  n18 = StaticSwitchParameter(
   True=TextureSample(UVs=@n12.UVs, Tex=@n17).RGBA
   False=MaterialFunctionCall(TextureObject=@n17, TextureSize=@n14, Export Float 4=StaticBool() {value=true}) {materialFunction=&WorldAlignedTexture}.Z Texture
  ) {parameterName=UseUVs}
  NamedRerouteDeclaration(None=ComponentMask(None=@n18) {a=true}) {name=DiffuseHeight, nodeColor=<rgba 0 0.6953 1 1>}
  NamedRerouteDeclaration(None=ComponentMask(None=@n18) {b=true, g=true, r=true}) {name=Diffuse, nodeColor=<rgba 0 0.6953 1 1>}
  n19 = TextureObjectParameter() {
   group=Textures
   parameterName=DetailsNormal
   samplerType=SAMPLERTYPE_Normal
   texture=&T_Apollo_Cliff_N_02
  }
  NamedRerouteDeclaration(None=MaterialFunctionCall(TextureObject=@n19, TextureSize=ScalarParameter() {defaultValue=255, parameterName=DetailsScale}) {materialFunction=&WorldAlignedNormal}.Z Texture) {name=DetailNormals, nodeColor=<rgba 0 0.6953 1 1>}
  n20 = NamedRerouteUsage()
  n21 = MaterialFunctionCall(
   A=VectorParameter() {parameterName=DecalColor2, defaultValue=<rgba 0.5 0.5 0.5 0>}.RGB
   B=VectorParameter() {parameterName=DecalColor, defaultValue=<rgba 0.5 0.5 0.5 0>}.RGB
   Transition Phase=ScalarParameter() {defaultValue=0.5, parameterName="ColorTransition "}
   Height Texture=NamedRerouteUsage()
   Contrast=ScalarParameter() {parameterName=ColorContrast}
  ) {materialFunction=&HeightLerp}
  n22 = StaticBoolParameter() {defaultValue=true, parameterName=UseHeightChannelForVariation}
  NamedRerouteDeclaration(None=StaticSwitch(
    True=MaterialFunctionCall(Base=@n20, Blend=@n21.Results) {materialFunction=&Blend_Overlay}.Result
    False=@n20
    Value=@n22
   )) {name=BaseColor, nodeColor=<rgba 0 0.6953 1 1>}
  n23 = ScalarParameter() {parameterName=DetailsValue}
  n24 = AppendVector(A=@n23, B=@n23)
  n25 = Constant()
  n26 = Constant2Vector() {g=1, r=1}
  n27 = Multiply(A=@n26, B=ScalarParameter() {defaultValue=1, parameterName=Nomal Map Value})
  n28 = StaticSwitch(True=@n21.Alpha, False=Constant(), Value=@n22)
  n29 = Multiply(A=NamedRerouteUsage(), B=AppendVector(A=LinearInterpolate(A=@n26, B=@n27, Alpha=@n28), B=Constant() {r=1}))
  n30 = Add(A=Multiply(A=NamedRerouteUsage(), B=AppendVector(A=@n24, B=@n25)), B=@n29)
  NamedRerouteDeclaration(None=StaticSwitch(
    True=LinearInterpolate(A=@n30, B=@n29, Alpha=@n28)
    False=@n29
    Value=StaticBoolParameter() {defaultValue=true, parameterName=UseDetailNormals}
   )) {name=FinalNormals, nodeColor=<rgba 0 0.6953 1 1>}
  n31 = OneMinus(None=NamedRerouteUsage())
  n32 = ScalarParameter() {defaultValue=0.5, group=OpenDesert_MacroDetail, parameterName=OD_MD_LayerBlendContrast, sortPriority=501}
  n33 = LinearInterpolate(
   A=VectorParameter() {group=BiomeColors, parameterName=OD_MD_Color0, defaultValue=<rgba 0.45 0.1445 0.045 1>}.RGB
   B=VectorParameter() {group=BiomeColors, parameterName=OD_MD_Color1, defaultValue=<rgba 0.6 0.1927 0.06 1>}.RGB
   Alpha=MaterialFunctionCall(Transition Phase=NamedRerouteUsage(), Height Texture=@n31, Contrast=@n32) {materialFunction=&HeightLerp}.Alpha
  )
  n34 = VectorParameter() {group=BiomeColors, parameterName=OD_MD_Color2, defaultValue=<rgba 0.65 0.2823 0.1625 1>}
  n35 = MaterialFunctionCall(Transition Phase=NamedRerouteUsage(), Height Texture=@n31, Contrast=@n32) {materialFunction=&HeightLerp}
  n36 = NamedRerouteUsage()
  n37 = LinearInterpolate(
   A=LinearInterpolate(
    A=LinearInterpolate(A=@n33, B=@n34.RGB, Alpha=@n35.Alpha)
    B=VectorParameter() {group=BiomeColors, parameterName=OD_MD_Color3, defaultValue=<rgba 0.8 0.275 0.104 1>}.RGB
    Alpha=MaterialFunctionCall(Transition Phase=@n36, Height Texture=@n31, Contrast=@n32) {materialFunction=&HeightLerp}.Alpha
   )
   B=VectorParameter() {group=BiomeColors, parameterName=OD_MD_Color4, defaultValue=<rgba 0.86 0.354 0.1892 1>}.RGB
   Alpha=MaterialFunctionCall(Transition Phase=NamedRerouteUsage(), Height Texture=@n31, Contrast=@n32) {materialFunction=&HeightLerp}.Alpha
  )
  n38 = NamedRerouteUsage()
  n39 = NamedRerouteUsage()
  n40 = ScalarParameter() {defaultValue=1, parameterName=RoughnessTweak2}
  n41 = ScalarParameter() {defaultValue=1, parameterName=RoughnessTweak}
  n42 = NamedRerouteUsage()
  n43 = ScalarParameter() {defaultValue=1, parameterName=SpecularTweak2}
  n44 = ScalarParameter() {defaultValue=1, parameterName=SpecularTweak}
  n45 = ScalarParameter() {defaultValue=1, parameterName=Opacity, sliderMax=1}
  n46 = NamedRerouteUsage()
  n47 = ScalarParameter() {defaultValue=0.5, parameterName=BlendIn}
  n48 = Multiply(
   A=Max(A=@n28, B=OneMinus(None=@n47))
   B=MaterialFunctionCall(Radius=Constant() {r=0.5}, Density=ScalarParameter() {defaultValue=4, parameterName=MaskDensity}) {materialFunction=&RadialGradientExponential}.RadialGradientExponential
  )
  n49 = LinearInterpolate(A=Constant() {r=-1}, B=Constant() {r=1}, Alpha=NamedRerouteUsage())
  n50 = ScalarParameter() {parameterName=Blend In Contrast}
  n51 = SetMaterialAttributes!(
   MaterialAttributes
   Base Color=StaticSwitch(True=Multiply(A=@n37, B=@n38), False=@n38, Value=StaticBoolParameter() {parameterName=UseRustBucketSand})
   Roughness=LinearInterpolate(A=Multiply(A=@n39, B=@n40), B=Multiply(A=@n39, B=@n41), Alpha=@n28)
   Specular=LinearInterpolate(A=Multiply(A=@n42, B=@n43), B=Multiply(A=@n42, B=@n44), Alpha=@n28)
   Normal=NamedRerouteUsage()
   Opacity=StaticSwitchParameter(
    True=Multiply(A=@n45, B=@n46)
    False=MaterialFunctionCall(Transition Phase=@n48, Height Texture=@n49, Contrast=@n50) {materialFunction=&HeightLerp}.Alpha
   ) {parameterName="Use Opacity "}
  ) {attributeSetTypes [
    #1
    #2
    #3
    #5
    #4
   ]}
  n52 = GetMaterialAttributes(None=StaticSwitchParameter(True=NamedRerouteUsage(), False=@n51) {parameterName=Puddle}) {attributeGetTypes [#5]}
  n53 = SetMaterialAttributes!(MaterialAttributes=@n52.MaterialAttributes, Normal=Transform(None=@n52.Normal)) {attributeSetTypes [#5]}
  n54 = GetMaterialAttributes(None=@n53) {attributeGetTypes [
    #1
    #3
    #2
    #5
    #4
   ]}
  n55 = ScalarParameter() {defaultValue=0.15, parameterName=Mobile Desaturation}
  n56 = Multiply(
   A=Power(Base=NamedRerouteUsage(), Exp=ScalarParameter() {defaultValue=1, group=Displacement, parameterName=Displacement_Heightmap_Power})
   B=ScalarParameter() {defaultValue=1, group=Displacement, parameterName=Displacement_Heightmap_Mod}
  )
  RuntimeVirtualTextureOutput!(
   BaseColor=ShadingPathSwitch(Default=@n54.Base Color, Mobile=Desaturation(None=@n54.Base Color, Fraction=@n55))
   Specular=@n54.Specular
   Roughness=@n54.Specular
   Normal=@n54.Specular
   WorldHeight
   Opacity=@n54.Specular
   Mask
   Displacement=LinearInterpolate(
    B=Saturate(None=@n56)
    Alpha=ScalarParameter() {defaultValue=1, group=Displacement, parameterName=Displacement_Heightmap_Amount, sortPriority=-500}
   ) {constA=0.5}
   Mask4
  )
  n57 = VirtualTextureFeatureSwitch!(
   No=@n53
   Yes=MaterialFunctionCall!(RVT=@n54.MaterialAttributes, ByPass RVT=@n53) {materialFunction=/Game/Athena/Apollo/Maps/Landscape/VT/MaterialFunctions/EnableRVT}.Result
  )
  ShadingPathSwitch(Default=@n57, Deferred=@n57, Forward=@n57, Mobile=@n57)
 ]
}
