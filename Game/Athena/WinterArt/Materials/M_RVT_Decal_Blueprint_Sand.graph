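# Translucent decal material that overlays a tintable texture on sand and outputs to Runtime Virtual Texture, apparently for terrain markings.
# World-aligned sand textures drive the normal and an erosion mask that multiplies diffuse alpha and opacity, breaking up the decal.
# Sand-derived values slightly distort the diffuse UVs, whose red channel also blends roughness and specular.
# A feature switch routes the attributes through EnableRVT when virtual texturing is supported, otherwise using them directly.
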Material /Game/Athena/WinterArt/Materials/M_RVT_Decal_Blueprint_Sand
graph {
 properties {
  bAutomaticallySetUsageInEditor=false
  bCastRayTracedShadows=false
  bTangentSpaceNormal=false
  bUseMaterialAttributes=true
  bUsedWithInstancedStaticMeshes=true
  blendMode=BLEND_Translucent
  pixelDepthOffsetMode=PDOM_Legacy
  usedShadingModels=Default Lit
 }
 nodes [
  n1 = TextureObject() {samplerType=SAMPLERTYPE_Masks, texture=/Game/Environments/Apollo/Sets/Terrain/Textures/T_Apollo_Sand_S_01}
  n2 = ScalarParameter() {defaultValue=512, parameterName=Sand Details Scale}
  NamedRerouteDeclaration(None=ComponentMask(None=MaterialFunctionCall(TextureObject=@n1, TextureSize=@n2) {materialFunction=/Game/Weapons/WeaponSkins/Wraps/MaterialFunctions/WorldAlignedTexture}.Z Texture) {b=true}) {name=SandDots, nodeColor=<rgba 0 0.5078 1 1>}
  n3 = VectorParameter() {parameterName=DecalColor, defaultValue=<rgba 1 1 1 1>}
  n4 = Min(A=NamedRerouteUsage(), B=NamedRerouteUsage())
  n5 = ScalarParameter() {defaultValue=0.01, parameterName=Distort by Sand}
  n6 = TextureObjectParameter() {parameterName=Diffuse, texture=/Game/Environments/Apollo/Sets/Terrain/Textures/T_Apollo_Cliff_D_02}
  n7 = TextureSample(UVs=Add(A=Multiply(A=@n4, B=@n5), B=TextureCoordinate()), Tex=@n6)
  n8 = ComponentMask(None=@n7.RGBA) {b=true, g=true, r=true}
  n9 = NamedRerouteUsage()
  n10 = Multiply(A=@n9, B=Constant() {r=0.85})
  n11 = Saturate(None=Round(None=@n7.R))
  n12 = LinearInterpolate(
   A=ScalarParameter() {defaultValue=0.1967, parameterName=SpecularTweak2}
   B=ScalarParameter() {defaultValue=0.5, parameterName=SpecularTweak}
   Alpha=@n7.R
  )
  n13 = OneMinus(None=NamedRerouteUsage())
  n14 = ScalarParameter() {defaultValue=1, parameterName=Opacity, sliderMax=1}
  n15 = ComponentMask(None=TextureSample(Tex=@n6).RGBA) {a=true}
  n16 = ScalarParameter() {defaultValue=1.1, parameterName=Erosion}
  n17 = MaterialFunctionCall(
   Export Float 4=StaticBool() {value=true}
   TextureObject=TextureObject() {texture=/Game/Environments/Apollo/Sets/Terrain/Textures/T_Apollo_Sand_D_01}
   TextureSize=@n2
  ) {materialFunction=&WorldAlignedTexture}
  n18 = Saturate(None=LinearInterpolate(A=Multiply(B=@n16) {constA=-1}, B=@n16, Alpha=ComponentMask(None=@n17.Z Texture) {a=true}))
  n19 = MaterialFunctionCall(
   TextureObject=TextureObject() {samplerType=SAMPLERTYPE_Normal, texture=/Game/Environments/Apollo/Sets/Terrain/Textures/T_Apollo_Sand_N_01}
   TextureSize=@n2
  ) {materialFunction=&WorldAlignedNormal}
  n20 = Constant3Vector() {constant=<rgba 2 2 1 0>}
  n21 = SetMaterialAttributes!(
   MaterialAttributes=SetMaterialAttributes!(
    MaterialAttributes
    Base Color=Multiply(A=@n3.RGB, B=@n8)
    Roughness=LinearInterpolate(A=@n10, B=@n9, Alpha=@n11)
    Specular=Add(A=@n12, B=@n13)
   ) {attributeSetTypes [#1, #2, #3]}
   Opacity=Multiply(A=Multiply(A=@n14, B=@n15), B=NamedRerouteDeclaration(None=@n18) {name=Sand A, nodeColor=<rgba 0 0.5312 1 1>})
   Normal=Transform(None=Multiply(A=@n19.Z Texture, B=@n20))
  ) {attributeSetTypes [#4, #5]}
  n22 = GetMaterialAttributes(None=@n21) {attributeGetTypes [
    #1
    #3
    #2
    #5
    #4
   ]}
  RuntimeVirtualTextureOutput!(
   BaseColor=@n22.Base Color
   Specular=@n22.Base Color
   Roughness=@n22.Base Color
   Normal=@n22.Base Color
   WorldHeight
   Opacity=@n22.Base Color
   Mask
   Displacement
   Mask4
  )
  n23 = VirtualTextureFeatureSwitch!(
   No=@n21
   Yes=MaterialFunctionCall!(RVT=@n22.MaterialAttributes, ByPass RVT=@n21) {materialFunction=/Game/Athena/Apollo/Maps/Landscape/VT/MaterialFunctions/EnableRVT}.Result
  )
  ShadingPathSwitch(Default=@n23, Deferred=@n23, Forward=@n23, Mobile=@n23)
 ]
}
