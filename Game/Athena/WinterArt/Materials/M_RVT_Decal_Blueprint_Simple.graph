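# M_RVT_Decal_Blueprint_Simple is a translucent Default Lit decal material for runtime virtual texture output, usable on instanced static meshes.
# A tintable diffuse texture supplies base color, and its alpha times an Opacity parameter sets opacity.
# Specular and roughness come from a mask texture's red and blue channels, each blending two tweak multipliers.
# The attributes feed an RVT output, with a feature switch routing through EnableRVT when virtual texturing is supported.
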Material /Game/Athena/WinterArt/Materials/M_RVT_Decal_Blueprint_Simple
graph {
 properties {
  bAutomaticallySetUsageInEditor=false
  bCastRayTracedShadows=false
  bTangentSpaceNormal=false
  bUseMaterialAttributes=true
  bUsedWithInstancedStaticMeshes=true
  blendMode=BLEND_Translucent
  pixelDepthOffsetMode=PDOM_Legacy
  usedShadingModels=Default Lit
 }
 nodes [
  n1 = VectorParameter() {parameterName=DecalColor, defaultValue=<rgba 1 1 1 1>}
  n2 = TextureSample(Tex=TextureObjectParameter() {parameterName=Diffuse, texture=/Game/Environments/Apollo/Sets/Terrain/Textures/T_Apollo_Cliff_D_02})
  n3 = ComponentMask(None=@n2.RGBA) {b=true, g=true, r=true}
  n4 = TextureObjectParameter() {parameterName=Specular, samplerType=SAMPLERTYPE_Masks, texture=/Game/Environments/Apollo/Sets/Terrain/Textures/T_Apollo_Cliff_S_02}
  n5 = MaterialFunctionCall(Float3=TextureSample(Tex=@n4) {samplerType=SAMPLERTYPE_Masks}.RGB) {materialFunction=&BreakOutFloat3Components}
  n6 = Multiply(A=@n5.B, B=ScalarParameter() {defaultValue=1, parameterName=RoughnessTweak2})
  n7 = Multiply(A=@n5.B, B=ScalarParameter() {defaultValue=1, parameterName=RoughnessTweak})
  n8 = Multiply(A=@n5.R, B=ScalarParameter() {defaultValue=1, parameterName=SpecularTweak2})
  n9 = Multiply(A=@n5.R, B=ScalarParameter() {defaultValue=1, parameterName=SpecularTweak})
  n10 = SetMaterialAttributes!(
   MaterialAttributes=SetMaterialAttributes!(
    MaterialAttributes
    Base Color=Multiply(A=@n1.RGB, B=@n3)
    Roughness=LinearInterpolate(A=@n6, B=@n7)
    Specular=LinearInterpolate(A=@n8, B=@n9)
   ) {attributeSetTypes [#1, #2, #3]}
   Opacity=Multiply(A=ScalarParameter() {defaultValue=1, parameterName=Opacity, sliderMax=1}, B=ComponentMask(None=@n2.RGBA) {a=true})
  ) {attributeSetTypes [#4]}
  n11 = GetMaterialAttributes(None=@n10) {attributeGetTypes [
    #1
    #3
    #2
    #5
    #4
   ]}
  RuntimeVirtualTextureOutput!(
   BaseColor=@n11.Base Color
   Specular=@n11.Base Color
   Roughness=@n11.Base Color
   Normal
   WorldHeight
   Opacity=@n11.Base Color
   Mask
   Displacement
   Mask4
  )
  n12 = VirtualTextureFeatureSwitch!(
   No=@n10
   Yes=MaterialFunctionCall!(RVT=@n11.MaterialAttributes, ByPass RVT=@n10) {materialFunction=/Game/Athena/Apollo/Maps/Landscape/VT/MaterialFunctions/EnableRVT}.Result
  )
  ShadingPathSwitch(Default=@n12, Deferred=@n12, Forward=@n12, Mobile=@n12)
 ]
}
