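# MF_DMTexture is a WinterArt material function that tints a packed three-channel DM input with three color inputs.
# The DM green channel blends Color1 to Color2, and the blue channel then blends that result toward Color3.
# The red channel multiplies the blended color as a shading or intensity term before output.
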MaterialFunction /Game/Athena/WinterArt/Materials/MF_DMTexture
graph {nodes [
  n1 = MaterialFunctionCall(Float3=FunctionInput() {inputName=DM, sortPriority=3}) {materialFunction=&BreakOutFloat3Components}
  n2 = LinearInterpolate(A=FunctionInput() {inputName=Color1, sortPriority=3}, B=FunctionInput() {inputName=Color2, sortPriority=4}, Alpha=@n1.G)
  n3 = FunctionInput() {inputName=Color3, sortPriority=5}
  FunctionOutput(None=Multiply(A=LinearInterpolate(A=@n2, B=@n3, Alpha=@n1.B), B=@n1.R))
 ]}
