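# Material function that layers a printed decal or graffiti over input material attributes, behind static switches.
# The decal samples mesh UVs, world or local projections, or a 2x2 flipbook randomized by actor position, with layer and axis masks.
# It modulates the base by default, or blends via height lerp or PBR, with optional time-of-day emissive and height masking.
# The GraffitiBP switch swaps in a flipbook graffiti blend with an optional extra layer.
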MaterialFunction /Game/Athena/WinterArt/Materials/MF_DecalPrint
graph {
 properties {description="This function uses a black and white texture and places it in front of the Texture Coordinate 2 UVs of a mesh, where Black is seen and White is masked out."}
 nodes [
  n1 = WorldPosition() {worldPositionShaderOffset=WPT_ExcludeAllShaderOffsets}
  n2 = StaticSwitchParameter(True=TransformPosition(None=@n1.XYZ) {transformSourceType=TRANSFORMPOSSOURCE_World}, False=@n1.XYZ) {group=Decal, parameterName=Use Decal WS LocalTransform}
  n3 = VectorParameter() {group=Decal, parameterName=Decal Bake Location, defaultValue=<rgba 0 0 0 1>}
  n4 = StaticSwitchParameter(
   True=@n3.RGB
   False=MaterialProxyReplace(Realtime=VectorParameter() {group=Decal, parameterName=Decal Location World}.RGB, MaterialProxy=@n3.RGB)
  ) {group=Decal, parameterName=Decal WS Bake Location debug, sortPriority=1}
  n5 = ScalarParameter() {defaultValue=1, group=Decal, parameterName=Decal_TileXY, sortPriority=16}
  n6 = AppendVector(A=AppendVector(A=@n5, B=@n5), B=ScalarParameter() {defaultValue=1, group=Decal, parameterName=Decal_Tile Z, sortPriority=16})
  n7 = Constant3Vector() {constant=<rgba -512 -512 -384 0>}
  n8 = Multiply(
   A=Divide(A=Subtract(A=@n2, B=@n4), B=Multiply(A=@n6, B=@n7))
   B=VectorParameter() {group=Decal, parameterName=DecalPerAxisScale, defaultValue=<rgba 1 1 1 0>}.RGB
  )
  n9 = StaticComponentMaskParameter(None=@n8) {defaultG=true, defaultR=true, group=Decal, parameterName=WorldPaceAxisProjectionBake}
  n10 = StaticSwitchParameter(
   True=@n9
   False=MaterialProxyReplace(
    Realtime=StaticComponentMaskParameter(None=@n8) {defaultG=true, defaultR=true, group=Decal, parameterName=WorldPaceAxisProjection}
    MaterialProxy=@n9
   )
  ) {group=Decal, parameterName=Decal WS Bake Location debug, sortPriority=1}
  Floor(None=@n10)
  n11 = NamedRerouteUsage()
  n12 = GetMaterialAttributes(None=@n11) {attributeGetTypes [#1, #2]}
  n13 = TextureCoordinate()
  n14 = ScalarParameter() {defaultValue=1, group=Decal, parameterName=GraffitiExtraScale}
  n15 = StaticSwitchParameter(True=NamedRerouteUsage(), False=TextureSample(UVs=Multiply(A=@n13, B=@n14), Tex=NamedRerouteUsage()).RGBA) {desc=Default Flipbook, group=Decal, parameterName=Graffiti+WorldZ, sortPriority=-2}
  n16 = Multiply(A=@n12.BaseColor, B=ComponentMask(None=@n15) {b=true, g=true, r=true})
  n17 = Add(A=ScalarParameter() {defaultValue=1, desc="From 1 to 4 ", group=Decal, parameterName=Graffiti_FlipBookSet, sortPriority=13}) {constB=-1}
  n18 = Constant() {r=2}
  n19 = StaticSwitchParameter(
   True=StaticSwitchParameter(True=TextureCoordinate() {coordinateIndex=4}, False=TextureCoordinate() {coordinateIndex=3}) {desc=Uses your base mesh UV's, group=Decal, parameterName=Graffiti UV4, sortPriority=-1}
   False=TextureCoordinate() {coordinateIndex=2}
  ) {desc=Uses your base mesh UV's, group=Decal, parameterName=Graffiti UV3, sortPriority=-1}
  n20 = TextureCoordinate()
  n21 = MaterialFunctionCall(
   "Animation  Phase (0-1)"=Saturate(None=Divide(A=@n17) {constB=4})
   Number of Rows=@n18
   Number of Columns=@n18
   Texture=TextureObjectParameter() {group=Decal, parameterName=Graffiti, texture=/Game/Athena/WinterArt/Textures/T-GraffitiSample_D}
   UVs=StaticSwitchParameter(
    True=NamedRerouteUsage()
    False=StaticSwitchParameter(True=@n19, False=@n20) {desc=Uses your base mesh UV's, group=Decal, parameterName=Graffiti UV 2, sortPriority=-1}
   ) {desc=Uses your base mesh UV's, group=Decal, parameterName=Graffiti Proc UV, sortPriority=-1}
  ) {materialFunction=&FlipBook}
  n22 = StaticSwitchParameter(True=LinearInterpolate(A=@n16, B=@n21.Result, Alpha=@n21.Result), False=@n21.Result) {desc=Mult on top of all, group=Decal, parameterName=Graffiti+ExtraLayer, sortPriority=-2}
  n23 = ScalarParameter() {defaultValue=0.65, group=Decal, parameterName=Graffiti Rough, sortPriority=13}
  n24 = StaticSwitchParameter(
   True=LinearInterpolate(A=ScalarParameter() {defaultValue=1, parameterName=Graffiti Extra Layer Rough, sortPriority=13}, B=@n23, Alpha=@n21.Alpha) {constA=1.5}
   False=@n23
  ) {desc=Mult on top of all, group=Decal, parameterName=Graffiti+ExtraLayer, sortPriority=-2}
  n25 = Multiply(A=@n21.Result, B=@n12.BaseColor)
  n26 = StaticSwitchParameter(True=LinearInterpolate(A=@n16, B=@n25, Alpha=@n21.Alpha), False=@n25) {desc=Mult on top of all, group=Decal, parameterName=Graffiti+ExtraLayer, sortPriority=-2}
  n27 = Multiply(A=@n12.Roughness, B=ScalarParameter() {defaultValue=0.65, group=Decal, parameterName=Graffiti Rough, sortPriority=13})
  n28 = StaticSwitchParameter(
   True=LinearInterpolate(A=ScalarParameter() {defaultValue=1, parameterName=Graffiti Extra Layer Rough, sortPriority=13}, B=@n27, Alpha=@n21.Alpha) {constA=1.5}
   False=@n27
  ) {desc=Mult on top of all, group=Decal, parameterName=Graffiti+ExtraLayer, sortPriority=-2}
  n29 = ComponentMask(None=@n15) {a=true}
  n30 = StaticSwitchParameter(True=Saturate(None=Add(A=@n29, B=@n21.Alpha)), False=@n21.Alpha) {desc=Mult on top of all, group=Decal, parameterName=Graffiti+ExtraLayer, sortPriority=-2}
  n31 = NamedRerouteUsage()
  n32 = BlendMaterialAttributes(
   A=@n11
   B=StaticSwitchParameter(
    True=SetMaterialAttributes!(MaterialAttributes=@n12.MaterialAttributes, BaseColor=@n22, Roughness=@n24) {attributeSetTypes [#1, #2]}
    False=SetMaterialAttributes(MaterialAttributes=@n12.MaterialAttributes, BaseColor=@n26, Roughness=@n28) {attributeSetTypes [#1, #2]}
   ) {group=Decal, parameterName=Graffiti Lerp, sortPriority=-2}
   Alpha=Multiply(A=NamedRerouteUsage(), B=Multiply(A=@n30, B=@n31))
  ) {vertexAttributeBlendType=UseA}
  n33 = TextureObjectParameter() {group=Decal, parameterName=Decal, texture=/Game/Global/Textures/Default/Blanks/T_WhiteUserInterface}
  n34 = NamedRerouteDeclaration(None=@n33) {name=DecalTextureObject, nodeColor=<rgba 1 0 0.4922 1>}
  n35 = ScalarParameter() {defaultValue=1, group=Decal, parameterName=Decal_TileX, sortPriority=12}
  n36 = ScalarParameter() {defaultValue=1, group=Decal, parameterName=Decal_TileY, sortPriority=12}
  n37 = TextureCoordinate()
  n38 = StaticSwitchParameter(
   True=StaticSwitchParameter(True=TextureCoordinate() {coordinateIndex=4}, False=TextureCoordinate() {coordinateIndex=3}) {desc=Uses your base mesh UV's, group=Decal, parameterName=Use Decal UV4, sortPriority=-1}
   False=TextureCoordinate() {coordinateIndex=2}
  ) {desc=Uses your base mesh UV's, group=Decal, parameterName=Use Decal UV3, sortPriority=-1}
  n39 = Add(
   A=AppendVector(
    A=ScalarParameter() {group=Decal, parameterName=Decal_PanX, sortPriority=11}
    B=ScalarParameter() {group=Decal, parameterName=Decal_PanY, sortPriority=11}
   )
   B=Multiply(
    A=AppendVector(A=@n35, B=@n36)
    B=StaticSwitchParameter(True=@n37, False=@n38) {desc=Uses your base mesh UV's, group=Decal, parameterName=Use Decal UV 0, sortPriority=-1}
   )
  )
  n40 = TextureSample(UVs=@n39, Tex=@n34)
  n41 = VectorParameter() {group=Decal, parameterName=Decal Tint, defaultValue=<rgba 1 1 1 0>}
  n42 = Multiply(A=@n40.RGB, B=@n41.RGB)
  n43 = TextureSampleParameter2D(UVs=@n39) {
   group=Decal
   parameterName=DecalSpecular
   samplerType=SAMPLERTYPE_Masks
   sortPriority=-4
   texture=&T_Fortnite_Default_S
  }
  n44 = Multiply(A=@n43.A, B=VectorParameter() {group=Decal, parameterName=Decal Emissive Tint, sortPriority=7}.RGB)
  n45 = TextureSampleParameter2D(UVs=@n39) {group=Decal, parameterName=DecalNormals, samplerType=SAMPLERTYPE_Normal, sortPriority=3, texture=/Engine/EngineMaterials/FlatNormal}
  n46 = SetMaterialAttributes!(MaterialAttributes, BaseColor=@n42, Metallic=@n43.G, Specular=@n43.G, Roughness=@n43.G, EmissiveColor=@n44, Normal=@n45.RGB) {attributeSetTypes [
    #1
    #3
    #4
    #2
    #5
    #6
   ]}
  n47 = SetMaterialAttributes!(MaterialAttributes, BaseColor=@n42, Metallic=@n43.G, Specular=@n43.G, Roughness=@n43.G, EmissiveColor=@n44) {attributeSetTypes [
    #1
    #3
    #4
    #2
    #5
   ]}
  n48 = StaticBool() {value=true}
  n49 = NamedRerouteDeclaration(None=MaterialFunctionCall!(
    transitionFactor
    1=ScalarParameter() {defaultValue=1, group=Decal, parameterName=L1_DecalMask, sortPriority=13}
    2=ScalarParameter() {defaultValue=1, group=Decal, parameterName=L2_DecalMask, sortPriority=14}
    3=ScalarParameter() {defaultValue=1, group=Decal, parameterName=L3_DecalMask, sortPriority=15}
    4=ScalarParameter() {defaultValue=1, group=Decal, parameterName=L4_DecalMask, sortPriority=16}
    Use 2 Textures=@n48
    Use 3 Textures=@n48
    Use 4 Textures=@n48
   ) {materialFunction=&/MF_LerpTexturesBasedOnInt_Float4_4Layers}.Result) {name=DecalLayerMask, nodeColor=<rgba 0.3359 0 1 1>}
  n50 = VertexNormalWS()
  n51 = VectorParameter() {group=Decal, parameterName=Decal Mask By World Normal, defaultValue=<rgba 0 0 1 0>}
  n52 = Subtract(A=WorldPosition().XYZ, B=ActorPositionWS())
  n53 = DotProduct(
   A=Add(
    A=Transform(None=@n52) {transformSourceType=TRANSFORMSOURCE_World, transformType=TRANSFORM_Local}
    B=Constant3Vector() {constant=<rgba 5 5 0 1>}
   )
   B=VectorParameter() {group=Decal, parameterName=Decal Mask By Mesh Axis, defaultValue=<rgba 0 1 0 0>}.RGB
  )
  n54 = Constant() {r=5}
  n55 = Saturate(None=StaticSwitchParameter(True=DotProduct(A=@n50, B=@n51.RGB), False=Divide(A=@n53, B=@n54)) {group=Decal, parameterName=Mask by WorldNormals})
  n56 = StaticSwitchParameter(True=Multiply(A=@n49, B=NamedRerouteDeclaration(None=@n55) {name=DecalsAxisMask, nodeColor=<rgba 1 0.1406 0 1>}), False=@n49) {group=Decal, parameterName=Decal Mask By Axis}
  n57 = MaterialFunctionCall(
   Transition Phase=ScalarParameter() {defaultValue=0.5, group=Decal, parameterName=PBRDecalTransition, sortPriority=8}
   Height Texture=@n40.A
   Contrast=ScalarParameter() {defaultValue=0.5, group=Decal, parameterName=PBRDecalContrast, sortPriority=8}
  ) {materialFunction=&HeightLerp}
  n58 = NamedRerouteUsage()
  n59 = GetMaterialAttributes(None=@n58) {attributeGetTypes [#1, #5]}
  n60 = ScalarParameter() {defaultValue=0.5, group=Decal, parameterName=Decal Mask Transition}
  n61 = ScalarParameter() {defaultValue=1, group=Decal, parameterName=Decal Mask Contrast, sortPriority=1}
  n62 = Constant() {r=1}
  n63 = MaterialFunctionCall(
   TextureObject=@n34
   TextureSize=MaterialFunctionCall(X=@n62, Y=@n62, Z=@n62) {materialFunction=&MakeFloat3}.Result
   WorldPosition=@n8
   Export Float 4=StaticBool() {value=true}
   World Space Normal=Transform(None=VertexNormalWS()) {transformSourceType=TRANSFORMSOURCE_World, transformType=TRANSFORM_Local}
  ) {materialFunction=&WorldAlignedTexture}
  n64 = ComponentMask(None=@n63.XYZ Texture) {a=true}
  n65 = Constant() {r=0.5}
  n66 = MaterialFunctionCall(Float2=@n10) {materialFunction=&BreakOutFloat2Components}
  n67 = SphereMask(A=@n65, B=@n66.R) {attenuationRadius=0.5}
  n68 = SphereMask(A=@n66.G, B=@n65) {attenuationRadius=0.5}
  n69 = TextureSampleParameter2D(UVs=@n10) {group=Decal, parameterName=DecalMask, samplerType=SAMPLERTYPE_Masks, texture=&T_WhiteMask}
  n70 = TextureSample(UVs=ComponentMask(None=@n8) {g=true, r=true}, Tex=@n34)
  n71 = AppendVector(A=@n70.RGB, B=@n70.RGB)
  n72 = TextureSample(UVs=ComponentMask(None=@n8) {b=true, g=true}, Tex=@n34)
  n73 = AppendVector(A=@n72.RGB, B=@n72.RGB)
  n74 = TextureSample(UVs=ComponentMask(None=@n8) {b=true, r=true}, Tex=@n34)
  n75 = AppendVector(A=@n74.RGB, B=@n74.RGB)
  n76 = ComponentMask(None=VertexNormalWS()) {r=true}
  n77 = LinearInterpolate(A=@n75, B=@n73, Alpha=ComponentMask(None=Abs(None=@n76)) {r=true}) {desc=Float3}
  n78 = StaticSwitchParameter(
   True=NamedRerouteDeclaration(None=@n71) {name=DecalWorldZ, nodeColor=<rgba 0 1 0.3672 1>}
   False=StaticSwitchParameter(True=@n73, False=StaticSwitchParameter(True=@n75, False=@n77) {group=Decal, parameterName=Use Decal WS Y Only, sortPriority=1}) {group=Decal, parameterName=Use Decal WS X Only, sortPriority=1}
  ) {group=Decal, parameterName=Use Decal WS Z Only, sortPriority=1}
  n79 = StaticSwitchParameter(True=Multiply(A=Multiply(A=@n67, B=@n68), B=@n69.R), False=ComponentMask(None=@n78) {a=true}) {group=Decal, parameterName=UseMagicDecal}
  n80 = StaticSwitchParameter(
   True=LinearInterpolate(B=StaticSwitchParameter(True=@n64, False=@n79) {group=Decal, parameterName=Use Decal WS LocalTransform}, Alpha=@n56)
   False=NamedRerouteUsage()
  ) {desc="World space projected decal, to render on top of multiple actors", group=Decal, parameterName=Use Decal WS, sortPriority=-1}
  n81 = ComponentMask(None=@n63.XYZ Texture) {b=true, g=true, r=true}
  n82 = StaticSwitchParameter(
   True=MaterialFunctionCall!(Texture=NamedRerouteUsage(), UV=@n10, Encrypt, DebugMips, MipLevel) {materialFunction=/Game/Packages/Fortress_SharedMaterials/Base_Material_Functions/MF_Magic}.Result
   False=ComponentMask(None=@n78) {b=true, g=true, r=true}
  ) {group=Decal, parameterName=UseMagicDecal}
  n83 = StaticSwitchParameter(
   True=LinearInterpolate(B=StaticSwitchParameter(True=@n81, False=@n82) {group=Decal, parameterName=Use Decal WS LocalTransform}, Alpha=@n56) {constA=1}
   False=LinearInterpolate(B=NamedRerouteUsage(), Alpha=@n56) {constA=1}
  ) {desc="World space projected decal, to render on top of multiple actors", group=Decal, parameterName=Use Decal WS, sortPriority=-1}
  n84 = Multiply(A=@n41.RGB, B=@n83)
  n85 = MaterialFunctionCall(A=@n59.BaseColor, B=@n84, Transition Phase=@n60, Height Texture=@n80, Contrast=@n61) {materialFunction=&HeightLerp}
  n86 = ScalarParameter() {group=Decal, parameterName=Decal Emissive Day, sortPriority=8}
  n87 = Multiply(
   A=MaterialFunctionCall(
    Color_Morning=@n86
    Color_Day=@n86
    Color_Evening=@n86
    Color_Night=ScalarParameter() {group=Decal, parameterName=Decal Emissive Night, sortPriority=8}
   ) {materialFunction=&Color_Time_of_Day}.Result
   B=VectorParameter() {group=Decal, parameterName=Decal Emissive Tint, sortPriority=7}.RGB
  )
  n88 = DotProduct(A=CameraVectorWS(), B=VertexNormalWS())
  n89 = GetMaterialAttributes(None=@n58) {attributeGetTypes [
    #1
    #3
    #4
    #2
    #5
    #6
   ]}
  n90 = ScalarParameter() {group=Decal, parameterName=Diffuse to decal Max, sortPriority=3}
  n91 = ScalarParameter() {group=Decal, parameterName=Diffuse to decal Min, sortPriority=2}
  n92 = ComponentMask(None=@n89.BaseColor) {r=true}
  n93 = MaterialFunctionCall(
   A=@n89.BaseColor
   B=@n84
   Transition Phase=@n60
   Height Texture=Subtract(A=@n80, B=LinearInterpolate(A=@n90, B=@n91, Alpha=@n92))
   Contrast=@n61
  ) {materialFunction=&HeightLerp}
  n94 = Multiply(
   A=Multiply(A=@n83, B=Multiply(A=@n87, B=@n88))
   B=StaticSwitchParameter(True=@n85.Alpha, False=@n93.Alpha) {desc="Default is Modulated, use Decal A channel for mask", group=Decal, parameterName=Use Lerp Simple, sortPriority=-2}
  )
  n95 = Add(A=@n59.EmissiveColor, B=@n94)
  n96 = ScalarParameter() {defaultValue=1, desc="0-1 multiply, 1-2 add through the mask", group=Decal, parameterName=Decal Metal Level, sliderMax=2, sortPriority=4}
  n97 = Multiply(A=@n89.Metallic, B=@n96)
  n98 = Saturate(None=Add(A=@n96) {constB=-1})
  n99 = ScalarParameter() {defaultValue=1, group=Decal, parameterName=Decal SpecLevel, sortPriority=5}
  n100 = ScalarParameter() {group=Decal, parameterName=Decal Rough Min, sortPriority=6}
  n101 = ScalarParameter() {defaultValue=1, group=Decal, parameterName=Decal Rough Max, sortPriority=6}
  n102 = ScalarParameter() {group=Decal, parameterName=Decal Flatten Normal, sortPriority=7}
  n103 = SetMaterialAttributes!(
   MaterialAttributes=@n89.MaterialAttributes
   BaseColor=@n93.Results
   Metallic=LinearInterpolate(A=@n89.Metallic, B=Add(A=@n97, B=@n98), Alpha=@n93.Alpha)
   Specular=LinearInterpolate(A=@n89.Specular, B=Multiply(A=@n99, B=@n89.Specular), Alpha=@n93.Alpha)
   Roughness=LinearInterpolate(A=@n89.Roughness, B=LinearInterpolate(A=@n100, B=@n101, Alpha=@n89.Roughness), Alpha=@n93.Alpha)
   EmissiveColor=Add(A=@n94, B=@n89.EmissiveColor)
   Normal=MaterialFunctionCall(Normal=@n89.Normal, Flatness=Multiply(A=@n93.Alpha, B=@n102)) {materialFunction=&FlattenNormal}.Result
  ) {attributeSetTypes [
    #1
    #3
    #4
    #2
    #5
    #6
   ]}
  n104 = GetMaterialAttributes(None=@n58) {attributeGetTypes [#1]}
  n105 = Multiply(A=@n104.BaseColor, B=@n84)
  n106 = GetMaterialAttributes(None=@n58) {attributeGetTypes [
    #1
    #3
    #4
    #2
   ]}
  n107 = SetMaterialAttributes!(
   MaterialAttributes=@n106.MaterialAttributes
   BaseColor=Multiply(A=@n84, B=@n106.BaseColor)
   Metallic=Multiply(A=@n84, B=@n106.Metallic)
   Specular=Multiply(A=@n106.Specular, B=@n84)
   Roughness=Multiply(A=@n106.Roughness, B=@n84)
  ) {attributeSetTypes [
    #1
    #3
    #4
    #2
   ]}
  n108 = StaticSwitchParameter(
   True=StaticSwitchParameter(
    True=SetMaterialAttributes!(MaterialAttributes=@n59.MaterialAttributes, BaseColor=@n85.Results, EmissiveColor=@n95) {attributeSetTypes [#1, #5]}
    False=@n103
   ) {desc="Default is Modulated, use Decal A channel for mask", group=Decal, parameterName=Use Lerp Simple, sortPriority=-2}
   False=StaticSwitchParameter(
    True=SetMaterialAttributes!(MaterialAttributes=@n104.MaterialAttributes, BaseColor=@n105) {attributeSetTypes [#1]}
    False=@n107
   ) {desc="Default is Modulated, use Decal A channel for mask", group=Decal, parameterName=Use Mod Simple, sortPriority=-4}
  ) {
   desc="Default is Modulated, use Decal A channel for mask ( Can be heightmap)"
   group=Decal
   parameterName="Use Lerp Decal (A is mask)"
   sortPriority=-3
  }
  n109 = StaticSwitchParameter(
   True=BlendMaterialAttributes(A=@n58, B=ShadingPathSwitch(Default=@n46, Deferred=@n46, Forward=@n47, Mobile=@n47), Alpha=Multiply(A=@n56, B=@n57.Alpha)) {vertexAttributeBlendType=UseA}
   False=@n108
  ) {group=Decal, parameterName="Use UV0 PBR Decal (DiffuseA is mask)", sortPriority=-4}
  n110 = ScalarParameter() {defaultValue=0.5, group=Decal, parameterName=Decal Height Mask}
  n111 = GetMaterialAttributes(None=@n58) {attributeGetTypes [#7]}
  n112 = ScalarParameter() {group=Decal, parameterName=Decal Height Mask Contrast}
  n113 = StaticSwitchParameter(
   True=BlendMaterialAttributes(
    A=@n58
    B=@n109
    Alpha=MaterialFunctionCall(Transition Phase=@n110, Height Texture=@n111.OpacityMask, Contrast=@n112) {materialFunction=&HeightLerp}.Alpha
   ) {vertexAttributeBlendType=UseA}
   False=@n109
  ) {desc=Default uses chanal 3 in 3ds max, group=Decal, parameterName=Use Decal Height Masked, sortPriority=-5}
  FunctionOutput(None=StaticSwitchParameter(True=StaticSwitchParameter(True=@n32, False=@n113) {group=Decal, parameterName=GraffitiBP, sortPriority=-2}, False=@n58) {desc=Default uses chanal 3 in 3ds max, group=Decal, parameterName="Use Decal ", sortPriority=-5}) {description=This function is adding buff masks Test}
  Multiply()
  n114 = Add(
   A=Transform(None=WorldPosition() {worldPositionShaderOffset=WPT_ExcludeAllShaderOffsets}.XYZ) {transformSourceType=TRANSFORMSOURCE_World, transformType=TRANSFORM_Local}
   B=Constant3Vector() {constant=<rgba 256 256 0 0>}
  ) {constB=256}
  n115 = Multiply(A=Abs(None=Constant3Vector() {constant=<rgba 512 512 384 0>})) {constB=-1}
  n116 = Frac(None=ComponentMask(None=Divide(A=@n114, B=@n115)) {b=true, r=true})
  n117 = StaticSwitchParameter(True=NamedRerouteUsage(), False=NamedRerouteDeclaration(None=@n116) {name=DecalActorSpaceUV, nodeColor=<rgba 0 1 0.4609 1>}) {desc=Otherwise Local Projection, group=Decal, parameterName=Decal_Flipbook_Use UVs, sortPriority=12}
  n118 = Constant() {r=0.5}
  n119 = Constant() {r=1}
  n120 = Constant() {r=0.5}
  n121 = ScalarParameter() {defaultValue=1, desc="From 1 to 4 ", group=Decal, parameterName=Decal_FlipBookSet, sliderMax=4, sliderMin=1, sortPriority=13}
  n122 = ScalarParameter() {group=Decal, parameterName=Decal_RandomSeed, sortPriority=13}
  n123 = DotProduct(
   A=ActorPositionWS()
   B=AppendVector(A=AppendVector(A=@n122, B=@n122), B=ScalarParameter() {group=Decal, parameterName=Decal_RandomSeed Z, sortPriority=13})
  )
  n124 = Constant() {r=3.9}
  n125 = Multiply(A=Frac(None=Divide(A=@n123, B=@n124))) {constB=5}
  n126 = Add(A=StaticSwitchParameter(True=Add(A=@n121, B=@n125), False=@n121) {defaultValue=true, desc=Change to non default later, group=Decal, parameterName=FlipbookRandom, sortPriority=13}) {constB=-1}
  n127 = Floor(None=Multiply(A=AppendVector(A=@n119, B=@n120), B=Clamp(None=@n126) {maxDefault=4}))
  n128 = Constant2Vector() {g=2, r=2}
  NamedRerouteDeclaration(None=Add(A=Multiply(A=@n117, B=@n118) {constB=0.5}, B=Divide(A=@n127, B=@n128))) {name=Flipbook, nodeColor=<rgba 1 0.2812 0 1>}
  n129 = NamedRerouteUsage()
  n130 = NamedRerouteDeclaration(None=@n39) {name=Decal Uv's, nodeColor=<rgba 1 0 0.4453 1>}
  n131 = MaterialFunctionCall(Float4=TextureSample(UVs=StaticSwitchParameter(True=@n129, False=@n130) {group=Decal, parameterName=Decal_Use_Flipbook, sortPriority=-5}, Tex=@n34).RGBA) {materialFunction=&BreakOutFloat4Components}
  NamedRerouteDeclaration(None=MaterialFunctionCall(X=@n131.R, Y=@n131.R, Z=@n131.R) {materialFunction=&MakeFloat3}.Result) {name=DecalTexture, nodeColor=<rgba 1 0 0.9141 1>}
  n132 = NamedRerouteUsage()
  n133 = Transform(None=VertexNormalWS()) {transformSourceType=TRANSFORMSOURCE_World, transformType=TRANSFORM_Local}
  n134 = Abs(None=Add(A=ComponentMask(None=@n133) {b=true}, B=ComponentMask(None=@n133) {r=true}))
  NamedRerouteDeclaration(None=StaticSwitchParameter(True=Max(A=@n132, B=@n134), False=@n132) {group=Decal, parameterName=Decal_Use_Flipbook, sortPriority=-5}) {name=DecalTextureFlipbookMask, nodeColor=<rgba 1 0.7031 0 1>}
  NamedRerouteDeclaration(None=FunctionInput() {inputType=FunctionInput_MaterialAttributes}) {name=DecalBGMaterial, nodeColor=<rgba 0.6406 0 1 1>}
  NamedRerouteDeclaration(None=Multiply(A=@n56, B=@n131.A)) {name=DecalAlpha_1, nodeColor=<rgba 0.2891 0 1 1>}
  n135 = NamedRerouteUsage()
  n136 = OneMinus(None=@n134)
  NamedRerouteDeclaration(None=StaticSwitchParameter(True=Multiply(A=@n135, B=@n136), False=@n135) {group=Decal, parameterName=Decal_Use_Flipbook, sortPriority=-5}) {name=DecalAlphaMasked, nodeColor=<rgba 0 0.625 1 1>}
  Reroute(None=@n71)
  Reroute(None=@n116)
  Reroute(None=@n33)
 ]
}
