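# Material function for layered materials that adds an optional custom emissive color texture to incoming material attributes and outputs an emissive mask.
# The texture samples a switchable UV channel; its alpha is the mask, scaling emissive, tinting BaseColor and capping Metallic.
# Tone, boost and metal values blend across up to four layers via MF_LerpTexturesBasedOnInt_Float4_4Layers.
# Mobile writes only raw texture RGB to emissive; disabling CustomEmissiveColor passes both inputs through.
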MaterialFunction /Game/Athena/WinterArt/Materials/MF_EmissiveColor
graph {nodes [
  n1 = TextureCoordinate() {coordinateIndex=1}
  n2 = TextureCoordinate() {coordinateIndex=2}
  n3 = TextureSampleParameter2D(UVs=StaticSwitchParameter(
    True=TextureCoordinate() {coordinateIndex=3}
    False=StaticSwitchParameter(True=@n1, False=@n2) {group=CustomEmissiveColor, parameterName=Use UV 1, sortPriority=-1}
   ) {group=CustomEmissiveColor, parameterName=Use UV 3, sortPriority=-1}) {
   desc="UV2 (UE) is used by default Alpha is the mask"
   group=CustomEmissiveColor
   parameterName=CustomEmissiveColor
   samplerSource=SSM_Wrap_WorldGroupSettings
   sortPriority=1
   texture=&T_WhitesRGB
  }
  n4 = FunctionInput() {bUsePreviewValueAsDefault=true, inputName=Emissive Mask, inputType=FunctionInput_Scalar, sortPriority=2, previewValue=<1 1 1 1>}
  FunctionOutput(None=StaticSwitchParameter(True=@n3.A, False=@n4) {group=CustomEmissiveColor, parameterName=CustomEmissiveColor, sortPriority=-2}) {outputName=EmissiveMask, sortPriority=2}
  n5 = FunctionInput() {bUsePreviewValueAsDefault=true, inputType=FunctionInput_MaterialAttributes, sortPriority=-7, previewValue=<1 1 1 1>}
  n6 = GetMaterialAttributes(None=@n5) {attributeGetTypes [#1, #2]}
  n7 = MaterialFunctionCall!(
   transitionFactor
   1=ScalarParameter() {defaultValue=1, group=CustomEmissiveColor, parameterName=L1_Emissive_DifTone, sortPriority=2}
   2=ScalarParameter() {defaultValue=1, group=CustomEmissiveColor, parameterName=L2__Emissive_DifTone, sortPriority=2}
   3=ScalarParameter() {defaultValue=1, group=CustomEmissiveColor, parameterName=L3__Emissive_DifTone, sortPriority=2}
   4=ScalarParameter() {defaultValue=1, group=CustomEmissiveColor, parameterName=L4_Emissive_DifTone, sortPriority=2}
   Use 2 Textures=StaticBoolParameter() {group=Material_2, parameterName=Use 2 Materials}
   Use 3 Textures=StaticBoolParameter() {group=Material_3, parameterName=Use 3 Materials}
   Use 4 Textures=StaticBoolParameter() {group=Material_4, parameterName=Use 4 Materials}
  ) {materialFunction=&/MF_LerpTexturesBasedOnInt_Float4_4Layers}
  n8 = MaterialFunctionCall(
   1=ScalarParameter() {defaultValue=1, group=CustomEmissiveColor, parameterName=L1_Emissive_Boost, sortPriority=2}
   2=ScalarParameter() {defaultValue=1, group=CustomEmissiveColor, parameterName=L2__Emissive_Boost, sortPriority=2}
   3=ScalarParameter() {defaultValue=1, group=CustomEmissiveColor, parameterName=L3__Emissive_Boost, sortPriority=2}
   4=ScalarParameter() {defaultValue=1, group=CustomEmissiveColor, parameterName=L4_Emissive_Boost, sortPriority=2}
   Use 2 Textures=StaticBoolParameter() {group=Material_2, parameterName=Use 2 Materials}
   Use 3 Textures=StaticBoolParameter() {group=Material_3, parameterName=Use 3 Materials}
   Use 4 Textures=StaticBoolParameter() {group=Material_4, parameterName=Use 4 Materials}
  ) {materialFunction=&MF_LerpTexturesBasedOnInt_Float4_4Layers}
  n9 = MaterialFunctionCall(
   1=ScalarParameter() {group=CustomEmissiveColor, parameterName=L1_Emissive_Metal, sortPriority=2}
   2=ScalarParameter() {group=CustomEmissiveColor, parameterName=L2__Emissive_Metal, sortPriority=2}
   3=ScalarParameter() {group=CustomEmissiveColor, parameterName=L3__Emissive_Metal, sortPriority=2}
   4=ScalarParameter() {group=CustomEmissiveColor, parameterName=L4_Emissive_Metal, sortPriority=2}
   Use 2 Textures=StaticBoolParameter() {group=Material_2, parameterName=Use 2 Materials}
   Use 3 Textures=StaticBoolParameter() {group=Material_3, parameterName=Use 3 Materials}
   Use 4 Textures=StaticBoolParameter() {group=Material_4, parameterName=Use 4 Materials}
  ) {materialFunction=&MF_LerpTexturesBasedOnInt_Float4_4Layers}
  n10 = SetMaterialAttributes!(
   MaterialAttributes=@n6.MaterialAttributes
   BaseColor=Multiply(A=@n6.BaseColor, B=LinearInterpolate(B=@n7.Result, Alpha=@n3.A) {constA=1})
   EmissiveColor=Multiply(A=Multiply(A=@n8.Result, B=@n3.A), B=@n3.RGB)
   Metallic=Min(A=@n6.Metallic, B=Multiply(A=@n9.Result, B=@n3.A))
  ) {attributeSetTypes [#1, #3, #2]}
  n11 = SetMaterialAttributes!(MaterialAttributes=@n6.MaterialAttributes, EmissiveColor=@n3.RGB) {attributeSetTypes [#3]}
  FunctionOutput(None=StaticSwitchParameter(True=ShadingPathSwitch(Default=@n10, Deferred=@n10, Forward=@n10, Mobile=@n11), False=@n5) {group=CustomEmissiveColor, parameterName=CustomEmissiveColor, sortPriority=-2})
 ]}
