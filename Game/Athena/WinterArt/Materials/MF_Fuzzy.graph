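# MF_Fuzzy is a material function in the WinterArt folder that adds a fuzzy Fresnel edge to incoming material attributes.
# A Fresnel with adjustable exponent and center, scaled by Fuzz Level, brightens base color by adding a multiple of itself and lerps roughness toward FuzzRoughness.
# The incoming OpacityMask blends the original and fuzzed attributes, keeping the original vertex attributes.
# The FuzzyEdge static switch selects this result or passes the input through unchanged.
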MaterialFunction /Game/Athena/WinterArt/Materials/MF_Fuzzy
graph {
 properties {description="This function is used to add black and white fresnel to a material. You can control the strength, value, and center of the fresnel."}
 nodes [
  n1 = FunctionInput() {inputType=FunctionInput_MaterialAttributes}
  n2 = GetMaterialAttributes(None=@n1) {attributeGetTypes [#1, #2, #3]}
  n3 = Fresnel(
   ExponentIn=ScalarParameter() {defaultValue=3, group=Fuzz, parameterName=FuzzExponent}
   BaseReflectFractionIn=ScalarParameter() {group=Fuzz, parameterName=FuzzCenter, sortPriority=2}
  )
  n4 = Multiply(A=@n2.BaseColor, B=@n3)
  n5 = ScalarParameter() {defaultValue=1, group=Fuzz, parameterName=Fuzz Level, sortPriority=1}
  n6 = Multiply(A=@n3, B=@n5)
  n7 = SetMaterialAttributes!(
   MaterialAttributes=@n2.MaterialAttributes
   BaseColor=Add(A=@n2.BaseColor, B=Multiply(A=@n4, B=@n5))
   Roughness=LinearInterpolate(A=@n2.Roughness, B=ScalarParameter() {defaultValue=1, group=Fuzz, parameterName=FuzzRoughness, sortPriority=1}, Alpha=Saturate(None=@n6))
  ) {attributeSetTypes [#1, #2]}
  FunctionOutput(None=StaticSwitchParameter(True=BlendMaterialAttributes(A=@n1, B=@n7, Alpha=@n2.OpacityMask) {vertexAttributeBlendType=UseA}, False=@n1) {group=Fuzz, parameterName=FuzzyEdge})
 ]
}
