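# Material function that adds bottom-to-top gradients to material attributes along an object's local Z, normalized by its local bounds.
# A switchable gradient tints base color, weighted by one of four mask parameters picked by transitionFactor (default UV1 red), and scales roughness.
# A second switch adds an emissive gradient, multiplied by base color and time-of-day weights from FortniteMaterialParameters.
# Low quality passes the input through unchanged.
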MaterialFunction /Game/Athena/WinterArt/Materials/MF_ObjectSpaceGradient
graph {nodes [
  n1 = FunctionInput() {inputName=MaterialIn, inputType=FunctionInput_MaterialAttributes}
  n2 = GetMaterialAttributes(None=@n1) {attributeGetTypes [#1, #2]}
  n3 = VectorParameter() {group=Actor Gradient, parameterName=Actor GradientBottom, sortPriority=1, defaultValue=<rgba 1 1 1 0>}
  n4 = VectorParameter() {group=Actor Gradient, parameterName=Actor GradientTop, sortPriority=1, defaultValue=<rgba 1 1 1 1>}
  n5 = MaterialFunctionCall() {materialFunction=&ObjectLocalBounds}
  n6 = Subtract(
   A=TransformPosition(None=WorldPosition() {worldPositionShaderOffset=WPT_ExcludeAllShaderOffsets}.XYZ) {transformSourceType=TRANSFORMPOSSOURCE_World}
   B=@n5.Local Bounds Minimum
  )
  n7 = Multiply(A=@n5.Local Bounds Size, B=ScalarParameter() {defaultValue=1, group=Actor Gradient, parameterName=Gradient Size, sortPriority=3})
  n8 = Saturate(None=ComponentMask(None=Divide(A=@n6, B=@n7)) {b=true})
  n9 = FunctionInput(Preview=ComponentMask(None=TextureCoordinate() {coordinateIndex=1}) {r=true}) {bUsePreviewValueAsDefault=true, inputName=transitionFactor, inputType=FunctionInput_Scalar, sortPriority=-1}
  n10 = Constant() {r=2}
  n11 = If(
   A=Constant() {r=1}
   B=@n9
   "A > B"=ScalarParameter() {defaultValue=1, group=Actor Gradient, parameterName=M1 Mask, sortPriority=4}
   "A < B"=ScalarParameter() {defaultValue=5, group=Actor Gradient, parameterName=M2 Mask}
  ) {desc=0}
  n12 = ScalarParameter() {defaultValue=6, group=Actor Gradient, parameterName=M3 Mask}
  n13 = LinearInterpolate(
   B=LinearInterpolate(A=@n3.RGB, B=@n4.RGB, Alpha=@n8)
   Alpha=If(
    A=Constant() {r=3}
    B=@n9
    "A > B"=If(A=@n10, B=@n9, "A > B"=@n11, "A < B"=@n12)
    "A < B"=ScalarParameter() {defaultValue=7, group=Actor Gradient, parameterName=M4 Mask}
   )
  ) {constA=1}
  n14 = LinearInterpolate(
   A=ScalarParameter() {defaultValue=1, group=Actor Gradient, parameterName=Actor GradientBottom R, sortPriority=2}
   B=ScalarParameter() {defaultValue=1, group=Actor Gradient, parameterName="Actor GradientTop R ", sortPriority=2}
   Alpha=@n8
  )
  n15 = StaticSwitchParameter(
   True=SetMaterialAttributes!(MaterialAttributes=@n2.MaterialAttributes, BaseColor=Multiply(A=@n2.BaseColor, B=@n13), Roughness=Multiply(A=@n2.Roughness, B=@n14)) {attributeSetTypes [#1, #2]}
   False=@n1
  ) {group=Actor Gradient, parameterName=Use Actor Gradient}
  n16 = GetMaterialAttributes(None=@n15) {attributeGetTypes [#1, #3]}
  n17 = VectorParameter() {group=Actor Gradient, parameterName=Actor EmissiveGradientBottom, sortPriority=1}
  n18 = VectorParameter() {group=Actor Gradient, parameterName="Actor  EmissiveGradientTop", sortPriority=1}
  n19 = CollectionParameter() {
   collection=&FortniteMaterialParameters
   parameterName=FortniteActiveTimeOfDay
  }
  n20 = VectorParameter() {
   group=Actor Gradient
   parameterName=TOD Actor Grad Emissive
   channelNames {a=Night, b=Evening, g=Day, r=Morning}
   defaultValue=<rgba 0 0 0 1>
  }
  n21 = AppendVector(A=@n20, B=@n20)
  n22 = Multiply(A=@n16.BaseColor, B=Multiply(A=LinearInterpolate(A=@n17.RGB, B=@n18.RGB, Alpha=@n8), B=DotProduct(A=@n19, B=@n21)))
  n23 = StaticSwitchParameter(
   True=SetMaterialAttributes!(MaterialAttributes=@n16.MaterialAttributes, EmissiveColor=Add(A=@n22, B=@n16.EmissiveColor)) {attributeSetTypes [#3]}
   False=@n15
  ) {group=Actor Gradient, parameterName=Use Actor Emissive Gradient}
  FunctionOutput(None=QualitySwitch(Default=@n23, Low=@n1, Medium=@n23)) {outputName=Color*Gradient}
 ]}
