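# Material function that scatters random decal splatters over incoming material attributes on a world-space XY grid.
# Per-cell vector noise picks an atlas sub-image and, against a threshold, decides whether that cell shows a splat.
# Center-scaled cell UVs sample diffuse, mask and normal maps, blended in by diffuse alpha within the cell bounds.
# A static switch gates the effect, which Low and Medium quality bypass.
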MaterialFunction /Game/Athena/WinterArt/Materials/MF_RandomDecalSplatter
graph {nodes [
  n1 = FunctionInput() {inputType=FunctionInput_MaterialAttributes}
  n2 = Subtract(
   A=Divide(A=WorldPosition().XYZ, B=ScalarParameter() {defaultValue=512, group=RandomSplatter, parameterName=RandomSplatterGridSize})
   B=VectorParameter() {group=RandomSplatter, parameterName=RandomSplatterOffset, defaultValue=<rgba 0.5 0.5 0.5 1>}.RGB
  )
  n3 = VectorNoise(World Position=Multiply(A=@n2, B=Constant3Vector() {constant=<rgba 1 1 0 0>}))
  n4 = ScalarParameter() {defaultValue=2, group=RandomSplatter, parameterName=RS_SubImages_X}
  n5 = ScalarParameter() {defaultValue=2, group=RandomSplatter, parameterName=RS_SubImages_Y}
  n6 = Multiply(A=ComponentMask(None=@n3) {g=true}, B=Multiply(A=@n4, B=@n5))
  n7 = Subtract(A=@n6, B=Frac(None=@n6) {desc=Frame Lerp}) {desc=IndexA}
  n8 = AppendVector(A=@n4, B=@n5)
  n9 = Fmod(A=@n7, B=ComponentMask(None=@n8) {desc="## Horz Images", r=true}) {desc=IndexA_Horz}
  n10 = Divide(A=Constant2Vector() {g=1, r=1}, B=@n8) {desc=Scale}
  n11 = Floor(None=Multiply(A=@n7, B=ComponentMask(None=@n10) {desc="1 / ## Horz Images", r=true})) {desc=IndexA_Vert}
  n12 = MaterialFunctionCall(
   UVs=Frac(None=ComponentMask(None=@n2) {g=true, r=true})
   Texture Scale=ScalarParameter() {defaultValue=0.3983, group=RandomSplatter, parameterName=RandomSplatter_Scale}
  ) {materialFunction=&ScaleUVsByCenter}
  n13 = Multiply(A=Add(A=AppendVector(A=@n9, B=@n11) {desc=IndexA_Combined}, B=@n12.UVs), B=@n10) {desc=UV_A}
  n14 = TextureSampleParameter2D(UVs=@n13) {group=RandomSplatter, parameterName=RandomSplattter_Diffuse, texture=/Game/Global/Textures/Default/Blanks/T_MidgreysRGB}
  n15 = TextureSampleParameter2D(UVs=@n13) {
   group=RandomSplatter
   parameterName=RandomSplattter_Mask
   samplerType=SAMPLERTYPE_Masks
   texture=&T_Fortnite_Default_S
  }
  n16 = TextureSampleParameter2D(UVs=@n13) {group=RandomSplatter, parameterName=RandomSplattter_Normal, samplerType=SAMPLERTYPE_Normal, texture=/Engine/EngineMaterials/FlatNormal}
  n17 = SetMaterialAttributes!(MaterialAttributes, BaseColor=@n14.RGB, Metallic=@n15.G, Specular=@n15.G, Roughness=@n15.G, Normal=@n16.RGB) {attributeSetTypes [
    #1
    #2
    #3
    #4
    #5
   ]}
  n18 = ComponentMask(None=@n3) {r=true}
  n19 = ScalarParameter() {defaultValue=0.792, group=RandomSplatter, parameterName=RandomSplatterThreshold}
  n20 = Saturate(None=Ceil(None=Subtract(A=@n18, B=@n19)))
  n21 = BlendMaterialAttributes(A=@n1, B=@n17, Alpha=Multiply(A=Multiply(A=@n12.0-1 mask, B=@n20), B=@n14.A))
  FunctionOutput(None=StaticSwitchParameter(True=QualitySwitch(Default=@n21, Low=@n1, High=@n21, Medium=@n1, Epic=@n21), False=@n1) {group=RandomSplatter, parameterName=UseRandomDecalSplatter})
 ]}
