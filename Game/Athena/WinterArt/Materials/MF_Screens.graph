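# Material function that optionally adds animated, glowing screens to incoming material attributes, replacing existing emissive where vertex red and blue mask screen areas.
# Screen color, hue-shifted per actor by world position, tints base color and fades with view angle and an emissive mask.
# A vertex-color-selected or RGB layer texture gets scrolling scanline ramps, parallax and sine jitter, with timing offset per actor.
# Forward and mobile paths skip these animated effects.
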MaterialFunction /Game/Athena/WinterArt/Materials/MF_Screens
graph {nodes [
  n1 = GetMaterialAttributes(None=FunctionInput() {inputType=FunctionInput_MaterialAttributes, sortPriority=-1}) {attributeGetTypes [#1, #2]}
  n2 = VertexColor()
  n3 = Add(A=Add(A=@n2.R, B=@n2.R), B=@n2.B)
  n4 = Multiply(A=@n1.EmissiveColor, B=OneMinus(None=@n3))
  n5 = Divide(
   A=MaterialFunctionCall!(f2, f3=ActorPositionWS(), f4) {materialFunction=/Game/Packages/Fortnite_UniqueMaterials_1/VectorComponentOps}.f3 Add
   B=Constant() {r=32}
  )
  n6 = VectorParameter() {group=Screens, parameterName=ScreensColor, sortPriority=1, defaultValue=<rgba 0.6793 0.9531 0.1291 0>}
  n7 = MaterialFunctionCall(Hue Shift Percentage=@n5, Texture=@n6.RGB) {materialFunction=&HueShift}
  n8 = Max(A=Frac(None=@n5), B=OneMinus(None=ScalarParameter() {defaultValue=1, group=Screens, parameterName=Random Value}))
  n9 = FunctionInput() {bUsePreviewValueAsDefault=true, inputName=Emissive Mask, inputType=FunctionInput_Scalar, previewValue=<1 0 0 1>}
  n10 = CameraVectorWS()
  n11 = VertexNormalWS()
  n12 = Multiply(A=Max(A=@n9, B=Constant() {r=0.2}), B=Power(Base=DotProduct(A=@n10, B=@n11), Exp=Constant() {r=3}))
  n13 = ComponentMask(None=StaticSwitchParameter(True=TextureCoordinate() {coordinateIndex=2}, False=TextureCoordinate() {coordinateIndex=1}) {group=Screens, parameterName="Use UV3 (3dsMax) "}) {g=true}
  n14 = Add(A=Time(), B=@n5)
  n15 = LinearInterpolate(A=Add(A=@n14, B=Constant() {r=0.5}) {constB=0.5}, B=@n14, Alpha=@n2.R)
  n16 = Multiply(
   A=TextureSample(UVs=Add(A=@n13, B=@n15)) {samplerType=SAMPLERTYPE_LinearGrayscale, texture=/Game/Global/Textures/Default/Masks_and_Gradients/T_ScreenRamp_H}.R
   B=Constant() {r=3}
  )
  n17 = Add(A=@n16, B=Constant() {r=0.5})
  n18 = StaticSwitchParameter(True=TextureCoordinate() {coordinateIndex=2}, False=TextureCoordinate() {coordinateIndex=1}) {group=Screens, parameterName="Use UV3 (3dsMax) ", sortPriority=4}
  n19 = Multiply(A=@n16) {constB=0.25}
  n20 = Constant() {r=-2}
  n21 = Add(A=Time(), B=@n5)
  n22 = Add(A=Sine(None=@n21), B=Sine(None=@n21) {period=0.125})
  n23 = LinearInterpolate(Alpha=Frac(None=@n5)) {constB=0.25}
  n24 = Add(A=BumpOffset(Coordinate=@n18, Height=Add(A=@n19, B=@n20)), B=Multiply(A=Multiply(A=@n22, B=@n23), B=Constant2Vector() {g=0.1}))
  n25 = TextureObjectParameter() {
   group=Screens
   parameterName=ScreensLayers
   samplerType=SAMPLERTYPE_Masks
   sortPriority=1
   texture=/Game/Environments/Apollo/Sets/MountainBase/Props/Textures/T_None_M
  }
  n26 = TextureSample(UVs=ShadingPathSwitch(Default=@n24, Deferred=@n24, Forward=@n18, Mobile=@n18), Tex=@n25)
  n27 = StaticSwitchParameter(True=@n26.RGB, False=DotProduct(A=@n26.RGB, B=@n2)) {desc=For regular RGB texture within borders of vertex color and emissive masks, group=Screens, parameterName=Use RGB Screen}
  n28 = Add(A=@n17, B=Multiply(A=@n17, B=@n27))
  n29 = Multiply(
   A=Multiply(A=@n1.BaseColor, B=LinearInterpolate(A=@n7.Result, B=@n6.RGB, Alpha=@n8))
   B=Multiply(A=Multiply(A=@n12, B=@n3), B=ShadingPathSwitch(Default=@n28, Deferred=@n28, Forward=@n27, Mobile=@n27))
  )
  n30 = StaticSwitchParameter(
   True=SetMaterialAttributes!(MaterialAttributes=@n1.MaterialAttributes, EmissiveColor=Add(A=@n4, B=@n29)) {attributeSetTypes [#2]}
   False=@n1.MaterialAttributes
  ) {
   desc="Uses Vertex RGB for screen masking. Default R, G for second, B for blinker"
   group=Screens
   parameterName="Screen True (Vertex RGB)"
   sortPriority=-5
  }
  FunctionOutput(None=@n30)
 ]}
