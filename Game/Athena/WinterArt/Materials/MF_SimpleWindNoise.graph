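# Material function that generates cheap procedural wind-style noise from a UV input, using sine and cosine math with no texture samples.
# Sheared UVs drive two frequencies of trig waves, which are blended and mixed back into the original UVs as a domain warp.
# The warped coordinates feed two sine-cosine product patterns; their average and the second pattern form a vector that interpolates between Min and Max.
# It has no time node, so any motion comes from the supplied UVs.
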MaterialFunction /Game/Athena/WinterArt/Materials/MF_SimpleWindNoise
graph {nodes [
  n1 = FunctionInput(Preview=TextureCoordinate()) {bUsePreviewValueAsDefault=true, inputName=UV1, inputType=FunctionInput_Vector2, sortPriority=2}
  n2 = ComponentMask(None=@n1) {r=true}
  n3 = MaterialFunctionCall(X=@n2, Y=Add(A=@n2, B=ComponentMask(None=@n1) {g=true})) {materialFunction=&MakeFloat2}
  n4 = MaterialFunctionCall(X=Sine(None=ComponentMask(None=@n3.Result) {g=true}) {period=0.7}, Y=Cosine(None=ComponentMask(None=@n3.Result) {r=true}) {period=1.1}) {materialFunction=&MakeFloat2}
  n5 = Constant() {r=4}
  n6 = Multiply(A=@n3.Result, B=@n5)
  n7 = Multiply(A=@n3.Result, B=@n5)
  n8 = MaterialFunctionCall(X=Sine(None=ComponentMask(None=@n6) {r=true}) {period=1.3}, Y=Cosine(None=ComponentMask(None=@n7) {g=true}) {period=1.5}) {materialFunction=&MakeFloat2}
  n9 = LinearInterpolate(A=@n1, B=ConstantBiasScale(None=LinearInterpolate(A=@n4.Result, B=@n8.Result) {constAlpha=0.2}))
  n10 = Multiply(A=@n9) {constB=0.37}
  n11 = Multiply(A=Sine(None=ComponentMask(None=@n10) {g=true}) {period=0.2}, B=Cosine(None=ComponentMask(None=@n9) {r=true}) {period=0.2})
  n12 = ComponentMask(None=Multiply(A=@n9) {constB=0.5}) {r=true}
  n13 = ComponentMask(None=@n9) {g=true}
  n14 = ConstantBiasScale(None=Multiply(A=Sine(None=@n12) {period=0.2}, B=Cosine(None=@n13) {period=0.4}))
  n15 = DotProduct(
   A=MaterialFunctionCall(X=ConstantBiasScale(None=@n11), Y=@n14) {materialFunction=&MakeFloat2}.Result
   B=Constant2Vector() {g=0.5, r=0.5}
  )
  FunctionOutput(None=LinearInterpolate(
    A=FunctionInput() {bUsePreviewValueAsDefault=true, inputName=Min, sortPriority=50}
    B=FunctionInput() {bUsePreviewValueAsDefault=true, inputName=Max, sortPriority=60, previewValue=<1 1 1 1>}
    Alpha=MaterialFunctionCall(X=@n15, Y=@n14) {materialFunction=&MakeFloat3}.Result
   ))
 ]}
