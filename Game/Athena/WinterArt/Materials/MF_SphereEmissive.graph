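# MF_SphereEmissive is a material attributes function that fakes a spherical glow inside large emissive surfaces.
# A world-space sphere mask, centered on the actor position plus a Pivot Location offset with tunable radius and hardness, is tinted by Sphere Color.
# It adds this to the incoming emissive, then multiplies by that emissive again, so black areas stay dark.
# A static switch enables the effect or passes attributes through unchanged.
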MaterialFunction /Game/Athena/WinterArt/Materials/MF_SphereEmissive
graph {nodes [
  n1 = FunctionInput() {inputType=FunctionInput_MaterialAttributes, sortPriority=-1}
  n2 = GetMaterialAttributes(None=@n1) {attributeGetTypes [#1]}
  n3 = WorldPosition()
  n4 = Add(A=ActorPositionWS(), B=VectorParameter() {group=EmissiveSphere, parameterName=Pivot Location, sortPriority=2}.RGB)
  n5 = ScalarParameter() {group=EmissiveSphere, parameterName=Sphere Radius, sortPriority=1}
  n6 = ScalarParameter() {group=EmissiveSphere, parameterName=Sphere Hardness, sortPriority=1}
  n7 = Add(
   A=Multiply(
    A=SphereMask(A=@n3.XYZ, B=@n4, Radius=@n5, Hardness=@n6)
    B=VectorParameter() {group=EmissiveSphere, parameterName=Sphere Color, sortPriority=3}.RGB
   )
   B=@n2.EmissiveColor
  )
  n8 = StaticSwitchParameter(
   True=SetMaterialAttributes!(MaterialAttributes=@n2.MaterialAttributes, EmissiveColor=Multiply(A=@n7, B=@n2.EmissiveColor)) {attributeSetTypes [#1]}
   False=@n1
  ) {
   desc="Use for making fake shperical emissive glow inside big emissive surfaces. Uses Actor Pivot as starting location, but can be adjusted in relative space"
   group=EmissiveSphere
   parameterName=Emissive SphereMask
  }
  FunctionOutput(None=@n8)
 ]}
