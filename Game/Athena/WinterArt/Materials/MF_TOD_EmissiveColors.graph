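# Material function that tints a material's emissive color by time of day, filed under Athena winter art.
# It multiplies the incoming material attributes' EmissiveColor by the Color_Time_of_Day function's result, fed by four TOD-group color parameters for morning, day, evening and night.
# A "TOD Emissive" static switch selects either the tinted attributes or the untouched input, so each material can toggle the effect.
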MaterialFunction /Game/Athena/WinterArt/Materials/MF_TOD_EmissiveColors
graph {
 properties {description=This function changes the emissive color based on the time of day.}
 nodes [
  n1 = FunctionInput() {inputType=FunctionInput_MaterialAttributes}
  n2 = GetMaterialAttributes(None=@n1) {attributeGetTypes [#1]}
  n3 = VectorParameter() {group=TOD, parameterName=EmissiveEvening, sortPriority=3, defaultValue=<rgba 0.5208 0.4396 0.02 1>}
  n4 = Multiply(
   A=@n2.EmissiveColor
   B=MaterialFunctionCall(
    Color_Morning=VectorParameter() {group=TOD, parameterName=EmissiveMorning, sortPriority=1, defaultValue=<rgba 0.4 0.0568 0.1168 1>}.RGB
    Color_Day=VectorParameter() {group=TOD, parameterName=EmissiveDay, sortPriority=2, defaultValue=<rgba 0.1199 0.2582 0.4 1>}.RGB
    Color_Evening=@n3.RGB
    Color_Night=VectorParameter() {group=TOD, parameterName=EmissiveNight, sortPriority=4, defaultValue=<rgba 0.0273 0.1373 0.25 1>}.RGB
   ) {materialFunction=&Color_Time_of_Day}.Result
  )
  FunctionOutput(None=StaticSwitchParameter(
    True=SetMaterialAttributes!(MaterialAttributes=@n2.MaterialAttributes, EmissiveColor=@n4) {attributeSetTypes [#1]}
    False=@n1
   ) {group=TOD, parameterName=TOD Emissive})
 ]
}
