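# Material function that adds switchable vertex-painted layers to incoming material attributes.
# Red vertex color height-blends UV or world-aligned color, mask and normal textures, optionally offset by water height, with cheaper mobile values.
# Blue optionally adds view-dependent wetness or noise-broken puddles, and scales paint contrast.
# An alternative switch replaces this with alpha-painted emissive weighted by time of day; a second output exposes the paint color.
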MaterialFunction /Game/Athena/WinterArt/Materials/MF_VPTexture
graph {nodes [
  n1 = CollectionParameter() {
   collection=&FortniteMaterialParameters
   parameterName=FortniteActiveTimeOfDay
  }
  n2 = VectorParameter() {group=VertexPaintTexture, parameterName=TOD_VP_Emissive, channelNames {a=Night, b=Evening, g=Day, r=Morning}, defaultValue=<rgba 0 0 0 1>}
  n3 = AppendVector(A=@n2, B=@n2)
  n4 = VertexColor()
  n5 = TextureObjectParameter() {
   group=VertexPaintTexture
   parameterName=VP_Emissive
   samplerType=SAMPLERTYPE_LinearGrayscale
   sortPriority=1
   texture=/Game/Global/Textures/Default/PointSplatter/T_GlowAlgae
  }
  n6 = TextureObjectParameter() {group=VertexPaintTexture, parameterName=VP_Color, sortPriority=1, texture=/Game/Global/Textures/Default/Blanks/T_MidgreysRGB}
  n7 = Multiply(A=ScalarParameter() {defaultValue=1, group=VertexPaintTexture, parameterName=VP_TextureScaleWS, sortPriority=4}, B=Constant() {r=256})
  n8 = VectorParameter() {
   group=VertexPaintTexture
   parameterName=VP_UVScaleByComponent
   sortPriority=3
   channelNames {a="", b=Normals, g=Specular, r=Diffuse}
   defaultValue=<rgba 1 1 1 0>
  }
  n9 = WorldPosition()
  n10 = StaticSwitchParameter(
   True=ScalarParameter() {group=WS Gradient, parameterName=WaterHeigth}
   False=CollectionParameter() {collection=/Game/Athena/Apollo/Environments/Landscape/MPC/Apollo_MPC_Terrain, parameterName=WaterLevel}
  ) {desc=Otherwise uses Apollo_MPC_Terrain, group=WS Gradient, parameterName=DebugWaterHeigth, sortPriority=11}
  n11 = Constant3Vector() {constant=<rgba 0 0 1 0>}
  n12 = StaticSwitchParameter(
   True=Subtract(A=ComponentMask(None=@n9.XYZ) {b=true, g=true, r=true}, B=Multiply(A=@n10, B=@n11))
   False=WorldPosition() {worldPositionShaderOffset=WPT_ExcludeAllShaderOffsets}.XYZ
  ) {desc=To match flood level automatically, group=WS Gradient, parameterName="Use Waterline Height "}
  n13 = MaterialFunctionCall(
   TextureObject=StaticSwitchParameter(True=@n5, False=@n6) {group=VertexPaintTexture, parameterName="VP Emissive (A)"}
   TextureSize=Multiply(A=@n7, B=@n8.Diffuse)
   WorldPosition=@n12
   Export Float 4=StaticBool() {value=true}
  ) {materialFunction=&WorldAlignedTexture}
  n14 = Multiply(
   A=Multiply(A=VectorParameter() {group=VertexPaintTexture, parameterName=VP_EmissiveColor, defaultValue=<rgba 1 0 0 0>}.RGB, B=DotProduct(A=@n1, B=@n3))
   B=MaterialFunctionCall(
    Transition Phase=OneMinus(None=@n4.A)
    Height Texture=@n13.XYZ Texture
    Contrast=ScalarParameter() {defaultValue=0.5, group=VertexPaintTexture, parameterName=VP_EmissiveContrast}
   ) {materialFunction=&HeightLerp}.Alpha
  )
  n15 = FunctionInput() {inputType=FunctionInput_MaterialAttributes}
  n16 = GetMaterialAttributes(None=@n15) {attributeGetTypes [
    #1
    #2
    #3
    #4
    #5
    #6
   ]}
  n17 = SetMaterialAttributes!(MaterialAttributes=@n15, EmissiveColor=Add(A=Multiply(A=@n14, B=@n16.BaseColor), B=@n16.EmissiveColor)) {attributeSetTypes [#6]}
  n18 = VectorParameter() {
   group=VertexPaintTexture
   parameterName=VP_UVScaleByComponent
   sortPriority=3
   channelNames {a="", b=Normals, g=Specular, r=Diffuse}
   defaultValue=<rgba 1 1 1 0>
  }
  n19 = Multiply(A=TextureCoordinate(), B=ScalarParameter() {defaultValue=1, group=VertexPaintTexture, parameterName=VP_TextureScale, sortPriority=4})
  n20 = StaticSwitchParameter(
   True=StaticSwitchParameter(
    True=@n13.XYZ Texture
    False=StaticSwitchParameter(True=@n13.Z Texture, False=@n13.Z Texture) {group=VertexPaintTexture, parameterName=WS_Z, sortPriority=7}
   ) {group=VertexPaintTexture, parameterName=WS_XYZ, sortPriority=7}
   False=TextureSample(UVs=Multiply(A=@n18.Diffuse, B=@n19), Tex=@n6).RGBA
  ) {group=VertexPaintTexture, parameterName=WS_VPTexture, sortPriority=6}
  n21 = Multiply(
   A=ScalarParameter() {defaultValue=1, group=VertexPaintTexture, parameterName=VP_DiffuseMod}
   B=ComponentMask(None=@n20) {b=true, g=true, r=true}
  )
  n22 = Multiply(
   A=GetMaterialAttributes(None=@n15) {attributeGetTypes [#7]}.OpacityMask
   B=ScalarParameter() {defaultValue=1, group=VertexPaintTexture, parameterName=VP_Background_HM_Max, sortPriority=9}
  )
  n23 = VertexColor()
  n24 = MaterialFunctionCall!(
   Diffuse 1
   Diffuse 2
   Transistion Phase=VertexColor().R
   Height Texture 1=Multiply(
    A=ComponentMask(None=@n20) {a=true}
    B=ScalarParameter() {defaultValue=1, group=VertexPaintTexture, parameterName=VP_TextureHeightMax, sortPriority=8}
   )
   Height Texture 2=@n22
   Contrast=Multiply(A=ScalarParameter() {defaultValue=5, group=VertexPaintTexture, parameterName=VP_Contrast, sortPriority=11}, B=@n23.B)
  ) {materialFunction=/Engine/Functions/Engine_MaterialFunctions02/Texturing/HeightLerpWithTwoHeightMaps}
  n25 = LinearInterpolate(A=@n16.BaseColor, B=@n21, Alpha=@n24.Alpha)
  n26 = TextureObjectParameter() {
   group=VertexPaintTexture
   parameterName=VP_RGBMask
   samplerType=SAMPLERTYPE_Masks
   sortPriority=2
   texture=&T_Fortnite_Default_S
  }
  n27 = MaterialFunctionCall(TextureObject=@n26, TextureSize=Multiply(A=@n7, B=@n8.Specular), WorldPosition=@n12) {materialFunction=&WorldAlignedTexture}
  n28 = StaticSwitchParameter(True=@n27.Z Texture, False=@n27.Z Texture) {group=VertexPaintTexture, parameterName=WS_Z, sortPriority=7}
  n29 = Multiply(A=@n18.Specular, B=@n19)
  n30 = MaterialFunctionCall(Float3=StaticSwitchParameter(
    True=StaticSwitchParameter(True=@n27.XYZ Texture, False=@n28) {group=VertexPaintTexture, parameterName=WS_XYZ, sortPriority=7}
    False=TextureSample(UVs=@n29, Tex=@n26) {samplerType=SAMPLERTYPE_Masks}.RGB
   ) {group=VertexPaintTexture, parameterName=WS_VPTexture, sortPriority=6}) {materialFunction=&BreakOutFloat3Components}
  n31 = Constant3Vector() {constant=<rgba 0 0 1 0>}
  n32 = TextureObjectParameter() {
   group=VertexPaintTexture
   parameterName=VP_Normal
   samplerType=SAMPLERTYPE_Normal
   sortPriority=3
   texture=/Engine/EngineMaterials/FlatNormal
  }
  n33 = MaterialFunctionCall(TextureObject=@n32, TextureSize=Multiply(A=@n7, B=@n8.Normals), WorldPosition=@n12) {materialFunction=&WorldAlignedNormal}
  n34 = StaticSwitchParameter(
   True=StaticSwitchParameter(
    True=@n33.XYZ Texture
    False=StaticSwitchParameter(True=@n33.Z Texture, False=@n33.Z Texture) {group=VertexPaintTexture, parameterName=WS_Z, sortPriority=7}
   ) {group=VertexPaintTexture, parameterName=WS_XYZ, sortPriority=7}
   False=TextureSample(UVs=Multiply(A=@n18.Normals, B=@n19), Tex=@n32) {samplerType=SAMPLERTYPE_Normal}.RGB
  ) {group=VertexPaintTexture, parameterName=WS_VPTexture, sortPriority=6}
  n35 = LinearInterpolate(
   A=@n16.EmissiveColor
   B=Multiply(A=@n16.EmissiveColor, B=ScalarParameter() {group=VertexPaintTexture, parameterName=VP_BG_Emissive, sortPriority=5})
   Alpha=@n24.Alpha
  ) {constB=0}
  n36 = SetMaterialAttributes!(
   MaterialAttributes=@n16.MaterialAttributes
   BaseColor=@n25
   Specular=LinearInterpolate(A=@n16.Specular, B=@n30.R, Alpha=@n24.Alpha)
   Roughness=LinearInterpolate(A=@n16.Roughness, B=@n30.B, Alpha=@n24.Alpha)
   Normal=LinearInterpolate(A=@n16.Normal, B=QualitySwitch(Default=@n34, Low=@n31, High=@n34, Medium=@n34), Alpha=@n24.Alpha)
   Metallic=LinearInterpolate(A=@n16.Metallic, B=@n30.G, Alpha=@n24.Alpha) {constB=0}
   EmissiveColor=@n35
  ) {attributeSetTypes [
    #1
    #2
    #3
    #4
    #5
    #6
   ]}
  n37 = ScalarParameter() {defaultValue=1, group=VertexPaintTexture, parameterName=VP_Mobile_NormalFlatness, sliderMax=1, sortPriority=2}
  n38 = SetMaterialAttributes(
   MaterialAttributes=@n16.MaterialAttributes
   BaseColor=@n25
   Specular=LinearInterpolate(
    A=@n16.Specular
    B=ScalarParameter() {defaultValue=0.733, group=VertexPaintTexture, parameterName=VP_Mobile_Specular, sliderMax=1, sortPriority=2}
    Alpha=@n24.Alpha
   )
   Roughness=LinearInterpolate(
    A=@n16.Roughness
    B=ScalarParameter() {defaultValue=0.8588, group=VertexPaintTexture, parameterName=VP_Mobile_Roughness, sliderMax=1, sortPriority=2}
    Alpha=@n24.Alpha
   )
   Normal=LinearInterpolate(
    A=@n16.Normal
    B=MaterialFunctionCall(Normal=@n16.Normal, Flatness=@n37) {materialFunction=&FlattenNormal}.Result
    Alpha=@n24.Alpha
   )
   Metallic=LinearInterpolate(
    A=@n16.Metallic
    B=ScalarParameter() {group=VertexPaintTexture, parameterName=VP_Mobile_Metallic, sliderMax=1, sortPriority=2}
    Alpha=@n24.Alpha
   ) {constB=0}
   EmissiveColor=@n35
  ) {attributeSetTypes [
    #1
    #2
    #3
    #4
    #5
    #6
   ]}
  n39 = StaticBool()
  n40 = FunctionInput(Preview=@n39) {
   bUsePreviewValueAsDefault=true
   description=Force only one sampler for mobile devices when material will be low on samplers.
   inputName=UseOneSampler
   inputType=FunctionInput_StaticBool
   sortPriority=4
  }
  n41 = StaticSwitchParameter(
   True=SetMaterialAttributes!(MaterialAttributes=@n15, BaseColor=@n25) {attributeSetTypes [#1]}
   False=StaticSwitch(
    True=ShadingPathSwitch(Default=@n36, Mobile=@n38) {desc=Migrated from FeatureLevelSwitch}
    False=@n36
    Value=StaticSwitchParameter(True=@n39, False=@n40) {group=VertexPaintTexture, parameterName=ForceMasksAndNormalSamplersEnabledOnMobile}
   )
  ) {group=VertexPaintTexture, parameterName=VP Diffuse Only}
  n42 = GetMaterialAttributes(None=StaticSwitchParameter(True=@n41, False=@n15) {desc=VP texture + Puddles, group=VertexPaintTexture, parameterName=DoublePaint, sortPriority=-5}) {attributeGetTypes [
    #1
    #2
    #3
    #4
   ]}
  n43 = DotProduct(A=CameraVectorWS(), B=VertexNormalWS())
  n44 = LinearInterpolate(
   A=ScalarParameter() {defaultValue=0.25, group=VertexPaintTexture, parameterName=Wet Diffuse Min, sortPriority=9}
   B=ScalarParameter() {defaultValue=0.75, group=VertexPaintTexture, parameterName=Wet Diffuse Max, sortPriority=9}
   Alpha=@n43
  )
  n45 = ScalarParameter() {defaultValue=0.9, parameterName=PuddlesFlatten}
  n46 = SetMaterialAttributes!(
   MaterialAttributes=@n42.MaterialAttributes
   BaseColor=Multiply(A=@n42.BaseColor, B=@n44) {constB=0.5}
   Specular=Multiply(A=@n42.Specular, B=@n44)
   Roughness=Multiply(A=@n42.Roughness, B=ScalarParameter() {defaultValue=0.1, group=VertexPaintTexture, parameterName=Wet Roughness, sortPriority=8}) {constB=0.1}
   Normal=LinearInterpolate(
    A=Constant3Vector() {constant=<rgba 0 0 1 0>}
    B=MaterialFunctionCall(Normal=@n42.Normal, Flatness=@n45) {materialFunction=&FlattenNormal}.Result
    Alpha=@n43
   )
  ) {attributeSetTypes [
    #1
    #2
    #3
    #4
   ]}
  n47 = WorldPosition()
  n48 = TextureSample(UVs=Divide(
    A=ComponentMask(None=@n47.XYZ) {g=true, r=true}
    B=ScalarParameter() {defaultValue=320, group=VertexPaintTexture, parameterName=Puddles Scale, sortPriority=2}
   )) {samplerSource=SSM_Wrap_WorldGroupSettings, samplerType=SAMPLERTYPE_Grayscale, texture=&LowResBlurredNoise}
  n49 = ScalarParameter() {defaultValue=1, group=VertexPaintTexture, parameterName=Puddles_Contrast, sortPriority=11}
  n50 = StaticSwitchParameter(
   True=MaterialFunctionCall(Transistion Phase=@n23.B, Height Texture 1=@n48.R, Height Texture 2=@n22, Contrast=@n49) {materialFunction=&HeightLerpWithTwoHeightMaps}.Alpha
   False=MaterialFunctionCall(
    Transition Phase=@n23.B
    Height Texture=ChannelMaskParameter(None=@n20) {group=VertexPaintTexture, parameterName=HM Channel Select}.R
    Contrast=@n49
   ) {materialFunction=&HeightLerp}.Alpha
  ) {desc=VP texture + Puddles, group=VertexPaintTexture, parameterName=DoublePaint, sortPriority=-5}
  n51 = StaticSwitchParameter(
   True=StaticSwitchParameter(
    True=BlendMaterialAttributes(A=@n41, B=@n46, Alpha=@n50) {vertexAttributeBlendType=UseA}
    False=BlendMaterialAttributes(A=@n46, B=@n15, Alpha=@n50) {vertexAttributeBlendType=UseA}
   ) {desc=VP texture + Puddles, group=VertexPaintTexture, parameterName=DoublePaint, sortPriority=-5}
   False=@n41
  ) {group=VertexPaintTexture, parameterName="VP Wet Paint (B Vertex Color)"}
  FunctionOutput(None=StaticSwitchParameter(True=StaticSwitchParameter(True=@n17, False=@n51) {group=VertexPaintTexture, parameterName="VP Emissive (A)"}, False=@n15) {desc="Red to Paint ", group=VertexPaintTexture, parameterName="UseVP Texture (R Vertex Color)", sortPriority=-1})
  FunctionOutput(None=@n21) {outputName=VP_Color, sortPriority=2}
 ]}
