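# Material function in the Athena WinterArt set that produces a tint color masked by vertex paint.
# A linear interpolate blends from a constant 1 to the single function input, using vertex color alpha as the blend factor.
# Zero alpha yields a neutral white, which leaves a color unchanged if multiplied, and full alpha yields the input tint.
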MaterialFunction /Game/Athena/WinterArt/Materials/MF_VPTintMask
graph {
 properties {description=This function helps overlay color to apply tint}
 nodes [FunctionOutput(None=LinearInterpolate(B=FunctionInput(), Alpha=VertexColor().A) {constA=1})]
}
