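# Material function that layers world-space procedural dirt over incoming material attributes, behind a static switch.
# A height-lerped mask combines normal direction, diffuse-alpha or custom AO height and a world-aligned leak texture, scaled per layer for up to four materials, optionally faded vertically.
# Dirt color blends two tints, with optional diffuse modulation and ground gradient.
# Optional Apollo landscape RVT sampling supplies ground color and height, bypassed for material proxies.
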MaterialFunction /Game/Athena/WinterArt/Materials/MF_WSDirt
graph {nodes [
  Constant()
  n1 = NamedRerouteUsage()
  n2 = StaticSwitchParameter(True=NamedRerouteUsage(), False=@n1) {group=Dirt, parameterName=UseRVTColor, sortPriority=-3}
  n3 = Multiply(
   A=OneMinus(None=NamedRerouteUsage())
   B=MaterialFunctionCall!(
    transitionFactor
    1=ScalarParameter() {defaultValue=1, group=Dirt, parameterName=L1_Dirt}
    2=ScalarParameter() {defaultValue=1, group=Dirt, parameterName=L2_Dirt}
    3=ScalarParameter() {defaultValue=1, group=Dirt, parameterName=L3_Dirt}
    4=ScalarParameter() {defaultValue=1, group=Dirt, parameterName=L4_Dirt}
    Use 2 Textures=StaticBoolParameter() {group=Material_2, parameterName=Use 2 Materials}
    Use 3 Textures=StaticBoolParameter() {group=Material_3, parameterName=Use 3 Materials}
    Use 4 Textures=StaticBoolParameter() {group=Material_4, parameterName=Use 4 Materials}
   ) {materialFunction=&/MF_LerpTexturesBasedOnInt_Float4_4Layers}.Result
  )
  n4 = NamedRerouteUsage()
  n5 = NamedRerouteUsage()
  n6 = BlendMaterialAttributes(
   A=@n5
   B=MaterialFunctionCall(NoRVT=@n1, RVT=MaterialProxyReplace(Realtime=@n2, MaterialProxy=@n1)) {materialFunction=&FN_RVT_Switch_Material}.Result
   Alpha=StaticSwitchParameter(True=Multiply(A=@n3, B=@n4), False=@n3) {group=Dirt, parameterName=Dirt Vertical Mask, sortPriority=7}
  ) {vertexAttributeBlendType=UseA}
  FunctionOutput(None=MaterialProxyReplace(Realtime=StaticSwitchParameter(True=@n6, False=@n5) {group=Dirt, parameterName=WS Dirt True, sortPriority=-5}, MaterialProxy=@n5))
  n7 = WorldPosition()
  n8 = RuntimeVirtualTextureSampleParameter() {
   group=Dirt
   materialType=WorldHeight
   parameterName=RVTHeight
   virtualTexture=/Game/Athena/Apollo/Maps/Landscape/VT/RVT_Apollo_Landscape_Height
  }
  NamedRerouteDeclaration(None=SphereMask(
    A=ComponentMask(None=@n7.XYZ) {b=true}
    B=@n8.WorldHeight
    Radius=ScalarParameter() {defaultValue=150, group=Dirt, parameterName=RVTFadeDistance}
    Hardness=ScalarParameter() {defaultValue=-2, group=Dirt, parameterName=RVTFade_Contrast}
   )) {name=Dirt RVT Mask, nodeColor=<rgba 0 0.7187 1 1>}
  n9 = RuntimeVirtualTextureSampleParameter!(UVs, World Position, Mip Bias=ScalarParameter() {defaultValue=5, group=Dirt, parameterName=RVT Color Mip}) {
   group=Dirt
   materialType=BaseColor_Normal_Specular_YCoCg
   mipValueMode=RVTMVM_MipBias
   parameterName=RVTColor
   virtualTexture=/Game/Athena/Apollo/Maps/Landscape/VT/RVT_Apollo_Landscape
  }
  NamedRerouteDeclaration(None=@n9.BaseColor) {name=DirtRVT_Color, nodeColor=<rgba 1 0.4688 0 1>}
  n10 = NamedRerouteUsage()
  n11 = VectorParameter() {group=Dirt, parameterName=Ground Color, sortPriority=6, defaultValue=<rgba 0.1049 0.6979 0.0655 0>}
  n12 = VectorParameter() {group=Dirt, parameterName=Dirt Color Inner, sortPriority=6, defaultValue=<rgba 0.1302 0.0969 0.0617 0>}
  n13 = MaterialFunctionCall(
   A=@n12.RGB
   B=VectorParameter() {group=Dirt, parameterName=Dirt Color, sortPriority=6, defaultValue=<rgba 0.4688 0.2779 0.1581 0>}.RGB
   Transition Phase=ScalarParameter() {defaultValue=0.5, group=Dirt, parameterName=Dirt Colors Balance, sortPriority=7}
   Height Texture=NamedRerouteUsage()
   Contrast=ScalarParameter() {defaultValue=0.5, group=Dirt, parameterName=Dirt Colors Balance Contrast, sortPriority=7}
  ) {materialFunction=&HeightLerp}
  n14 = StaticBoolParameter() {desc="RGB to dirt color, Alpha to Dirt Mask (AO)", group=Dirt, parameterName=Use Dirt_Custom AO for Colors and Mask, sortPriority=-2}
  n15 = StaticSwitchParameter(True=StaticSwitch(True=NamedRerouteUsage(), False=@n13.Results, Value=@n14), False=@n13.Results) {desc=Otherwise uses Diffuse Alpha, group=Dirt, parameterName=WS Custom AO Texture, sortPriority=-1}
  n16 = StaticSwitchParameter(True=LinearInterpolate(A=@n15, B=Multiply(A=@n11.RGB, B=@n15), Alpha=NamedRerouteUsage()), False=@n15) {group=Dirt, parameterName=UseGroundGradientColor, sortPriority=-3}
  n17 = SetMaterialAttributes!(
   MaterialAttributes
   BaseColor=StaticSwitchParameter(True=Multiply(A=@n10, B=@n16), False=@n16) {group=Dirt, parameterName=Modulate Dirt Diffuse, sortPriority=-3}
   Metallic=ScalarParameter() {defaultValue=0.2, group=Dirt, parameterName=Dirt Metallic, sortPriority=7}
   CustomData0=Constant()
   CustomData1=Constant() {r=1}
   Specular=ScalarParameter() {defaultValue=0.35, group=Dirt, parameterName=Dirt Specular, sortPriority=7}
   Roughness=LinearInterpolate(
    A=ScalarParameter() {defaultValue=0.85, group=Dirt, parameterName=Dirt Roughness, sortPriority=7}
    B=ScalarParameter() {defaultValue=1, group=Dirt, parameterName=Dirt RoughnessMax, sortPriority=7}
    Alpha=NamedRerouteUsage()
   )
   Normal=MaterialFunctionCall(Normal=NamedRerouteUsage(), Flatness=ScalarParameter() {defaultValue=1, group=Dirt, parameterName=Dirt Normal Flatten, sortPriority=8}) {materialFunction=&FlattenNormal}.Result
   Opacity=Constant() {r=1}
   SubsurfaceColor=Multiply(A=NamedRerouteUsage(), B=ScalarParameter() {defaultValue=1, group=Dirt, parameterName=DirtSSSValue, sortPriority=8})
  ) {attributeSetTypes [
    #1
    #2
    #3
    #4
    #5
    #6
    #7
    #8
    #9
   ]}
  NamedRerouteDeclaration(None=@n17) {name=Basic Colors, nodeColor=<rgba 1 0.5859 0 1>}
  n18 = NamedRerouteUsage()
  n19 = NamedRerouteUsage()
  NamedRerouteDeclaration(None=BlendMaterialAttributes(
    A=@n19
    B=SetMaterialAttributes!(MaterialAttributes=@n19, BaseColor=@n18) {attributeSetTypes [#1]}
    Alpha=NamedRerouteUsage()
   ) {pixelAttributeBlendType=UseA}) {name=RVT Colors, nodeColor=<rgba 0.7812 1 0 1>}
  n20 = GetMaterialAttributes(None=FunctionInput() {inputType=FunctionInput_MaterialAttributes}) {attributeGetTypes [#10, #7, #1]}
  NamedRerouteDeclaration(None=@n20.BaseColor) {name=Background Diffuse, nodeColor=<rgba 0.6875 0 1 1>}
  n21 = StaticSwitchParameter(True=VectorParameter() {group=Dirt, parameterName=World Location}.RGB, False=ActorPositionWS()) {group=Dirt, parameterName=Use World Location}
  n22 = StaticSwitch(True=NamedRerouteUsage(), False=@n21, Value=StaticBoolParameter() {group=Dirt, parameterName=UseRVT, sortPriority=-3})
  n23 = ComponentMask(None=Subtract(A=WorldPosition().XYZ, B=MaterialProxyReplace(Realtime=@n22, MaterialProxy=@n21))) {b=true}
  n24 = Divide(
   A=Subtract(A=@n23, B=ScalarParameter() {group=Dirt, parameterName=Dirt Vert Opacity Offset, sortPriority=7})
   B=ScalarParameter() {defaultValue=512, group=Dirt, parameterName=Dirt Vert Opacity Fade, sortPriority=7}
  )
  NamedRerouteDeclaration(None=Saturate(None=OneMinus(None=@n24))) {name=Dirt Height Mask, nodeColor=<rgba 0.9453 1 0 1>}
  NamedRerouteDeclaration(None=@n8.WorldHeight) {name=RVT Ground Height, nodeColor=<rgba 0 0.8359 1 1>}
  NamedRerouteDeclaration(None=@n20.MaterialAttributes) {name=BackgroundMaterial, nodeColor=<rgba 0 1 0.4609 1>}
  n25 = Divide(A=@n23, B=ScalarParameter() {defaultValue=100, group=Dirt, parameterName=GroundGradientDistance, sortPriority=6})
  NamedRerouteDeclaration(None=Saturate(None=OneMinus(None=@n25))) {name=GroundGradientMask, nodeColor=<rgba 0.6172 1 0 1>}
  n26 = Transform(None=NamedRerouteUsage())
  n27 = Constant3Vector() {constant=<rgba 0 0 -1 1>}
  n28 = Multiply(
   A=Saturate(None=DotProduct(A=@n26, B=@n27))
   B=ScalarParameter() {defaultValue=0.3, group=Dirt, parameterName=Dirt Down Value, sortPriority=5}
  )
  n29 = MaterialFunctionCall!(NoRVT=@n28, RVT=NamedRerouteUsage()) {materialFunction=/Game/Packages/Fortress_SharedMaterials/Base_Material_Functions/FN_RVT_Switch_Scalar}
  n30 = StaticBoolParameter() {group=Dirt, parameterName=UseRVT, sortPriority=-3}
  n31 = Max(
   A=Multiply(A=ComponentMask(None=@n26) {b=true}, B=ScalarParameter() {defaultValue=0.0571, group=Dirt, parameterName=Dirt Up Value, sortPriority=4})
   B=MaterialProxyReplace(Realtime=StaticSwitch(True=@n29.Result, False=@n28, Value=@n30), MaterialProxy=@n28)
  )
  n32 = NamedRerouteUsage()
  n33 = Multiply(A=OneMinus(None=@n32), B=ScalarParameter() {defaultValue=1, group=Dirt, parameterName=Dirt AO/Heigh Value, sortPriority=3})
  n34 = Add(A=ScalarParameter() {defaultValue=0.4976, group=Dirt, parameterName=Dirt Global, sortPriority=1}, B=Saturate(None=Add(A=@n31, B=@n33)))
  NamedRerouteDeclaration(None=MaterialFunctionCall!(
    Diffuse 1
    Diffuse 2
    Transistion Phase=Saturate(None=@n34)
    Height Texture 1=@n32
    Height Texture 2=NamedRerouteUsage()
    Contrast=ScalarParameter() {defaultValue=0.4, group=Dirt, parameterName=Dirt Contrast, sortPriority=2}
   ) {materialFunction=/Engine/Functions/Engine_MaterialFunctions02/Texturing/HeightLerpWithTwoHeightMaps}.Alpha) {name=Dirt Alpha mask, nodeColor=<rgba 0.5 1 0 1>}
  n35 = NamedRerouteUsage()
  n36 = TextureCoordinate() {coordinateIndex=1}
  n37 = FunctionInput(Preview=TextureCoordinate()) {inputName=UV, inputType=FunctionInput_Vector2}
  n38 = TextureObjectParameter() {
   group=Dirt
   parameterName=Dirt_CustomAO
   samplerType=SAMPLERTYPE_LinearGrayscale
   sortPriority=-2
   texture=/Game/Global/Textures/Default/Blanks/T_BlackLinearGrayscale
  }
  n39 = TextureSample(
   UVs=StaticSwitchParameter(
    True=TextureCoordinate() {coordinateIndex=2}
    False=StaticSwitchParameter(True=@n36, False=@n37) {group=Dirt, parameterName=UV2 AO Texture, sortPriority=-4}
   ) {group=Dirt, parameterName=UV3 AO Texture, sortPriority=-4}
   Tex=@n38
  )
  n40 = StaticBoolParameter() {desc="RGB to dirt color, Alpha to Dirt Mask (AO)", group=Dirt, parameterName=Use Dirt_Custom AO for Colors and Mask, sortPriority=-2}
  n41 = StaticSwitch(True=@n39.A, False=@n39.A, Value=@n40)
  n42 = Multiply(A=@n41, B=@n35)
  NamedRerouteDeclaration(None=StaticSwitchParameter(
    True=StaticSwitchParameter(True=@n42, False=@n41) {desc=Otherwise uses Custom AO only, group=Dirt, parameterName=WS Custom AO*Height, sortPriority=-2}
    False=@n35
   ) {desc=Otherwise uses Diffuse Alpha, group=Dirt, parameterName=WS Custom AO Texture, sortPriority=-2}) {name=HM, nodeColor=<rgba 0 0.2969 1 1>}
  NamedRerouteDeclaration(None=@n20.Normal) {name=Background Normal, nodeColor=<rgba 0.5938 0 1 1>}
  NamedRerouteDeclaration(None=@n39.RGB) {name=Custom AO Texture, nodeColor=<rgba 0.5703 1 0 1>}
  NamedRerouteDeclaration(None=@n20.OpacityMask) {name="Background Heightmap (Diffuse Alpha)", nodeColor=<rgba 0 1 0.8594 1>}
  n43 = TextureObjectParameter() {group=Dirt, parameterName=DirtMask, samplerType=SAMPLERTYPE_Masks, sortPriority=-1, texture=/Game/Athena/WinterArt/Textures/T_DirtLeaks_M}
  n44 = ScalarParameter() {defaultValue=256, group=Dirt, parameterName=Dirt Scale, sortPriority=-1}
  NamedRerouteDeclaration(None=StaticComponentMaskParameter(None=MaterialFunctionCall(TextureObject=@n43, TextureSize=@n44) {materialFunction=&WorldAlignedTexture}.XYZ Texture) {defaultR=true, group=Dirt, parameterName="DirtChannel (TestforNow)", sortPriority=-1}) {name=DirtMask, nodeColor=<rgba 0 1 0.3438 1>}
  Normalize()
  Reroute(None=@n20.BaseColor)
 ]}
