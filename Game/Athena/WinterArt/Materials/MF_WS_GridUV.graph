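# Material function that splits a surface by world height into a basement and three floors, giving each band its own UV scale and mask value.
# World Z is compared against rounded-up story parameters in 384-unit steps, with saturation producing near-hard band edges.
# These masks drive chained lerps that pick a rounded, clamped scale for the texture coordinates and blend four color-defaulted inputs into StoryMaskOut.
# The unsaturated basement height gradient is also output.
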MaterialFunction /Game/Athena/WinterArt/Materials/MF_WS_GridUV
graph {nodes [
  n1 = Ceil(None=ScalarParameter() {group=WS Gradient, parameterName=Floor1_Scale, sortPriority=13})
  n2 = Ceil(None=ScalarParameter() {defaultValue=1, group=WS Gradient, parameterName=Floor2_Scale, sortPriority=14})
  n3 = ComponentMask(None=WorldPosition().XYZ) {b=true}
  n4 = ScalarParameter() {group=WS Gradient, parameterName=BasementHeightWS, sortPriority=17}
  n5 = Add(A=@n4, B=ScalarParameter() {defaultValue=1, group=WS Gradient, parameterName=Story Mask1, sortPriority=5})
  n6 = Subtract(A=@n3, B=Multiply(A=Constant() {r=384}, B=Ceil(None=@n5)))
  n7 = Saturate(None=@n6)
  n8 = ScalarParameter() {defaultValue=2, group=WS Gradient, parameterName=Floor3_Scale, sortPriority=15}
  n9 = Constant() {r=384}
  n10 = Ceil(None=Subtract(A=ScalarParameter() {defaultValue=2, group=WS Gradient, parameterName=Story Mask2, sortPriority=6}))
  n11 = Saturate(None=Subtract(A=@n6, B=Multiply(A=@n9, B=@n10)))
  n12 = Constant() {r=384}
  n13 = Ceil(None=@n4)
  n14 = OneMinus(None=Subtract(A=Multiply(A=@n12, B=@n13), B=@n3))
  n15 = Saturate(None=@n14)
  n16 = LinearInterpolate(
   A=Ceil(None=ScalarParameter() {defaultValue=1, group=WS Gradient, parameterName=Basement_Scale, sortPriority=16})
   B=LinearInterpolate(A=LinearInterpolate(A=@n1, B=@n2, Alpha=@n7), B=Ceil(None=@n8), Alpha=@n11)
   Alpha=@n15
  )
  FunctionOutput(None=Multiply(A=TextureCoordinate(), B=Clamp(None=@n16) {maxDefault=1000, minDefault=0.5})) {outputName=UV Scale per Floor}
  n17 = Constant3Vector() {constant=<rgba 1 0 1 0>}
  n18 = LinearInterpolate(
   A=FunctionInput(Preview=Constant3Vector() {constant=<rgba 0 1 0 0>}) {bUsePreviewValueAsDefault=true, inputName=Floor1}
   B=FunctionInput(Preview=Constant3Vector() {constant=<rgba 1 0 0 0>}) {bUsePreviewValueAsDefault=true, inputName=Floor2}
   Alpha=@n7
  )
  n19 = FunctionInput(Preview=Constant3Vector() {constant=<rgba 0 0 1 0>}) {bUsePreviewValueAsDefault=true, inputName=Floor3}
  FunctionOutput(None=LinearInterpolate(
    A=FunctionInput(Preview=@n17) {bUsePreviewValueAsDefault=true, inputName=Basement}
    B=LinearInterpolate(A=@n18, B=@n19, Alpha=@n11)
    Alpha=@n15
   )) {outputName=StoryMaskOut}
  FunctionOutput(None=@n14) {outputName=BasementWorldHeight}
 ]}
