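# Material function that adds a world-height colour and roughness gradient and an optional flood waterline to incoming material attributes.
# Gradient height, measured from water level or MF_WS_GridUV's basement height, spans 384-unit floors, with an offset for proxy materials.
# Per-layer masks chosen by transitionFactor, metallic and an optional local-axis mask limit it; Low and Medium quality bypass it.
# Two soft edges then tint the waterline, scaling down roughness and specular.
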MaterialFunction /Game/Athena/WinterArt/Materials/MF_WSGradient
graph {nodes [
  n1 = FunctionInput() {inputName=MaterialIn, inputType=FunctionInput_MaterialAttributes}
  n2 = GetMaterialAttributes(None=@n1) {attributeGetTypes [#1, #2, #3]}
  n3 = VectorParameter() {group=WS Gradient, parameterName=WS GradientBottom, sortPriority=1, defaultValue=<rgba 1 1 1 0>}
  n4 = VectorParameter() {group=WS Gradient, parameterName=WS GradientTop, sortPriority=2, defaultValue=<rgba 1 1 1 1>}
  n5 = ScalarParameter() {group=WS Gradient, parameterName=WaterlineOffset, sortPriority=12}
  n6 = ComponentMask(None=WorldPosition().XYZ) {b=true}
  n7 = StaticSwitchParameter(
   True=ScalarParameter() {group=WS Gradient, parameterName=WaterHeigth}
   False=CollectionParameter() {collection=/Game/Athena/Apollo/Environments/Landscape/MPC/Apollo_MPC_Terrain, parameterName=WaterLevel}
  ) {desc=Otherwise uses Apollo_MPC_Terrain, group=WS Gradient, parameterName=DebugWaterHeigth, sortPriority=11}
  n8 = StaticSwitchParameter(
   True=Add(A=OneMinus(None=@n5), B=Subtract(A=@n6, B=@n7))
   False=MaterialFunctionCall!(Floor1, Floor2, Floor3, Basement) {materialFunction=&/MF_WS_GridUV}.BasementWorldHeight
  ) {desc=To match flood level automatically, group=WS Gradient, parameterName="Use Waterline Height "}
  n9 = Add(A=@n8, B=ScalarParameter() {group=WS Gradient, parameterName=ProxyHLODOffset, sortPriority=8})
  n10 = Constant() {r=384}
  n11 = Ceil(None=ScalarParameter() {defaultValue=1, group=WS Gradient, parameterName=GradientFloors, sortPriority=3})
  n12 = Saturate(None=Divide(A=MaterialProxyReplace(Realtime=@n8, MaterialProxy=@n9), B=Multiply(A=@n10, B=@n11)))
  n13 = FunctionInput(Preview=ComponentMask(None=TextureCoordinate() {coordinateIndex=1}) {r=true}) {bUsePreviewValueAsDefault=true, inputName=transitionFactor, inputType=FunctionInput_Scalar, sortPriority=-1}
  n14 = Constant() {r=2}
  n15 = If(
   A=Constant() {r=1}
   B=@n13
   "A > B"=ScalarParameter() {defaultValue=1, group=WS Gradient, parameterName=M1 Mask, sortPriority=4}
   "A < B"=ScalarParameter() {defaultValue=1, group=WS Gradient, parameterName=M2 Mask, sortPriority=5}
  ) {desc=0}
  n16 = ScalarParameter() {defaultValue=1, group=WS Gradient, parameterName=M3 Mask, sortPriority=6}
  n17 = Saturate(None=If(
    A=Constant() {r=3}
    B=@n13
    "A > B"=If(A=@n14, B=@n13, "A > B"=@n15, "A < B"=@n16)
    "A < B"=ScalarParameter() {defaultValue=1, group=WS Gradient, parameterName=M4 Mask, sortPriority=7}
   ))
  n18 = Multiply(A=@n2.Metallic, B=ScalarParameter() {group=WS Gradient, parameterName=Metal Mask})
  n19 = LinearInterpolate(A=LinearInterpolate(B=LinearInterpolate(A=@n3.RGB, B=@n4.RGB, Alpha=@n12), Alpha=@n17) {constA=1}, Alpha=@n18)
  n20 = ScalarParameter() {defaultValue=1, group=WS Gradient, parameterName=WS Gradient R Bottom, sortPriority=3}
  n21 = ScalarParameter() {defaultValue=1, group=WS Gradient, parameterName=WS Gradient R Top, sortPriority=3}
  n22 = Multiply(A=LinearInterpolate(B=LinearInterpolate(A=@n20, B=@n21, Alpha=@n12), Alpha=@n17) {constA=1}, B=@n2.Roughness)
  n23 = StaticBool() {value=true}
  n24 = MaterialFunctionCall!(
   transitionFactor
   1=ScalarParameter() {defaultValue=1, group=WS Gradient, parameterName=L1_GradMask, sortPriority=13}
   2=ScalarParameter() {defaultValue=1, group=WS Gradient, parameterName=L2_GradMask, sortPriority=14}
   3=ScalarParameter() {defaultValue=1, group=WS Gradient, parameterName=L3_GradMask, sortPriority=15}
   4=ScalarParameter() {defaultValue=1, group=WS Gradient, parameterName=L4_GradMask, sortPriority=16}
   Use 2 Textures=@n23
   Use 3 Textures=@n23
   Use 4 Textures=@n23
  ) {materialFunction=&/MF_LerpTexturesBasedOnInt_Float4_4Layers}
  n25 = WorldPosition()
  n26 = ActorPositionWS()
  n27 = Add(
   A=Transform(None=Subtract(A=@n25.XYZ, B=@n26)) {transformSourceType=TRANSFORMSOURCE_World, transformType=TRANSFORM_Local}
   B=Constant3Vector() {constant=<rgba 5 5 0 1>}
  )
  n28 = VectorParameter() {group=WS Gradient, parameterName=WS Gradient Mask By Mesh Axis, defaultValue=<rgba 0 1 0 0>}
  n29 = Saturate(None=Divide(A=DotProduct(A=@n27, B=@n28.RGB), B=Constant() {r=5}))
  n30 = BlendMaterialAttributes(
   A=@n2.MaterialAttributes
   B=SetMaterialAttributes!(
    MaterialAttributes=@n2.MaterialAttributes
    BaseColor=Multiply(A=@n2.BaseColor, B=@n19)
    Roughness=LinearInterpolate(A=@n22, B=@n2.Roughness, Alpha=@n18)
   ) {attributeSetTypes [#1, #2]}
   Alpha=StaticSwitchParameter(True=Multiply(A=@n24.Result, B=@n29), False=@n24.Result) {group=WS Gradient, parameterName=WS Gradient Axis Mask}
  )
  n31 = GetMaterialAttributes(None=QualitySwitch(Default=StaticSwitchParameter(True=@n30, False=@n1) {group=WS Gradient, parameterName=Use WS Gradient}, Low=@n1, Medium=@n1)) {attributeGetTypes [#1, #4, #2]}
  n32 = VectorParameter() {group=WS Gradient, parameterName=WaterlineDiff, defaultValue=<rgba 0.65 0.65 0.65 0>}
  n33 = WorldPosition()
  n34 = ScalarParameter() {group=WS Gradient, parameterName=WaterHeigth}
  n35 = CollectionParameter() {collection=&Apollo_MPC_Terrain, parameterName=WaterLevel}
  n36 = Add(
   A=OneMinus(None=ScalarParameter() {group=WS Gradient, parameterName=WaterlineOffset, sortPriority=12})
   B=Subtract(
    A=ComponentMask(None=@n33.XYZ) {b=true}
    B=StaticSwitchParameter(True=@n34, False=@n35) {desc=Otherwise uses Apollo_MPC_Terrain, group=WS Gradient, parameterName=DebugWaterHeigth, sortPriority=11}
   )
  )
  n37 = Saturate(None=Divide(A=@n36, B=ScalarParameter() {defaultValue=25, group=WS Gradient, parameterName=WaterlineEdge1, sortPriority=13}))
  n38 = Saturate(None=Divide(A=@n36, B=ScalarParameter() {defaultValue=100, group=WS Gradient, parameterName=WaterlineEdge2, sortPriority=13}))
  n39 = Multiply(A=@n31.BaseColor, B=Multiply(A=LinearInterpolate(A=@n32.RGB, Alpha=@n37), B=LinearInterpolate(A=@n32.RGB, Alpha=@n38)))
  n40 = Multiply(A=@n31.Specular, B=LinearInterpolate(Alpha=@n38) {constA=0.25}) {constB=0.25}
  n41 = Multiply(
   A=@n31.Roughness
   B=LinearInterpolate(A=ScalarParameter() {defaultValue=0.2, group=WS Gradient, parameterName=WaterlineRougness}, Alpha=@n38) {constA=0.2}
  )
  FunctionOutput(None=StaticSwitchParameter(
    True=SetMaterialAttributes!(MaterialAttributes=@n31.MaterialAttributes, BaseColor=@n39, Specular=@n40, Roughness=@n41) {attributeSetTypes [#1, #4, #2]}
    False=@n31.MaterialAttributes
   ) {desc=FloodWaterline, group=WS Gradient, parameterName=UseWaterline, sortPriority=11}) {outputName=Color*Gradient}
 ]}
