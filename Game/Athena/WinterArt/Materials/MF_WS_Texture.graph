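# Winter-art material function that optionally blends world-aligned color, spec, normal and emissive textures over incoming material attributes.
# Static switches select XYZ, Z-only or XY-plus-custom-Z projection, and diffuse-only or normal-only output.
# A height-lerp mask built from texture alpha, layer weights and optional UV mask drives the blend, optionally using landscape RVT height for ground-relative transitions.
# Emissive uses spec alpha, inverted Fresnel and optional time-of-day tint.
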MaterialFunction /Game/Athena/WinterArt/Materials/MF_WS_Texture
graph {nodes [
  n1 = NamedRerouteUsage()
  n2 = NamedRerouteUsage()
  n3 = GetMaterialAttributes(None=@n2) {attributeGetTypes [#1, #2]}
  n4 = SetMaterialAttributes!(MaterialAttributes=@n1, Normal=@n3.Normal) {attributeSetTypes [#2]}
  n5 = SetMaterialAttributes!(
   MaterialAttributes=StaticSwitchParameter(
    True=SetMaterialAttributes!(MaterialAttributes=@n1, BaseColor=@n3.BaseColor) {attributeSetTypes [#1]}
    False=StaticSwitchParameter(True=@n4, False=@n2) {group=WS_Texture, parameterName=WSTexture Normal Only, sortPriority=-4}
   ) {group=WS_Texture, parameterName=WSTexture Diffuse Only, sortPriority=-4}
   Opacity=NamedRerouteUsage()
  ) {desc=Adds new height map to 4 layers height, attributeSetTypes [#3]}
  n6 = WorldPosition()
  n7 = Constant()
  n8 = RuntimeVirtualTextureSampleParameter() {
   group=Dirt
   materialType=WorldHeight
   parameterName=RVTHeight
   virtualTexture=/Game/Athena/Apollo/Maps/Landscape/VT/RVT_Apollo_Landscape_Height
  }
  n9 = NamedRerouteDeclaration(None=SphereMask(
    A=ComponentMask(None=@n6.XYZ) {b=true}
    B=VirtualTextureFeatureSwitch!(No=@n7, Yes=@n8.WorldHeight)
    Radius=ScalarParameter() {defaultValue=150, group=WS_Texture, parameterName=WSTexture_RVT_FadeDistance, sortPriority=23}
    Hardness=Constant() {r=-3}
   )) {name=RVT Gradient, nodeColor=<rgba 1 0.0703 0 1>}
  n10 = ScalarParameter() {defaultValue=2, group=WS_Texture, parameterName=WSTexture_RVT_Boost, sortPriority=23}
  n11 = ScalarParameter() {defaultValue=1, group=WS_Texture, parameterName=WSTexture_TransitionScale, sortPriority=23}
  n12 = GetMaterialAttributes(None=NamedRerouteUsage()) {attributeGetTypes [#4, #5]}
  n13 = TextureCoordinate() {coordinateIndex=2}
  n14 = TextureCoordinate() {coordinateIndex=1}
  n15 = TextureObjectParameter() {group=WS_Texture, parameterName=WSTextures_Mask, sortPriority=17, texture=/Game/Global/Textures/Default/Blanks/T_MidgreysRGB}
  n16 = TextureSample(
   UVs=StaticSwitchParameter(
    True=TextureCoordinate() {coordinateIndex=3}
    False=StaticSwitchParameter(True=@n13, False=@n14) {desc=Uncheck to use UV1, group=WS_Texture, parameterName=WSTextures Alpha UV3, sortPriority=19}
   ) {group=WS_Texture, parameterName=WSTextures Alpha UV4, sortPriority=19}
   Tex=@n15
  ) {texture=/Game/Global/Textures/Default/Blanks/T_Blank_LegalWhite_sRGBColor}
  n17 = StaticSwitchParameter(
   True=Min(
    A=@n12.OpacityMask
    B=StaticComponentMaskParameter(None=@n16.RGB) {defaultR=true, group=WS_Texture, parameterName=WSTextures Mask Channel, sortPriority=18}
   )
   False=@n12.OpacityMask
  ) {group=WS_Texture, parameterName=WSTexture Custom Alpha, sortPriority=15}
  n18 = ScalarParameter() {group=WS_Texture, parameterName=WS_Texture_MetalMask, sortPriority=23}
  n19 = Saturate(None=Subtract(A=Multiply(A=@n11, B=@n17), B=Multiply(A=@n12.Metallic, B=@n18)))
  n20 = NamedRerouteUsage()
  n21 = OneMinus(None=MaterialFunctionCall!(
    Diffuse 1
    Diffuse 2
    Transistion Phase=Multiply(A=@n9, B=@n10)
    Height Texture 1=@n19
    Height Texture 2=@n20
    Contrast=ScalarParameter() {defaultValue=0.5, group=WS_Texture, parameterName=WSTexture_RVT_Contrast, sortPriority=23}
   ) {materialFunction=/Engine/Functions/Engine_MaterialFunctions02/Texturing/HeightLerpWithTwoHeightMaps}.Alpha)
  n22 = ScalarParameter() {defaultValue=0.5, group=WS_Texture, parameterName=WSTexture_MixContrast, sortPriority=23}
  n23 = Multiply(
   A=NamedRerouteUsage()
   B=StaticSwitchParameter(
    True=StaticSwitchParameter(True=@n9, False=@n21) {group=WS_Texture, parameterName=WSTexture RVT Height Simpler, sortPriority=23}
    False=MaterialFunctionCall(Transition Phase=@n19, Height Texture=@n20, Contrast=@n22) {materialFunction=&HeightLerp}.Alpha
   ) {group=WS_Texture, parameterName=WSTexture use RVT Height, sortPriority=23}
  )
  n24 = Abs(None=ComponentMask(None=VertexNormalWS()) {b=true})
  n25 = StaticSwitchParameter(True=Multiply(A=@n23, B=@n24), False=Subtract(A=@n23, B=@n24)) {desc="False= Z only", group=WS_Texture, parameterName=WS_Texture_Z only, sortPriority=-4}
  n26 = Saturate(None=StaticSwitchParameter(True=@n23, False=StaticSwitchParameter(True=@n23, False=@n25) {group=WS_Texture, parameterName=WS_Texture_XYZ, sortPriority=-4}) {group=WS_Texture, parameterName=WS_Texture_XY+Custom Z, sortPriority=-4})
  FunctionOutput(None=StaticSwitchParameter(True=BlendMaterialAttributes(A=@n1, B=@n5, Alpha=@n26) {vertexAttributeBlendType=UseB}, False=NamedRerouteUsage()) {group=WS_Texture, parameterName=Use_WS_Texture})
  n27 = ScalarParameter() {defaultValue=0.5, group=WS_Texture, parameterName=WS_TextureAlpha Transition, sortPriority=17}
  n28 = ComponentMask(None=VertexNormalWS()) {b=true}
  n29 = Saturate(None=@n28)
  n30 = Abs(None=@n28)
  n31 = LinearInterpolate(
   A=@n27
   B=LinearInterpolate(
    A=ScalarParameter() {defaultValue=0.5, group=WS_Texture, parameterName=WS_TextureAlpha Transition Z-, sortPriority=17}
    B=ScalarParameter() {defaultValue=0.5, group=WS_Texture, parameterName=WS_TextureAlpha Transition Z+, sortPriority=17}
    Alpha=@n29
   )
   Alpha=@n30
  )
  n32 = VectorParameter() {
   desc=You can use Vector Alpha to controll Tiled texture alpha channel
   group=WS_Texture
   parameterName="WS Texture Color "
   sortPriority=4
   defaultValue=<rgba 1 1 1 1>
  }
  n33 = AppendVector(A=@n32.RGB, B=@n32.RGB)
  n34 = TextureObjectParameter() {
   group=WS_Texture
   parameterName=WS_Spec
   samplerType=SAMPLERTYPE_Masks
   sortPriority=2
   texture=&T_Fortnite_Default_S
  }
  n35 = TextureObjectParameter() {group=WS_Texture, parameterName=WS_Normal, samplerType=SAMPLERTYPE_Normal, sortPriority=3, texture=/Engine/EngineMaterials/FlatNormal}
  n36 = StaticBool() {value=true}
  n37 = MaterialFunctionCall!(
   Diffuse Texture Object=TextureObjectParameter() {group=WS_Texture, parameterName=WS_Color, sortPriority=1, texture=&T_MidgreysRGB}
   Spec Texture Object=@n34
   TextureSize=ScalarParameter() {defaultValue=512, group=WS_Texture, parameterName=WS Textures Scale, sortPriority=5}
   --------------------------
   Use seperate settings for Z projection
   Use seperate texture for Z projection
   Texture Size Z projection
   Z Plane Diffuse Texture Object
   Z Plane Spec Texture Object
   "--------------------------      "
   ProjectionTransitionContrast
   Use Texture Alpha Heightmap to Refine the Transistion
   HeightLerpTransitionContrast
   PickTheDiffuseHeightChannel
   TheDiffuseHeightChannel
   "--------------------------       "
   NormalTextureObject=@n35
   World Space Normals=StaticBool()
   "--------------------------  "
   Export Float 4 Diffuse=@n36
   Export Float 4 Spec=@n36
   "--------------------------     "
   WorldPosition
   World Space Normal
  ) {materialFunction=/Engine/Functions/Engine_MaterialFunctions02/Texturing/WorldAlignedTextures_Complex}
  n38 = StaticSwitchParameter(True=@n37.Z Diffuse Texture, False=@n37.Z Diffuse Texture) {desc="False= Z only", group=WS_Texture, parameterName=WS_Texture_Z only, sortPriority=-4}
  n39 = StaticSwitchParameter(
   True=NamedRerouteUsage()
   False=Multiply(A=@n33, B=StaticSwitchParameter(True=@n37.XYZ Diffuse Texture, False=@n38) {group=WS_Texture, parameterName=WS_Texture_XYZ, sortPriority=-4})
  ) {group=WS_Texture, parameterName=WS_Texture_XY+Custom Z, sortPriority=-4}
  n40 = ScalarParameter() {defaultValue=0.5, group=WS_Texture, parameterName=WS_TextureAlpha Contrast, sortPriority=17}
  n41 = LinearInterpolate(
   A=@n40
   B=LinearInterpolate(
    A=ScalarParameter() {defaultValue=0.5, group=WS_Texture, parameterName=WS_TextureAlpha Contrast Z-, sortPriority=17}
    B=ScalarParameter() {defaultValue=0.5, group=WS_Texture, parameterName=WS_TextureAlpha Contrast Z+, sortPriority=17}
    Alpha=@n29
   )
   Alpha=@n30
  )
  NamedRerouteDeclaration(None=MaterialFunctionCall(
    Transition Phase=StaticSwitchParameter(True=@n31, False=@n27) {group=WS_Texture, parameterName=WS_Texture_XY+Custom Z, sortPriority=-4}
    Height Texture=ComponentMask(None=@n39) {a=true}
    Contrast=StaticSwitchParameter(True=@n41, False=@n40) {group=WS_Texture, parameterName=WS_Texture_XY+Custom Z, sortPriority=-4}
   ) {materialFunction=&HeightLerp}.Alpha) {name=WS TexturesAlpha, nodeColor=<rgba 0.5 0 1 1>}
  n42 = NamedRerouteDeclaration(None=FunctionInput() {inputType=FunctionInput_MaterialAttributes}) {name=WSTextures BG Mat Input, nodeColor=<rgba 0 1 0.9766 1>}
  n43 = ComponentMask(None=@n39) {b=true, g=true, r=true}
  n44 = ScalarParameter() {defaultValue=1, group=WS_Texture, parameterName=Boost Normals, sortPriority=3}
  n45 = ScalarParameter() {defaultValue=1, group=WS_Texture, parameterName=Boost Normals Z-, sortPriority=3}
  n46 = ScalarParameter() {defaultValue=1, group=WS_Texture, parameterName=Boost Normals Z+, sortPriority=3}
  n47 = Multiply(
   A=ComponentMask(None=@n37.Normal XYZ Texture) {g=true, r=true}
   B=LinearInterpolate(A=@n44, B=LinearInterpolate(A=@n45, B=@n46, Alpha=@n29), Alpha=@n30)
  )
  n48 = ComponentMask(None=@n37.Normal XYZ Texture) {b=true}
  n49 = StaticSwitchParameter(True=@n37.Normal Z Texture, False=@n37.Normal Z Texture) {desc="False= Z only", group=WS_Texture, parameterName=WS_Texture_Z only, sortPriority=-4}
  n50 = Multiply(A=ComponentMask(None=@n49) {g=true, r=true}, B=@n44)
  n51 = ComponentMask(None=@n49) {b=true}
  n52 = StaticSwitchParameter(
   True=NamedRerouteUsage()
   False=StaticSwitchParameter(True=AppendVector(A=@n47, B=@n48), False=AppendVector(A=@n50, B=@n51)) {group=WS_Texture, parameterName=WS_Texture_XYZ, sortPriority=-4}
  ) {group=WS_Texture, parameterName=WS_Texture_XY+Custom Z, sortPriority=-4}
  n53 = SetMaterialAttributes!(MaterialAttributes=@n42, Normal=@n52) {attributeSetTypes [#2]}
  n54 = StaticSwitchParameter(True=@n37.Z Spec, False=@n37.Z Spec) {desc="False= Z only", group=WS_Texture, parameterName=WS_Texture_Z only, sortPriority=-4}
  n55 = MaterialFunctionCall(Float4=StaticSwitchParameter(
    True=NamedRerouteUsage()
    False=StaticSwitchParameter(True=@n37.XYZ Spec, False=@n54) {group=WS_Texture, parameterName=WS_Texture_XYZ, sortPriority=-4}
   ) {group=WS_Texture, parameterName=WS_Texture_XY+Custom Z, sortPriority=-4}) {materialFunction=&BreakOutFloat4Components}
  n56 = SetMaterialAttributes!(MaterialAttributes, BaseColor=@n43, Metallic=@n55.G, Specular=@n55.G, Roughness=@n55.G, Normal=@n52, EmissiveColor=NamedRerouteUsage()) {attributeSetTypes [
    #1
    #5
    #6
    #7
    #2
    #8
   ]}
  NamedRerouteDeclaration(None=StaticSwitchParameter(
    True=SetMaterialAttributes!(MaterialAttributes=@n42, BaseColor=@n43) {attributeSetTypes [#1]}
    False=StaticSwitchParameter(True=@n53, False=@n56) {group=WS_Texture, parameterName=WSTexture Normal Only, sortPriority=-4}
   ) {group=WS_Texture, parameterName=WSTexture Diffuse Only, sortPriority=-4}) {name=WS Textures, nodeColor=<rgba 1 0 0.3984 1>}
  n57 = StaticBool() {value=true}
  NamedRerouteDeclaration(None=MaterialFunctionCall!(
    1=ScalarParameter() {defaultValue=1, group=WS_Texture, parameterName=L1_WSTexture_Mask, sortPriority=14}
    2=ScalarParameter() {defaultValue=1, group=WS_Texture, parameterName=L2_WSTexture_Mask, sortPriority=14}
    3=ScalarParameter() {defaultValue=1, group=WS_Texture, parameterName=L3_WSTexture_Mask, sortPriority=14}
    4=ScalarParameter() {defaultValue=1, group=WS_Texture, parameterName=L4_WSTexture_Mask, sortPriority=14}
    Use 2 Textures=@n57
    Use 3 Textures=@n57
    Use 4 Textures=@n57
   ) {materialFunction=&/MF_LerpTexturesBasedOnInt_Scalar_4Layers}.Result) {name=WS Textures Layers Mask, nodeColor=<rgba 0 1 0.5781 1>}
  n58 = VectorParameter() {
   desc=You can use Vector Alpha to controll Tiled texture alpha channel
   group=WS_Texture
   parameterName=WS Texture Color Z Down
   sortPriority=4
   defaultValue=<rgba 1 1 1 1>
  }
  n59 = VectorParameter() {
   desc=You can use Vector Alpha to controll Tiled texture alpha channel
   group=WS_Texture
   parameterName=WS Texture Color Z Up
   sortPriority=4
   defaultValue=<rgba 1 1 1 1>
  }
  n60 = LinearInterpolate(A=AppendVector(A=@n58.RGB, B=@n58.RGB), B=AppendVector(A=@n59.RGB, B=@n59.RGB), Alpha=@n29)
  n61 = ScalarParameter() {defaultValue=512, group=WS_Texture, parameterName=WS Textures Z Scale, sortPriority=5}
  n62 = StaticBool() {value=true}
  n63 = MaterialFunctionCall(
   TextureObject=TextureObjectParameter() {group=WS_Texture, parameterName=WS_Color Z, sortPriority=1, texture=&T_MidgreysRGB}
   TextureSize=@n61
   Export Float 4=@n62
  ) {materialFunction=&WorldAlignedTexture}
  NamedRerouteDeclaration(None=LinearInterpolate(A=Multiply(A=@n33, B=@n37.XY Diffuse Texture), B=Multiply(A=@n60, B=@n63.Z Texture), Alpha=@n30)) {name=WSTexture Z D, nodeColor=<rgba 0.8984 0 1 1>}
  n64 = TextureObjectParameter() {group=WS_Texture, parameterName=WS_Normal Z, samplerType=SAMPLERTYPE_Normal, sortPriority=3, texture=&FlatNormal}
  n65 = LinearInterpolate(
   A=@n37.Normal XY Texture
   B=MaterialFunctionCall(TextureObject=@n64, TextureSize=@n61) {materialFunction=&WorldAlignedNormal}.Z Texture
   Alpha=@n30
  )
  n66 = ComponentMask(None=@n65) {g=true, r=true}
  n67 = LinearInterpolate(
   A=ScalarParameter() {defaultValue=1, group=WS_Texture, parameterName=Boost Normals, sortPriority=3}
   B=LinearInterpolate(
    A=ScalarParameter() {defaultValue=1, group=WS_Texture, parameterName=Boost Normals Z-, sortPriority=3}
    B=ScalarParameter() {defaultValue=1, group=WS_Texture, parameterName=Boost Normals Z+, sortPriority=3}
    Alpha=@n29
   )
   Alpha=@n30
  )
  NamedRerouteDeclaration(None=AppendVector(A=Multiply(A=@n66, B=@n67), B=ComponentMask(None=@n65) {b=true})) {name=WSTexture Z N, nodeColor=<rgba 0 0.25 1 1>}
  n68 = TextureObjectParameter() {
   group=WS_Texture
   parameterName=WS_Spec Z
   samplerType=SAMPLERTYPE_Masks
   sortPriority=2
   texture=&T_Fortnite_Default_S
  }
  NamedRerouteDeclaration(None=LinearInterpolate(
    A=@n37.XY Spec
    B=MaterialFunctionCall(TextureObject=@n68, TextureSize=@n61, Export Float 4=@n62) {materialFunction=&WorldAlignedTexture}.Z Texture
    Alpha=@n30
   )) {name=WS Texture Z S, nodeColor=<rgba 0.7344 0 1 1>}
  n69 = ScalarParameter() {defaultValue=3, group=WS_Texture, parameterName=WS_EmissiveFresnel Exponent, sortPriority=15}
  n70 = ScalarParameter() {group=WS_Texture, parameterName=WS_Emissive_FresnelFraction, sortPriority=15}
  n71 = VertexNormalWS()
  n72 = Multiply(A=@n55.A, B=OneMinus(None=Fresnel(ExponentIn=@n69, BaseReflectFractionIn=@n70, Normal=@n71)))
  n73 = Multiply(A=@n43, B=@n72)
  n74 = VectorParameter() {group=WS_Texture, parameterName=WS_EmissiveTint, sortPriority=15}
  NamedRerouteDeclaration(None=StaticSwitchParameter(True=Multiply(A=@n73, B=@n74.RGB), False=Multiply(A=@n72, B=@n74.RGB)) {group=WS_Texture, parameterName=WS_UseDiffuseColorForEmissiveColor, sortPriority=15}) {name=WSEmissive, nodeColor=<rgba 0 0.3672 1 1>}
  n75 = VectorParameter() {group=WS_Texture, parameterName=WS_EmissiveMorning, sortPriority=15, defaultValue=<rgba 0.5 0.5 0.5 1>}
  n76 = VectorParameter() {group=WS_Texture, parameterName=WS_EmissiveDay, sortPriority=15, defaultValue=<rgba 0.5 0.5 0.5 1>}
  n77 = VectorParameter() {group=WS_Texture, parameterName=WS_EmissiveEvening, sortPriority=15, defaultValue=<rgba 0.5 0.5 0.5 1>}
  n78 = VectorParameter() {group=WS_Texture, parameterName=WS_EmissiveNight, sortPriority=15, defaultValue=<rgba 0.5 0.5 0.5 1>}
  n79 = NamedRerouteUsage()
  n80 = Multiply(
   A=MaterialFunctionCall(Color_Morning=@n75.RGB, Color_Day=@n76.RGB, Color_Evening=@n77.RGB, Color_Night=@n78.RGB) {materialFunction=&Color_Time_of_Day}.Result
   B=@n79
  )
  NamedRerouteDeclaration(None=StaticSwitchParameter(True=@n80, False=@n79) {group=WS_Texture, parameterName=WS_EmissiveTOD, sortPriority=15}) {name=WSEmissiveTOD, nodeColor=<rgba 0 1 0.3438 1>}
  Reroute(None=@n80)
 ]}
