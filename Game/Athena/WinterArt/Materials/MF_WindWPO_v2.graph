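# Material function (WindWPO v2) that adds switchable wind vertex animation to materials, with banner, flag, swing and turbulence variants plus a debug view.
# Wind actor data and panned noise drive rotations around parameter or Pivot Painter pivots, or turbulence offsets.
# Vertex-color masks and a wind-angle-based limiter shape the motion, with distance-fade helpers included.
# Normals are adjusted for rotation and blended with noise-warped scrolling cloud normal maps.
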MaterialFunction /Game/Athena/WinterArt/Materials/MF_WindWPO_v2
graph {nodes [
  n1 = Multiply(
   A=NamedRerouteUsage()
   B=AppendVector(
    A=ScalarParameter() {defaultValue=1, group=WindWPO_V2, parameterName=Wind Normal Texture Scale X, sortPriority=1}
    B=ScalarParameter() {defaultValue=1, group=WindWPO_V2, parameterName=Wind Normal Texture Scale Z, sortPriority=1}
   )
  )
  n2 = NamedRerouteUsage()
  n3 = Constant2Vector() {g=-1.5}
  n4 = TextureSample(UVs=Panner(Coordinate=@n1, Speed=Multiply(A=@n2, B=@n3))) {samplerType=SAMPLERTYPE_LinearColor, texture=&HighContrastNoise}
  n5 = Multiply(
   A=ConstantBiasScale(None=ComponentMask(None=@n4.RGB) {g=true, r=true}) {bias=-0.5, scale=2}
   B=ScalarParameter() {defaultValue=0.01, parameterName=Wind Normals Warp Amount}
  )
  n6 = ScalarParameter() {defaultValue=0.5, group=WindWPO_V2, parameterName=Wind Normals Speed Multiplier}
  n7 = Panner(Coordinate=@n1, Speed=Multiply(A=NamedRerouteUsage(), B=@n6))
  Add(A=@n5, B=@n7)
  n8 = NamedRerouteUsage()
  n9 = Panner(Coordinate=@n1, Speed=Multiply(A=Multiply(A=@n6, B=@n8), B=Constant2Vector() {g=1.02, r=-0.98}))
  Add(A=@n5, B=@n9)
  n10 = NamedRerouteUsage()
  n11 = Constant() {r=1}
  n12 = MaterialFunctionCall(X=Power(Base=NamedRerouteUsage()), Y=@n11, Z=@n11) {materialFunction=&MakeFloat3}
  n13 = Abs(None=Constant3Vector() {constant=<rgba 512 512 384 0>})
  n14 = Multiply(
   A=Divide(A=Multiply(A=@n10, B=@n12.Result), B=Multiply(A=@n13) {constB=-1})
   B=MaterialFunctionCall() {materialFunction=&ObjectScale}.Scale XYZ
  )
  n15 = ChannelMaskParameter(None=@n14) {group=WindWPO_V2, maskChannel=Blue, parameterName=Wind Texture Z Mapping, defaultValue=<rgba 0 0 1 0>}
  n16 = Multiply(A=TransformPosition(None=Constant3Vector()) {transformType=TRANSFORMPOSSOURCE_World}) {constB=3.14}
  n17 = ComponentMask(None=@n16) {g=true, r=true}
  n18 = ComponentMask(None=@n16) {b=true}
  n19 = Add(
   A=MaterialFunctionCall(X=ChannelMaskParameter(None=@n14) {group=WindWPO_V2, parameterName=Wind Texture X Mapping}.R, Y=@n15.R) {materialFunction=&MakeFloat2}.Result
   B=Divide(A=Add(A=@n17, B=@n18)) {constB=3.143}
  ) {constB=0}
  n20 = ScalarParameter() {defaultValue=0.05, group=WindWPO_V2, parameterName=Wind Texture Scale X, sortPriority=1}
  n21 = ScalarParameter() {defaultValue=0.3, group=WindWPO_V2, parameterName=Wind Texture Scale Z, sortPriority=1}
  n22 = ScalarParameter() {defaultValue=0.05, group=WindWPO_V2, parameterName=Wind Speed X, sortPriority=2}
  n23 = ScalarParameter() {defaultValue=-0.1, group=WindWPO_V2, parameterName=Wind Speed Z, sortPriority=2}
  n24 = Panner(
   Coordinate=Multiply(A=NamedRerouteDeclaration(None=@n19) {name=Wind Coordinates, nodeColor=<rgba 0.0781 1 0 1>}, B=AppendVector(A=@n20, B=@n21))
   Speed=NamedRerouteDeclaration(None=AppendVector(A=@n22, B=@n23)) {name=Wind Speed, nodeColor=<rgba 0.2422 0 1 1>}
  )
  n25 = NamedRerouteDeclaration(None=MaterialFunctionCall!(UV1=NamedRerouteDeclaration(None=@n24) {name=Panning UVs, nodeColor=<rgba 0.5937 1 0 1>}, Min=Constant(), Max=Constant() {r=0.4}) {materialFunction=&/MF_SimpleWindNoise}.Result) {name=Wind Normal, nodeColor=<rgba 0 0.5 1 1>}
  n26 = AppendVector(A=AppendVector(A=Constant() {r=2}, B=Constant() {r=7}), B=Constant())
  n27 = NamedRerouteUsage()
  n28 = OneMinus(None=@n27)
  n29 = Multiply(A=@n27, B=ScalarParameter() {defaultValue=1, parameterName=StormWindInfluence}) {constB=0.2}
  n30 = DotProduct(
   A=ComponentMask(None=Multiply(A=@n25, B=@n26)) {g=true, r=true}
   B=Normalize(VectorInput=MaterialFunctionCall(X=@n28, Y=@n29) {materialFunction=&MakeFloat2}.Result)
  )
  ConstantBiasScale(None=@n30)
  n31 = NamedRerouteUsage()
  n32 = StaticSwitchParameter(True=NamedRerouteUsage(), False=NamedRerouteUsage()) {parameterName=Swing Wind}
  n33 = StaticSwitchParameter(
   True=NamedRerouteUsage()
   False=StaticSwitchParameter(True=NamedRerouteUsage(), False=StaticSwitchParameter(True=@n31, False=@n32) {group=WindWPO_V2, parameterName=Flag Wind, sortPriority=1}) {group=WindWPO_V2, parameterName=Banner Wind, sortPriority=1}
  ) {group=WindWPO_V2 - Debug, parameterName=Wind Debug}
  n34 = NamedRerouteUsage()
  FunctionOutput(None=StaticSwitchParameter(
    True=MaterialFunctionCall!(Material=@n33, WorldPositionOffset=@n34) {materialFunction=/Engine/Functions/MaterialLayerFunctions/MatLayerBlend_AddWorldPositionOffset}.Blended Material
    False=NamedRerouteUsage()
   ) {group=WindWPO_V2, parameterName=Enable Wind WPO v2, sortPriority=0})
  MaterialExpressionNamedRerouteUsage_48 = NamedRerouteUsage()
  MaterialFunctionCall(X=@MaterialExpressionNamedRerouteUsage_48, Y=NamedRerouteUsage(), Z=NamedRerouteUsage()) {materialFunction=&MakeFloat3}
  n35 = VertexColor()
  n36 = AppendVector(A=@n35, B=@n35)
  n37 = InverseLinearInterpolate(
   A=ScalarParameter() {group=WindWPO_V2, parameterName=Remap WPO Mask Min}
   B=ScalarParameter() {defaultValue=1, group=WindWPO_V2, parameterName=Remap WPO Mask Max}
   Value=ChannelMaskParameter(None=@n36) {parameterName=Wind WPO Mask Channel}.R
  )
  NamedRerouteDeclaration(None=FunctionInput(Preview=Saturate(None=@n37)) {bUsePreviewValueAsDefault=true, inputName=Custom Mask, inputType=FunctionInput_Scalar, sortPriority=10}) {name=Vertex Mask, nodeColor=<rgba 0.5234 0 1 1>}
  n38 = ScalarParameter() {group=WindWPO_V2 - Intensity Limiter, parameterName=Limiter Gradient Remap Min, sortPriority=50}
  n39 = ScalarParameter() {defaultValue=1, group=WindWPO_V2 - Intensity Limiter, parameterName=Limiter Gradient Remap Max, sortPriority=51}
  n40 = ChannelMaskParameter(None=@n36) {group=WindWPO_V2 - Intensity Limiter, maskChannel=Green, parameterName=Limiter Vertex Color Channel, defaultValue=<rgba 0 1 0 0>}
  NamedRerouteDeclaration(None=StaticSwitchParameter(True=LinearInterpolate(A=@n38, B=@n39, Alpha=@n40.R), False=Constant() {r=1}) {group=WindWPO_V2 - Intensity Limiter, parameterName=Use Vertex Color For Limiter, sortPriority=1}) {name=Limiter Gradient, nodeColor=<rgba 0.5 0 0.0996 1>}
  n41 = GetMaterialAttributes(None=FunctionInput() {bUsePreviewValueAsDefault=true, inputType=FunctionInput_MaterialAttributes}) {attributeGetTypes [#1, #2]}
  NamedRerouteDeclaration(None=@n41.WorldPositionOffset) {name=Input WPO, nodeColor=<rgba 1 0.9609 0 1>}
  n42 = TextureSample(UVs=@n7) {samplerType=SAMPLERTYPE_Normal, texture=/Game/Global/Textures/Default/Noise/T_TilingClouds_N}
  n43 = TextureSample(UVs=@n9) {samplerType=SAMPLERTYPE_Normal, texture=&T_TilingClouds_N}
  n44 = AppendVector(A=ComponentMask(None=Add(A=@n42.RGB, B=@n43.RGB)) {g=true, r=true}, B=Multiply(A=@n42.B, B=@n43.B))
  n45 = LinearInterpolate(
   B=ScalarParameter() {defaultValue=1, group=WindWPO_V2, parameterName=Flatten Wind Normals}
   Alpha=Power(Base=VertexColor().R) {constExponent=0.2}
  ) {constA=1}
  NamedRerouteDeclaration(None=MaterialFunctionCall(
    BaseNormal=@n41.Normal
    AdditionalNormal=MaterialFunctionCall(Normal=@n44, Flatness=@n45) {materialFunction=&FlattenNormal}.Result
   ) {materialFunction=&BlendAngleCorrectedNormals}.Result) {name=Input Normal, nodeColor=<rgba 0 0.2793 1 1>}
  NamedRerouteDeclaration(None=@n41.MaterialAttributes) {name=Original Material, nodeColor=<rgba 1 1 1 1>}
  n46 = NamedRerouteUsage()
  n47 = MaterialFunctionCall(Rotation Axis=NamedRerouteUsage(), Rotation Angle=NamedRerouteUsage()) {materialFunction=&FixRotateAboutAxisNormals}
  NamedRerouteDeclaration(None=StaticSwitchParameter(
    True=MaterialFunctionCall(BaseNormal=@n46, AdditionalNormal=@n47.New Tangent Space Vertex Normal) {materialFunction=&BlendAngleCorrectedNormals}.Result
    False=@n46
   ) {parameterName=Use Correct Normals}) {name=Adjusted Normals, nodeColor=<rgba 0 0.2793 1 1>}
  n48 = VectorParameter() {group=WindWPO_V2, parameterName=Custom Rotation Axis, defaultValue=<rgba 0 0 1 0>}
  n49 = Constant3Vector() {constant=<rgba 0 0 1 0>}
  n50 = NamedRerouteUsage()
  n51 = StaticSwitchParameter(
   True=Transform(None=Normalize(VectorInput=@n48.RGB)) {transformSourceType=TRANSFORMSOURCE_Local}
   False=NamedRerouteDeclaration(None=CrossProduct(A=@n49, B=@n50)) {name=Bend Direction, nodeColor=<rgba 0.3359 1 0 1>}
  ) {group=WindWPO_V2, parameterName=Use Custom Rotation Axis}
  n52 = NamedRerouteUsage()
  n53 = Multiply(
   A=Multiply(A=Divide(A=@n52) {constB=10}, B=NamedRerouteUsage())
   B=ScalarParameter() {defaultValue=0.25, group=WindWPO_V2, parameterName=Wind Strength}
  )
  n54 = MaterialFunctionCall(X=ScalarParameter() {defaultValue=-1, parameterName=Pivot Elongation}) {materialFunction=&MakeFloat3}
  n55 = NamedRerouteUsage()
  n56 = NamedRerouteDeclaration(None=Add(A=VectorParameter() {group=WindWPO_V2, parameterName=Movement Pivot, defaultValue=<rgba 0 0 0 1>}.RGB, B=Multiply(A=@n54.Result, B=@n55))) {name=Local Pivot, nodeColor=<rgba 1 0.6328 0 1>}
  n57 = NamedRerouteUsage()
  NamedRerouteDeclaration(None=RotateAboutAxis(
    NormalizedRotationAxis=FunctionInput(Preview=@n51) {bUsePreviewValueAsDefault=true, inputName=Custom Rotation Axis, sortPriority=20}
    RotationAngle=NamedRerouteDeclaration(None=@n53) {name=Rotation Angle, nodeColor=<rgba 1 0.9375 0 1>}
    PivotPoint=@n56
    Position=@n57
   )) {name=Swing WPO, nodeColor=<rgba 1 0.3281 0 1>}
  n58 = MaterialFunctionCall!(Position, IsTrees, IsGrass, IsMediumPlant, IsMediumPlant_2024Anim, IsMediumPlant2ndLevel, MediumPlant_BasePivotPostion) {materialFunction=/Game/Environments/Asteria/Foliage/Wind/MF_WindActorData}
  NamedRerouteDeclaration(None=@n58.StormMask) {name=Storm Mask, nodeColor=<rgba 0.0547 0 1 1>}
  NamedRerouteDeclaration(None=Subtract(A=@n56, B=@n57)) {name=Debug Pivot, nodeColor=<rgba 1 0.5625 0 1>}
  n59 = NamedRerouteUsage()
  n60 = MaterialFunctionCall() {materialFunction=/Engine/Functions/Engine_MaterialFunctions02/PivotPainter/PivotPainter_PerObjectData}
  n61 = Constant3Vector() {constant=<rgba 0 1 0 0>}
  n62 = Constant3Vector() {constant=<rgba 0 -1 0 0>}
  n63 = VectorParameter() {
   group=WindWPO_V2 - Intensity Limiter
   parameterName=Limit Multiplier
   channelNames {a=Enable Invert Forward, b=Limit Multiplier -Y, g=Limit Multiplier +Y, r=Limit Multiplier X}
   defaultValue=<rgba 1 1 1 0>
  }
  n64 = Multiply(A=@n63.Enable Invert Forward, B=NamedRerouteUsage())
  n65 = DotProduct(
   A=Transform(None=Normalize(VectorInput=@n59)) {transformSourceType=TRANSFORMSOURCE_World, transformType=TRANSFORM_Local}
   B=StaticSwitchParameter(
    True=Transform(None=@n60.X-Axis Vector) {transformSourceType=TRANSFORMSOURCE_World, transformType=TRANSFORM_Local}
    False=LinearInterpolate(A=@n61, B=@n62, Alpha=@n64)
   ) {parameterName=Swing Wind}
  )
  NamedRerouteDeclaration(None=Abs(None=@n65)) {name=Wind Parallel, nodeColor=<rgba 1 0 0.6797 1>}
  n66 = Multiply(A=NamedRerouteUsage(), B=NamedRerouteUsage())
  n67 = NamedRerouteUsage()
  NamedRerouteDeclaration(None=SetMaterialAttributes!(MaterialAttributes=NamedRerouteUsage(), WorldPositionOffset=LinearInterpolate(A=@n66, Alpha=@n67) {constB=0}, Normal=NamedRerouteUsage()) {attributeSetTypes [#1, #2]}) {name=Banner Wind, nodeColor=<rgba 1 0 0.2109 1>}
  n68 = NamedRerouteUsage()
  n69 = NamedRerouteUsage()
  n70 = Constant()
  n71 = Sign(None=DotProduct(A=@n68, B=@n69))
  n72 = Multiply(A=If(A=@n71, "A > B"=@n71, "A == B"=Constant() {r=1}, "A < B"=@n71)) {constB=-1}
  n73 = ScalarParameter() {desc="Uses forward bias, but flips axis when going against wind", group=WindWPO_V2, parameterName=WindYBias, sortPriority=501}
  n74 = LinearInterpolate(
   A=@n69
   B=Transform(None=MaterialFunctionCall(X=@n70, Y=@n72, Z=@n70) {materialFunction=&MakeFloat3}.Result) {transformSourceType=TRANSFORMSOURCE_Local}
   Alpha=@n73
  )
  n75 = ScalarParameter() {defaultValue=0.5, group=WindWPO_V2, parameterName=WindForwardBias, sortPriority=500}
  NamedRerouteDeclaration(None=Normalize(VectorInput=LinearInterpolate(A=@n68, B=@n74, Alpha=@n75))) {name=Wind Direction Wiggle, nodeColor=<rgba 1 0 0.9609 1>}
  NamedRerouteDeclaration(None=MaterialFunctionCall!(FadeBlendDistance=ScalarParameter() {defaultValue=5000, parameterName=Fade Blend Distance}, AutoHandleBadDistance) {materialFunction=/Game/Global/MaterialFunctions/TechArt/MF_WPODisableBlend}.VertexShaderOutput) {name=Disable Blend, nodeColor=<rgba 0 0.7891 1 1>}
  n76 = Transform(None=Normalize(VectorInput=VectorParameter() {group=WindWPO_V2, parameterName=LocalForward, defaultValue=<rgba 0 1 0 1>}.RGB)) {transformSourceType=TRANSFORMSOURCE_Local}
  NamedRerouteDeclaration(None=Multiply(A=@n76) {constB=-1}) {name=World Forward, nodeColor=<rgba 0.3359 0 1 1>}
  NamedRerouteDeclaration(None=ChannelMaskParameter(None=@n36) {maskChannel=Blue, parameterName=Wind WPO Invert Forward Direction Channel, defaultValue=<rgba 0 0 1 0>}.R) {name=Invert Forward, nodeColor=<rgba 0.8281 1 0 1>}
  n77 = Constant3Vector() {constant=<rgba 1 0 0 0>}
  n78 = Constant3Vector() {constant=<rgba 0 1 0 0>}
  n79 = NamedRerouteUsage()
  n80 = Saturate(None=Sign(None=@n79))
  MaterialExpressionReroute_2 = Reroute(None=LinearInterpolate(B=LinearInterpolate(A=@n77, B=@n78, Alpha=@n80), Alpha=Abs(None=@n79)))
  MaterialExpressionSaturate_1 = Saturate(None=NamedRerouteUsage())
  n81 = Switch!(
   SwitchValue=ScalarParameter() {group=WindWPO_V2 - Debug, parameterName="Debug Swich (0 Wind, 1 Pivot, 2 Vert Color)"}
   Default
   WindInfluence=@MaterialExpressionReroute_2
   DebugPivot=@MaterialExpressionSaturate_1
   VertexColor=@MaterialExpressionNamedRerouteUsage_48
  ) {inputs [
    {
     inputName=WindInfluence
     input {
      expression=&self:MaterialExpressionReroute_2
      inputName=None
      mask=0
      maskA=0
      maskB=0
      maskG=0
      maskR=0
      outputIndex=0
     }
    }
    {
     inputName=DebugPivot
     input {
      expression=&self:MaterialExpressionSaturate_1
      inputName=None
      mask=0
      maskA=0
      maskB=0
      maskG=0
      maskR=0
      outputIndex=0
     }
    }
    {
     inputName=VertexColor
     input {
      expression=&self:MaterialExpressionNamedRerouteUsage_48
      inputName=None
      mask=0
      maskA=0
      maskB=0
      maskG=0
      maskR=0
      outputIndex=0
     }
    }
   ]}
  NamedRerouteDeclaration(None=SetMaterialAttributes!(MaterialAttributes, EmissiveColor=@n81, WorldPositionOffset=Constant()) {attributeSetTypes [#3, #1]}) {name=Debug Wind, nodeColor=<rgba 0 0.5312 1 1>}
  n82 = ScalarParameter() {defaultValue=4, group=WindWPO_V2, parameterName=Turbulence Strength}
  n83 = Multiply(
   A=ComponentMask(None=NamedRerouteUsage()) {g=true}
   B=LinearInterpolate(A=@n76, B=VertexNormalWS(), Alpha=ScalarParameter() {defaultValue=0.1, group=WindWPO_V2, parameterName=Vertex Normal Bias})
  )
  n84 = Multiply(A=Multiply(A=Multiply(A=@n82, B=@n83), B=NamedRerouteUsage())) {constB=100}
  NamedRerouteDeclaration(None=SetMaterialAttributes!(
    MaterialAttributes=NamedRerouteUsage()
    WorldPositionOffset=NamedRerouteDeclaration(None=@n84) {name=Turbulence WPO, nodeColor=<rgba 1 0 0.375 1>}
   ) {attributeSetTypes [#1]}) {name=Turbulence Wind, nodeColor=<rgba 0.8516 1 0 1>}
  n85 = NamedRerouteUsage()
  n86 = MaterialFunctionCall(
   BaseNormal=@n85
   AdditionalNormal=MaterialFunctionCall(Rotation Axis=NamedRerouteUsage(), Rotation Angle=NamedRerouteUsage()) {materialFunction=&FixRotateAboutAxisNormals}.New Tangent Space Vertex Normal
  ) {materialFunction=&BlendAngleCorrectedNormals}
  n87 = NamedRerouteUsage()
  n88 = Normalize(VectorInput=VectorParameter() {group=WindWPO_V2, parameterName=Flag Up Vctor, defaultValue=<rgba 0 0 1 1>}.RGB)
  n89 = NamedRerouteUsage()
  n90 = Multiply(A=Divide(A=ConstantBiasScale(None=@n89) {bias=-0.5, scale=1}) {constB=10}, B=NamedRerouteUsage())
  n91 = Multiply(A=ScalarParameter() {defaultValue=0.25, group=WindWPO_V2, parameterName=Wind Strength}, B=@n90)
  n92 = VectorParameter() {group=WindWPO_V2, parameterName=Movement Pivot, defaultValue=<rgba 0 0 0 1>}
  n93 = LocalPosition()
  n94 = LinearInterpolate(
   A=RotateAboutAxis(
    NormalizedRotationAxis=NamedRerouteDeclaration(None=@n88) {name=Flag Rotation Axis, nodeColor=<rgba 0 1 0.5313 1>}
    RotationAngle=NamedRerouteDeclaration(None=@n91) {name=Flag Angle, nodeColor=<rgba 1 0 0.6797 1>}
    PivotPoint=@n92.RGB
    Position=@n93.XYZ
   )
   B=Subtract(A=@n92.RGB, B=Transform(None=@n93.XYZ) {transformSourceType=TRANSFORMSOURCE_Local}) {constA=0}
   Alpha=Multiply(A=Abs(None=@n90), B=ScalarParameter() {defaultValue=3, group=WindWPO_V2, parameterName=Flag Pull Multiplier}) {constB=5}
  ) {constAlpha=1}
  n95 = NamedRerouteUsage()
  NamedRerouteDeclaration(None=SetMaterialAttributes!(
    MaterialAttributes=NamedRerouteUsage()
    WorldPositionOffset=Multiply(A=@n94, B=@n95)
    Normal=LinearInterpolate(A=@n85, B=@n86.Result, Alpha=@n87)
   ) {attributeSetTypes [#1, #2]}) {name=Flag Wind, nodeColor=<rgba 0.5826 1 0 1>}
  n96 = MaterialFunctionCall() {materialFunction=&PivotPainter_PerObjectData}
  n97 = WorldPosition()
  n98 = Normalize(VectorInput=CrossProduct(A=Constant3Vector() {constant=<rgba 0 0 1 0>}, B=NamedRerouteUsage()))
  n99 = Multiply(A=Multiply(A=NamedRerouteUsage()) {constB=0.1}, B=NamedRerouteUsage())
  n100 = NamedRerouteUsage()
  n101 = MaterialFunctionCall(Rotation Axis=@n98, Rotation Angle=@n99) {materialFunction=&FixRotateAboutAxisNormals}
  NamedRerouteDeclaration(None=SetMaterialAttributes(
    MaterialAttributes=NamedRerouteUsage()
    WorldPositionOffset=RotateAboutAxis(NormalizedRotationAxis=@n98, RotationAngle=@n99, PivotPoint=@n96.Pivot Position, Position=@n97.XYZ)
    Normal=MaterialFunctionCall(BaseNormal=@n100, AdditionalNormal=@n101.New Tangent Space Vertex Normal) {materialFunction=&BlendAngleCorrectedNormals}.Result
   ) {attributeSetTypes [#1, #2]}) {name=Swing Wind, nodeColor=<rgba 0 1 0.4844 1>}
  n102 = VectorParameter() {
   group=WindWPO_V2
   parameterName=Opposite Wind Strength
   channelNames {a="", b=Opposite Strength -Y, g=Opposite Strength +Y, r=Opposite Strength X}
  }
  n103 = MaterialFunctionCall!(
   1=AppendVector(A=@n63.Limit Multiplier +Y, B=@n102.Opposite Strength +Y)
   2=AppendVector(A=@n63.Limit Multiplier X, B=@n102.Opposite Strength X)
   3=AppendVector(A=@n63.Limit Multiplier -Y, B=@n102.Opposite Strength -Y)
   4
   A=ConstantBiasScale(None=@n65)
  ) {materialFunction=/Engine/Functions/Engine_MaterialFunctions02/Lerp_Multiple_Float2}
  NamedRerouteDeclaration(None=ComponentMask(None=@n103.Lerp 3 Inputs) {g=true}) {name=Opposite Strength, nodeColor=<rgba 0.7578 0 1 1>}
  n104 = Multiply(A=NamedRerouteUsage()) {constB=-1}
  NamedRerouteDeclaration(None=Multiply(
    A=LinearInterpolate(A=@n104, Alpha=@n30) {constA=-0.25}
    B=MaterialFunctionCall() {materialFunction=&ObjectScale}.Scale XYZ
   )) {name=Wind Noise, nodeColor=<rgba 1 0 0.6328 1>}
  NamedRerouteDeclaration(None=Normalize(VectorInput=@n58.WindI_Direction)) {name=Wind Direction, nodeColor=<rgba 0 1 0.0859 1>}
  n105 = ComponentMask(None=@n103.Lerp 3 Inputs) {r=true}
  NamedRerouteDeclaration(None=LinearInterpolate(B=NamedRerouteDeclaration(None=@n105) {name=Limiter, nodeColor=<rgba 1 0 0.375 1>}, Alpha=NamedRerouteUsage()) {constA=1}) {name=Limiter Multiplier, nodeColor=<rgba 1 0.6797 0 1>}
  n106 = InverseLinearInterpolate(
   A=ScalarParameter() {defaultValue=3000, parameterName=WPO Fade End}
   B=ScalarParameter() {defaultValue=1000, parameterName=WPO Fade Start}
   Value=Distance(A=ActorPositionWS(), B=CameraPositionWS())
  )
  NamedRerouteDeclaration(None=Power(Base=Saturate(None=@n106))) {name=Distance Blend, nodeColor=<rgba 0.8047 0 1 1>}
  NamedRerouteDeclaration(None=@n58.WindI_Direction_WithNoise) {name=WindIntensity, nodeColor=<rgba 0 0.5547 1 1>}
  StaticSwitchParameter(
   True=TransformPosition(None=MaterialFunctionCall() {materialFunction=&PivotPainter_PerObjectData}.Pivot Position) {transformSourceType=TRANSFORMPOSSOURCE_World}
   False=@n10
  ) {parameterName=Swing Wind}
  Subtract(
   A=NamedRerouteDeclaration(None=LocalPosition() {includedOffsets=ExcludeOffsets}.XYZ) {name=Local Position, nodeColor=<rgba 0.5 0.15 0.15 1>}
   B=NamedRerouteUsage()
  )
  Transform(None=@n88) {transformSourceType=TRANSFORMSOURCE_Local}
 ]}
