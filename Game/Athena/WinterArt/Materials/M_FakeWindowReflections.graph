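# Material function that adds fake world-space window reflections to the emissive of incoming material attributes.
# It samples a cubemap parameter along the world reflection vector and tints it by ReflectionsColor.
# The ceiling of one minus the second UV channel's U coordinate masks the result before it is added to the existing emissive.
# The UseFakeWindowReflections static switch toggles the effect, otherwise passing emissive through unchanged.
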MaterialFunction /Game/Athena/WinterArt/Materials/M_FakeWindowReflections
graph {
 properties {description=This function is used to create a fake world positioned window reflection.}
 nodes [
  n1 = OneMinus(None=ComponentMask(None=TextureCoordinate() {coordinateIndex=1}) {r=true})
  Floor(None=@n1)
  n2 = GetMaterialAttributes(None=FunctionInput() {inputType=FunctionInput_MaterialAttributes}) {attributeGetTypes [#1]}
  n3 = VectorParameter() {group=NeonEmissive, parameterName=ReflectionsColor, defaultValue=<rgba 1 1 1 1>}
  n4 = TextureSampleParameterCube!(UVs=ReflectionVectorWS(), Apply View MipBias) {
   group=NeonEmissive
   parameterName=FakeReflection
   samplerType=SAMPLERTYPE_LinearColor
   texture=/Game/Environments/Sets/Neo/Architecture/Textures/T_FakeWindowReflections
  }
  n5 = Add(A=Multiply(A=Multiply(A=@n3.RGB, B=@n4.RGB), B=Ceil(None=@n1)), B=@n2.EmissiveColor)
  FunctionOutput(None=SetMaterialAttributes!(
    MaterialAttributes=@n2.MaterialAttributes
    EmissiveColor=StaticSwitchParameter(True=@n5, False=@n2.EmissiveColor) {parameterName=UseFakeWindowReflections}
   ) {attributeSetTypes [#1]})
 ]
}
